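# Masked, lit material for Niagara mesh particles, evidently burning pieces of the Patches character.
# A radial sphere mask minus cloud noise is compared against two thresholds from the Dynamic Parameter's Dissolve, and their difference forms an edge band.
# That band blends the scalable, offsettable body texture toward particle color and feeds an emissive glow through MF_FX_EmissiveAdjustments.
# Standalone Dissolve, Thickness and Emissive parameters have no visible connections.
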Material /BRCosmetics/Characters/Player/Male/Medium/Bodies/M_MED_Patches/FX/M_Patches_BurningPieces
graph {
 properties {bUsedWithNiagaraMeshParticles=true, blendMode=BLEND_Masked, pixelDepthOffsetMode=PDOM_Legacy, usedShadingModels=Default Lit}
 nodes [
  n1 = Multiply(A=TextureCoordinate(), B=ScalarParameter() {defaultValue=1, parameterName=Scale, sliderMax=5})
  n2 = AppendVector(
   A=ScalarParameter() {parameterName=Offset X, sliderMax=1, sliderMin=-1}
   B=ScalarParameter() {parameterName=Offset Y, sliderMax=1, sliderMin=-1}
  )
  n3 = ParticleColor()
  n4 = TextureCoordinate()
  n5 = Constant2Vector() {g=0.5, r=0.5}
  n6 = Subtract(
   A=Multiply(A=SphereMask(A=@n4, B=@n5) {attenuationRadius=0.5, hardnessPercent=0}) {constB=2}
   B=TextureSample() {texture=/BRCosmetics/Athena/Effects/Particles/Skydiving/NS_Hightower_Date/T_AtmosphericCloudNoise_Vertical}.R
  )
  n7 = DynamicParameter() {defaultValue=<rgba 0 0.1 1 1>, paramNames [Dissolve,Thickness,Param3,Param4]}
  n8 = Constant() {r=1}
  n9 = Constant()
  n10 = Subtract(
   A=If(A=@n6, B=@n7.Dissolve, "A > B"=@n8, "A < B"=@n9)
   B=If(A=@n6, B=Add(A=@n7.Dissolve, B=@n7.Dissolve) {constB=0.2}, "A > B"=@n8, "A < B"=@n9)
  )
  LinearInterpolate(
   A=TextureSample(UVs=Add(A=@n1, B=@n2)) {texture=/Game/Characters/Player/Male/Medium/Bodies/M_MED_Patches/Textures/T_Character_-_Patches_Body_D}.RGB
   B=@n3.RGB
   Alpha=@n10
  )
  MaterialFunctionCall(Emissive=LinearInterpolate(B=@n3.RGB, Alpha=@n10)) {materialFunction=&MF_FX_EmissiveAdjustments}
  ScalarParameter() {parameterName=Dissolve, sliderMax=2}
  ScalarParameter() {defaultValue=0.2, parameterName=Thickness, sliderMax=1}
  VectorParameter() {parameterName=Emissive, defaultValue=<rgba 5 0 4.6231 1>}
 ]
}
