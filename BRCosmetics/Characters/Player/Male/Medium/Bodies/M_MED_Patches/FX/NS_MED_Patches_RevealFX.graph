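# Niagara reveal effect for the male Patches character body, producing smoke, burning mesh pieces, magenta-purple embers and flame splashes on the skeletal mesh.
# Emitters sample the User skeletal mesh, using two spawn groups with mirrored side offsets.
# A system Heart Burn value scales spawn rates, radii and sizes, while KillReactBlend adjusts smoke opacity and material parameters.
# Front-end versus gameplay switches and time-of-day scalars tune timing, spawning and brightness.
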NiagaraSystem /BRCosmetics/Characters/Player/Male/Medium/Bodies/M_MED_Patches/FX/NS_MED_Patches_RevealFX
dependencies {
 usedDataInterfaces [
  &NiagaraDataInterfaceCurve
  &NiagaraDataInterfaceArrayFloat
  &NiagaraDataInterfaceMeshRendererInfo
  &NiagaraDataInterfaceVector2DCurve
  &NiagaraDataInterfaceColorCurve
  &NiagaraDataInterfaceArrayFloat3
  &NiagaraDataInterfaceArrayFloat2
 ]
 usedDynamicInputs [&RandomRangeVector]
 usedModules [
  &SystemState
  &Timeline
  &self:SystemSpawnScript.SystemScriptSource.SystemScriptGraph.NiagaraNodeAssignment_0.SetVariables_#1
  &self:SystemSpawnScript.SystemScriptSource.SystemScriptGraph.NiagaraNodeAssignment_1.SetVariables_#2
 ]
}
emitters {
 Body_Smoke {
  emitterUpdateScript [
   &EmitterState(
    Life Cycle Mode=::NewEnumerator1
    Loop Behavior=::NewEnumerator1
    Loop Duration=&FN_FrontEndFloatModifier(Gameplay=2, Front End=3)
    UseLoopDelay=true
   )
   &SpawnRate(
    Spawn Probability=&FloatFromCurve(
     FloatCurve=DataInterface() {propertyValues {Curve {keys [
         {tangentMode=RCTM_User, time=0, value=1}
         {arriveTangent=-0.0405, interpMode=RCIM_Cubic, leaveTangent=-0.0405, tangentMode=RCTM_User, time=0.5, value=1}
         {arriveTangent=-3.3726, interpMode=RCIM_Cubic, leaveTangent=-3.3726, tangentMode=RCTM_User, time=1, value=0.4}
        ]}}}
     CurveIndex=$Emitter.NormalizedLoopAge
     Scale Curve=&Lerp_Float(A=0.1, B=1, Alpha=$System.Heart Burn Clamped)
    )
    SpawnRate=15
    Use Spawn Probability=true
   )
  ]
  particleSpawnScript [
   &InitializeParticle(
    Color=<rgba 0.01 0.01 0.01 1>
    Lifetime=&RandomRangeFloat(Minimum=0.6, Maximum=1.5)
    Mesh Scale=&Multiply_VectorByFloat(Vector=<1 1 2>, Float=&Lerp_Float(A=0.6, B=1, Alpha=$System.Heart Burn Clamped))
    Mesh Scale Mode=::NewEnumerator4
    UsePositionOffset=false
    Write Material Random=true
   )
   &FN_SetSpawnGroup(Number of Spawn Groups=2)
   /SetVariables() {isSetParameters=true}
   &CameraOffset(Camera Offset Amount=&MakeCustomFloatFromBool(
     Bool=&SetBoolByIntComparison(A=&SpawnGroup(), B=3, Comparison Type=::NewEnumerator4)
     True Float=-10
     False Float=16
    ))
   /SetVariables(
    Particles.Radius:NiagaraFloat=&Multiply_Float(A=10, B=$System.Heart Burn Clamped)
    Particles.Theta:NiagaraFloat=&RandomRangeFloat(Minimum=0, Maximum=1)
   ) {isSetParameters=true}
   &SkeletalMeshLocation(
    Bone Sampling Mode=::NewEnumerator4
    MaskSpawnGroup=false
    Orientation Sampling=::NewEnumerator1
    Sampled Position Offset=&MakeVector(
     X=&Add_Float(
      A=-5
      B=&Multiply_Float(A=$Particles.Radius, B=&Cosine(Normalized Angle=$Particles.Theta, Period=1, Scale=1, Bias=0))
     )
     Y=&MaskFloatBySpawnGroup(Masked Float=-9, Passthrough Float=9)
     Z=&Add_Float(
      A=0
      B=&Multiply_Float(A=$Particles.Radius, B=&Sine(Normalized Angle=$Particles.Theta, Period=1, Scale=1, Bias=0))
     )
    )
    Skeletal Mesh=$User.SkeletalMesh
   )
   &AddVelocity(Velocity=<0 0 1>) {enabled=false}
   /SetVariables(Particles.MeshOrientation=&MakeQuatFromAxisAngle(Axis=<0 0 1>, AngleInDegrees=&RandomRangeFloat(Minimum=0, Maximum=180), NormalizeAxis=true)) {isSetParameters=true}
   &ApplyInitialForces()
   &ShapeLocation(MaskSpawnGroup=false, Override Seed=false, Sphere Radius=10, UseOwnerScale=true) {enabled=false}
  ]
  particleUpdateScript [
   &ParticleState()
   &ScaleMeshSize(Scale Factor=&MakeVector(
     X=1
     Y=1
     Z=&SelectFloatFromArray(
      Direct Array Index=&SpawnGroup()
      Float Selection Array=DataInterface() {propertyValues {FloatData [1.1,2,1.1,1.1,2,1.1,1]}}
     )
    ))
   &SkeletalMeshLocation(
    Bone Sampling Mode=::NewEnumerator4
    Initialize Particle Position To Sampled Position=false
    "Interpolate Mesh Sampling (Update)"=true
    MaskSpawnGroup=false
    Orientation Sampling=::NewEnumerator1
    Position Sampling=::NewEnumerator1
    Sampled Position Coordinate Space=::World
    Sampled Position Offset Coordinate Space:ENiagaraSkelMeshTransforms=::NewEnumerator1
    Skeletal Mesh=$User.SkeletalMesh
    Write To Particle Position=false
   ) {module=SkeletalMeshLocation001}
   &SkeletalMeshLocation(
    Bone / Socket Index=&Add_Integer(A=$Particles.Initial.SpawnGroup, B=1)
    Bone Sampling Mode=::NewEnumerator4
    "Interpolate Mesh Sampling (Update)"=true
    MaskSpawnGroup=false
    Orientation Sampling=::NewEnumerator1
    Sampled Position Coordinate Space=::World
    Skeletal Mesh=$User.SkeletalMesh
    Write To Particle Position=false
   ) {enabled=false, module=SkeletalMeshLocation002}
   /SetVariables(Particles.SkelSampleA:NiagaraPosition=$Particles.SkeletalMeshLocation001.SampledPosition) {enabled=false, isSetParameters=true}
   /SetVariables(Particles.SkelSampleB:NiagaraPosition=$Particles.SkeletalMeshLocation002.SampledPosition) {enabled=false, isSetParameters=true}
   /SetVariables(Particles.Position=&LerpPosition(A=$Particles.SkelSampleA, B=$Particles.SkelSampleB, Alpha=&RandomRangeFloat(Minimum=0, Maximum=1))) {enabled=false, isSetParameters=true}
   /SetVariables(Particles.Center size:Vector2f=&Multiply_Vector2D_ByFloat(Float=2, Vector2D=<85 160>)) {isSetParameters=true}
   /SetVariables(Particles.Center color:LinearColor=&MakeLinearColorFromVectorAndFloat(
     "Vector (RGB)"=&MakeVectorFromLinearColorRGB(Color=&MultiplyLinearColor(
       Linear Color=<rgba 50.9375 56.1121 60 1>
       Scale Factor=&MakeVector4_FromFloat(Float=&Multiply_Float(
         A=&FNTimeOfDayScalar(Night Scale=0.5, Morning Scale=0.2, Evening Scale=0.2)
         B=&FN_FrontEndFloatModifier(Gameplay=1, Front End=1)
        ))
      ))
     "Float (Alpha)"=&Multiply_Float(
      A=0.11
      B=&FloatFromCurve(
       FloatCurve=DataInterface() {propertyValues {Curve {keys [
           {tangentMode=RCTM_User, time=0, value=1}
           {arriveTangent=-1.4447, interpMode=RCIM_Cubic, leaveTangent=-1.4447, tangentMode=RCTM_User, time=0.6, value=1}
           {arriveTangent=-1.2179, interpMode=RCIM_Cubic, leaveTangent=-1.2179, tangentMode=RCTM_User, time=1, value=0.65}
          ]}}}
       Scale Curve=&FloatFromCurve(
        FloatCurve=DataInterface() {propertyValues {Curve {keys [<at 0 0.2>,<at 0.0415 4>,<at 0.1132 3>,<at 0.198 4>,<at 0.2511 3>,<at 0.3284 4>,<at 0.4103 3>,<at 0.4812 4>,<at 0.5793 3>,<at 0.6755 4>,<at 0.7472 3>,<at 0.8114 4>,<at 0.8663 3>,<at 0.9265 4>,<at 1 2>]}}}
        Scale Curve=&FloatFromCurve(
         FloatCurve=DataInterface() {propertyValues {Curve {keys [
             {interpMode=RCIM_Cubic, tangentMode=RCTM_User, time=0, value=0}
             {interpMode=RCIM_Cubic, tangentMode=RCTM_User, time=0.6, value=1}
             {tangentMode=RCTM_User, time=1, value=0}
            ]}}}
         CurveIndex=$Emitter.Age
         Scale Curve=&Multiply_Float(
          A=$User.global Alpha
          B=&ClampFloat(
           Float=&FN_LOD_Distance_Remap(Start Distance=200, End Distance=300, Minimum Output=0.3, Maximum Output=1)
           Min=0.3
          )
         )
        )
       )
      )
     )
    )) {isSetParameters=true}
   &ScaleColor(
    Scale Alpha=&Multiply_Float(
     A=&FN_FrontEndFloatModifier(Gameplay=1, Front End=1)
     B=&FloatFromCurve(FloatCurve=DataInterface() {propertyValues {Curve {keys [
          <at 0 0 RCIM_Cubic>
          <at 0.7 1 RCIM_Cubic>
          {arriveTangent=0.5326, interpMode=RCIM_Cubic, leaveTangent=0.5326, tangentMode=RCTM_User, time=1, value=0}
         ]}}})
    )
    Scale RGB=&Multiply_VectorByFloat(
     Vector=&Multiply_VectorByFloat(Vector=<1 1 1>, Float=&FN_FrontEndFloatModifier(Gameplay=2, Front End=1))
     Float=&FNTimeOfDayScalar(Night Scale=0.3, Morning Scale=0.3, Evening Scale=0.4)
    )
    ScaleA=true
    ScaleRGB=false
   )
   &DynamicMaterialParameters(
    Index 0 Param 2=&FloatFromCurve(
     FloatCurve=DataInterface() {propertyValues {Curve {keys [{tangentMode=RCTM_User, time=0, value=0}, {tangentMode=RCTM_User, time=1, value=0.5}]}}}
     Scale Curve=&FloatFromCurve(FloatCurve=DataInterface() {propertyValues {Curve {keys [<at 0 1.7 RCIM_Cubic>,<at 1 1.7 RCIM_Cubic>]}}})
    )
    Module.Param0WriteEnabled=true
    Module.Param1WriteEnabled=true
    Module.Param2WriteEnabled=true
    Module.Param3WriteEnabled=true
   )
   &ScaleColor(
    Color Value To Scale=$StackContext.Color
    Scale Alpha=&Lerp_Float(A=0.5, B=1, Alpha=$System.KillReactBlend)
    ScaleA=true
    ScaleRGB=false
   ) {module=ScaleColor001}
   &InterpolateOverTime(
    Rate Of Change:NiagaraFloat=&RandomRangeFloat(Minimum=5, Maximum=15)
    Target Vector Value:Vector3f=&ConvertPositionToVector(Input Position=$Particles.Position, Convert To Absolute Large World Vector=false)
    Variable Type=::NewEnumerator2
   )
   /SetVariables(Particles.SpriteDynParams:Vector4f=&MakeVector4(
     X=0
     Y=&FloatFromCurve(
      FloatCurve=DataInterface() {propertyValues {Curve {keys [{tangentMode=RCTM_User, time=0, value=0}, {tangentMode=RCTM_User, time=1, value=0.5}]}}}
      Scale Curve=&FloatFromCurve(FloatCurve=DataInterface() {propertyValues {Curve {keys [<at 0 1.7 RCIM_Cubic>,<at 1 1.7 RCIM_Cubic>]}}})
     )
     Z=0
     W=0
    )) {isSetParameters=true}
   /SetVariables(Particles.Interpolate Dynamic=&MakeFloat4FromVectorAndFloat("Vector (RGB)"=$StackContext.InterpolateOverTime.InterpolateOverTime_V3_.Moving Average, "Float (Alpha)"=1)) {isSetParameters=true}
   /SetVariables(Particles.InterpolatedPosition:NiagaraPosition=$Particles.InterpolateOverTime.InterpolateOverTime_V3_.Moving Average) {isSetParameters=true}
   &DynamicMaterialParameters(
    Index 0 Param 1=&RandomRangeFloat(Minimum=0.9, Maximum=0.75)
    Index 0 Param 2=&RandomRangeFloat(Minimum=0.58, Maximum=1)
    Index 0 Param 3=50
    Index 0 Param 4=1
    Index 1 Param 1=&MakeFloatFromVector4(Vector 4=$Particles.Interpolate Dynamic, Channel=::NewEnumerator0)
    Index 1 Param 2=&MakeFloatFromVector4(Vector 4=$Particles.Interpolate Dynamic, Channel=::NewEnumerator1)
    Index 1 Param 3=&Lerp_Float(
     A=&MakeFloatFromVector(Vector=Linked() {linkedVariable {name=Particles.Position, type=NiagaraPosition}}, Channel=::NewEnumerator2)
     B=&MakeFloatFromVector4(Vector 4=$Particles.Interpolate Dynamic, Channel=::NewEnumerator2)
     Alpha=&MakeCustomFloatFromBool(Bool=$User.IsSkydiving, True Float=0.1, False Float=0.35)
    )
    Index 1 Param 4=1.5
    Index 2 Param 1=&Lerp_Float(A=6, B=8, Alpha=$System.KillReactBlend)
    Index 2 Param 2=&Lerp_Float(A=0.2, B=0.45, Alpha=$System.KillReactBlend)
    Module.Param0WriteEnabled=true
    Module.Param0WriteEnabled1=true
    Module.Param0WriteEnabled2=true
    Module.Param1WriteEnabled=true
    Module.Param1WriteEnabled1=true
    Module.Param1WriteEnabled2=true
    Module.Param2WriteEnabled=true
    Module.Param2WriteEnabled1=true
    Module.Param2WriteEnabled2=true
    Module.Param3WriteEnabled=true
    Module.Param3WriteEnabled1=true
    Module.Param3WriteEnabled2=true
    Write Parameter Index 1=true
    Write Parameter Index 2=true
   ) {module=DynamicMaterialParameters001}
   /SetVariables(Particles.SmokeStripColor:LinearColor=&MakeLinearColorFromVectorAndFloat(
     "Vector (RGB)"=<1 1 1>
     "Float (Alpha)"=&FloatFromCurve(
      FloatCurve=DataInterface() {propertyValues {Curve {keys [
          {interpMode=RCIM_Cubic, tangentMode=RCTM_User, time=0, value=0}
          {interpMode=RCIM_Cubic, tangentMode=RCTM_User, time=0.5, value=1}
          {tangentMode=RCTM_User, time=1, value=0}
         ]}}}
      Scale Curve=0.1
     )
    )) {isSetParameters=true}
   &SpriteFacingAndAlignment(Sprite Alignment=<0 0 1>, WriteAlignment=true, WriteFacing=true)
  ]
  rendererClasses [&NiagaraSpriteRendererProperties, &NiagaraMeshRendererProperties]
  renderers [
   {rendererClass=&NiagaraSpriteRendererProperties}
   {rendererClass=&NiagaraMeshRendererProperties, rendererIndex=1}
   {rendererClass=&NiagaraSpriteRendererProperties, rendererIndex=2}
  ]
 }
 BurningPieces {
  emitterUpdateScript [
   &EmitterState(
    Life Cycle Mode=::NewEnumerator1
    Loop Behavior=::NewEnumerator1
    Loop Delay=0.1
    Loop Duration=&FN_FrontEndFloatModifier(Gameplay=1, Front End=2.25)
    UseLoopDelay=false
   )
   &SpawnRate(
    Spawn Probability=&MakeCustomFloatFromBool(Bool=$User.bIsFrontEnd, True Float=0, False Float=0.7)
    SpawnRate=&FloatFromCurve(
     FloatCurve=DataInterface() {propertyValues {Curve {keys [
         {interpMode=RCIM_Cubic, tangentMode=RCTM_User, time=0, value=1}
         {arriveTangent=-2.2679, interpMode=RCIM_Cubic, leaveTangent=-2.2679, tangentMode=RCTM_User, time=1, value=0}
        ]}}}
     CurveIndex=$Emitter.NormalizedLoopAge
     Scale Curve=&FN_FrontEndFloatModifier(
      Gameplay=&Multiply_Float(A=15, B=&Lerp_Float(A=0.4, B=0.6, Alpha=$System.Heart Burn Clamped))
      Front End=30
     )
    )
    Use Spawn Probability=false
   )
   &SpawnBurst_Instantaneous(
    Spawn Count=&RandomRangeInt(Minimum=2, Maximum=4)
    Spawn Probability=&MakeCustomFloatFromBool(Bool=$User.bIsFrontEnd, True Float=0, False Float=1)
    Use Spawn Probability=false
    UseLoopCountLimit=false
   )
  ]
  particleSpawnScript [
   &FN_SetSpawnGroup(Number of Spawn Groups=2)
   &InitializeParticle(
    Color=<rgba 3 0 2.4624 1>
    Lifetime=&RandomRangeFloat(Minimum=0.25, Maximum=1)
    Mesh Renderer Array Visibility Mode=::NewEnumerator3
    Mesh Renderer Info=DataInterface() {propertyValues {MeshRenderer=&self:BurningPieces_5.NiagaraMeshRendererProperties_0}}
    Mesh Scale Mode=::NewEnumerator3
    Mesh Uniform Scale=&Multiply_Float(
     A=&RandomRangeFloat(Minimum=0.05, Maximum=0.1)
     B=&Lerp_Float(A=0.25, B=0.4, Alpha=$System.Heart Burn Clamped)
    )
    UsePositionOffset=false
    Write Material Random=true
   )
   /SetVariables(
    Particles.Radius:NiagaraFloat=&Multiply_Float(A=12, B=$System.Heart Burn Clamped)
    Particles.Theta:NiagaraFloat=&RandomRangeFloat(Minimum=0, Maximum=1)
   ) {isSetParameters=true}
   &SkeletalMeshLocation(
    Bone Sampling Mode=::NewEnumerator4
    MaskSpawnGroup=false
    Sampled Position Offset=&MakeVector(
     X=&Add_Float(
      A=-5
      B=&Multiply_Float(A=$Particles.Radius, B=&Cosine(Normalized Angle=$Particles.Theta, Period=1, Scale=1, Bias=0))
     )
     Y=&MaskFloatBySpawnGroup(Masked Float=-10, Passthrough Float=10)
     Z=&Add_Float(
      A=0
      B=&Multiply_Float(A=$Particles.Radius, B=&Sine(Normalized Angle=$Particles.Theta, Period=1, Scale=1, Bias=0))
     )
    )
    Skeletal Mesh=$User.SkeletalMesh
   )
   &ShapeLocation(
    Box Size=&MakeVector(X=5, Y=75, Z=30)
    MaskSpawnGroup=false
    Override Seed=false
    Shape Primitive=::NewEnumerator2
    Surface Only=false
    UseOwnerScale=true
   ) {enabled=false}
   &AddVelocity(
    Cone Angle=120
    Cone Axis=&RotateVectorByQuaternion(
     Vector To Rotate=&SelectVectorFromArray(Direct Array Index=$Particles.SpawnGroup, Vector Selection Array=DataInterface() {propertyValues {FloatData [<1 -1 0>, <1 1 0>]}})
     Quaternion=$Particles.SkeletalMeshLocation.SampledRotation
    )
    Module.UseConeFalloff=false
    Module.UseVelDistribution=false
    Override Seed=false
    Rotation Coordinate Space=::World
    Velocity Mode=::NewEnumerator2
    Velocity Speed=&Multiply_Float(A=&RandomRangeFloat(Minimum=25, Maximum=100), B=1)
   )
   &AddVelocity(
    Rotation Coordinate Space=::World
    Velocity=&RotateVectorByQuaternion(
     Vector To Rotate=&SelectVectorFromArray(
      Direct Array Index=&SpawnGroup()
      Vector Selection Array=DataInterface() {propertyValues {FloatData [<-1 0 0>, <1 0 0>]}}
     )
     Quaternion=$Particles.SkeletalMeshLocation.SampledRotation
    )
    Velocity Speed Scale=&Multiply_Float(A=&RandomRangeFloat(Minimum=500, Maximum=600), B=1)
   ) {enabled=false, module=AddVelocity001}
   &InitialMeshOrientation(Use Rotation Vector=false)
  ]
  particleUpdateScript [
   &ParticleState()
   &SkeletalMeshLocation(
    Bone Sampling Mode=::NewEnumerator4
    Initialize Particle Position To Sampled Position=false
    MaskSpawnGroup=false
    Position Sampling=::NewEnumerator1
    Skeletal Mesh=$User.SkeletalMesh
    Source Movement Scale:Vector3f=&VectorFromFloat(Value=1)
    Write To Particle Position=false
   ) {enabled=false, module=SkeletalMeshLocation001}
   &ScaleMeshSize(Scale Factor=&VectorFromFloat(Value=&FloatFromCurve(FloatCurve=DataInterface() {propertyValues {Curve {keys [
          {arriveTangent=4.1785, interpMode=RCIM_Cubic, leaveTangent=8.1351, tangentMode=RCTM_Break, time=0, value=0}
          {arriveTangent=0.0224, interpMode=RCIM_Cubic, leaveTangent=0.0224, tangentMode=RCTM_Break, time=0.05, value=1}
          {interpMode=RCIM_Cubic, tangentMode=RCTM_Break, time=0.7, value=1}
          {interpMode=RCIM_Cubic, tangentMode=RCTM_Break, time=1, value=0}
         ]}}})))
   &UpdateMeshOrientation(
    Rotation Rate=&RandomRangeFloat(Minimum=-1, Maximum=1)
    Rotation Vector=&RandomVector(Vector Scale=1)
   )
   &GravityForce(Gravity=<0 0 10>)
   &GravityForce(Gravity=&RotateVectorByQuaternion(
     Vector To Rotate=&Multiply_VectorByFloat(
      Vector=<0 1 0>
      Float=&FloatFromCurve(
       FloatCurve=DataInterface() {propertyValues {Curve {keys [
           {arriveTangent=0.0255, interpMode=RCIM_Cubic, leaveTangent=0.0255, tangentMode=RCTM_User, time=0, value=0}
           {arriveTangent=0.0112, interpMode=RCIM_Cubic, leaveTangent=0.0112, tangentMode=RCTM_User, time=0.1, value=1.6}
           {arriveTangent=-0.0134, interpMode=RCIM_Cubic, leaveTangent=-0.0134, tangentMode=RCTM_User, time=1, value=0}
          ]}}}
       Scale Curve=10000
      )
     )
     Quaternion=$Particles.SkeletalMeshLocation.SampledRotation
    )) {enabled=false, module=GravityForce001}
   &AerodynamicDrag(
    Aerodynamic Drag=3
    Pivot Offset Maximum:Vector3f=<10 10 10>
    Pivot Offset Minimum:Vector3f=<-10 -10 -10>
    Pivot Offset Minimum Length:Vector3f=<1 1 1>
    Use Linear Drag=true
    Use Rotational Drag=true
   )
   &InheritSourceMovement(
    Applied Position Scale=&VectorFromFloat(Value=&FloatFromCurve(FloatCurve=DataInterface() {propertyValues {Curve {keys [
          {arriveTangent=-0.195, interpMode=RCIM_Cubic, leaveTangent=-0.195, tangentMode=RCTM_User, time=0, value=1}
          {arriveTangent=-2.1928, interpMode=RCIM_Cubic, leaveTangent=-2.1928, tangentMode=RCTM_User, time=1, value=0}
         ]}}}))
    ApplyForce=false
    ApplyPosition=true
    ApplyPositionFalloff=false
   )
   &CurlNoiseForce(Noise Frequency=25, Noise Strength=200, Pan Noise=false)
   &Drag(Drag=5, Use Linear Drag=true, Use Rotational Drag=true)
   &SolveForcesAndVelocity(Clamp Velocity=false, Limit Acceleration=false)
   &AlignSpriteToMeshOrientation(SetAlignment=true, SetFacing=true)
   &DynamicMaterialParameters(
    Index 0 Param 1=&FloatFromCurve(FloatCurve=DataInterface() {propertyValues {Curve {keys [
         {arriveTangent=0.1337, interpMode=RCIM_Cubic, leaveTangent=0.1337, tangentMode=RCTM_User, time=0, value=0}
         {arriveTangent=4.4695, interpMode=RCIM_Cubic, leaveTangent=4.4695, tangentMode=RCTM_User, time=1, value=2}
        ]}}})
    Index 0 Param 2=&FloatFromCurve(
     FloatCurve=DataInterface() {propertyValues {Curve {keys [
         {interpMode=RCIM_Cubic, tangentMode=RCTM_User, time=0, value=0.5}
         {interpMode=RCIM_Cubic, tangentMode=RCTM_User, time=0.3, value=1}
         {tangentMode=RCTM_User, time=1, value=0.5}
        ]}}}
     Scale Curve=0.3
    )
    Module.Param0WriteEnabled=true
    Module.Param1WriteEnabled=true
    Module.Param2WriteEnabled=true
    Module.Param3WriteEnabled=true
   )
  ]
  rendererClasses [&NiagaraMeshRendererProperties]
  renderers [{rendererClass=&NiagaraMeshRendererProperties, rendererIndex=0}]
 }
 Embers {
  emitterUpdateScript [
   &EmitterState(
    Life Cycle Mode=::NewEnumerator1
    Loop Behavior=::NewEnumerator1
    Loop Duration=&FN_FrontEndFloatModifier(Gameplay=1.5, Front End=2.6)
    UseLoopDelay=false
   )
   &SpawnRate(
    Spawn Probability=&FloatFromCurve(
     FloatCurve=DataInterface() {propertyValues {Curve {keys [
         {arriveTangent=0.0025, interpMode=RCIM_Cubic, leaveTangent=0.0025, tangentMode=RCTM_User, time=0, value=0.7}
         {arriveTangent=0.0068, interpMode=RCIM_Cubic, leaveTangent=0.0068, tangentMode=RCTM_User, time=1, value=0.1}
        ]}}}
     CurveIndex=$Emitter.NormalizedLoopAge
     Scale Curve=&FN_FrontEndFloatModifier(Gameplay=&Lerp_Float(A=0.25, B=1, Alpha=$System.Heart Burn Clamped), Front End=1)
    )
    SpawnRate=125
    Use Spawn Probability=true
   )
  ]
  particleSpawnScript [
   &FN_SetSpawnGroup(Number of Spawn Groups=2)
   /SetVariables(Particles.PerParticleRandom:NiagaraFloat=&RandomRangeFloat(Minimum=0, Maximum=1)) {isSetParameters=true}
   &InitializeParticle(
    Color Mode=::NewEnumerator0
    Lifetime=&Lerp_Float(A=&Lerp_Float(A=0.3, B=0.75, Alpha=$System.Heart Burn Clamped), B=0.25, Alpha=$Particles.PerParticleRandom)
    Sprite Size=&Multiply_Vector2D(
     A=<0.2 0.5>
     B=&MakeVector2D(
      X=&FloatFromCurve(
       FloatCurve=DataInterface() {propertyValues {Curve {keys [<at 0 1>,<at 0.8 2 RCIM_Constant>,<at 0.95 3 RCIM_Constant>]}}}
       CurveIndex=$Particles.PerParticleRandom
      )
      Y=&FloatFromCurve(
       FloatCurve=DataInterface() {propertyValues {Curve {keys [<at 0 1>,<at 0.8 2 RCIM_Constant>,<at 0.95 3>]}}}
       CurveIndex=$Particles.PerParticleRandom
      )
     )
    )
    Sprite Size Mode=::NewEnumerator4
    UsePositionOffset=false
    Write Material Random=true
   )
   /SetVariables(
    Particles.Radius:NiagaraFloat=&Multiply_Float(A=12, B=$System.Heart Burn Clamped)
    Particles.Theta:NiagaraFloat=&RandomRangeFloat(Minimum=0, Maximum=1)
   ) {isSetParameters=true}
   &SkeletalMeshLocation(
    Bone Sampling Mode=::NewEnumerator4
    MaskSpawnGroup=false
    Sampled Position Offset=&MakeVector(
     X=&Add_Float(
      A=-5
      B=&Multiply_Float(A=$Particles.Radius, B=&Cosine(Normalized Angle=$Particles.Theta, Period=1, Scale=1, Bias=0))
     )
     Y=&MaskFloatBySpawnGroup(Masked Float=-10, Passthrough Float=10)
     Z=&Add_Float(
      A=0
      B=&Multiply_Float(A=$Particles.Radius, B=&Sine(Normalized Angle=$Particles.Theta, Period=1, Scale=1, Bias=0))
     )
    )
    Skeletal Mesh=$User.SkeletalMesh
   )
   &AddVelocity(
    Cone Angle=120
    Cone Axis=<0 -0.5 1>
    Module.UseConeFalloff=false
    Module.UseVelDistribution=false
    Override Seed=false
    Velocity Mode=::NewEnumerator2
    Velocity Speed=&FloatFromCurve(
     FloatCurve=DataInterface() {propertyValues {Curve {keys [
         {arriveTangent=-0.0405, interpMode=RCIM_Cubic, leaveTangent=-0.0405, tangentMode=RCTM_User, time=0, value=1}
         {arriveTangent=-2.7201, interpMode=RCIM_Cubic, leaveTangent=-2.7201, tangentMode=RCTM_User, time=1, value=0}
        ]}}}
     CurveIndex=$Emitter.NormalizedLoopAge
     Scale Curve=&Lerp_Float(A=5, B=50, Alpha=$Particles.PerParticleRandom)
    )
   )
  ]
  particleUpdateScript [
   &ParticleState()
   &Drag(
    Drag=&FloatFromCurve(
     FloatCurve=DataInterface() {propertyValues {Curve {keys [
         {interpMode=RCIM_Cubic, tangentMode=RCTM_User, time=0, value=0}
         {arriveTangent=2.3593, interpMode=RCIM_Cubic, leaveTangent=2.3593, tangentMode=RCTM_User, time=1, value=1}
        ]}}}
     Scale Curve=&RandomRangeFloat(Minimum=80, Maximum=150)
    )
    Use Linear Drag=true
    Use Rotational Drag=false
   )
   &CurlNoiseForce(
    Noise Frequency=200
    Noise Strength=&RandomRangeFloat(Minimum=250, Maximum=600)
    Pan Noise=false
    Random Seed=33
   ) {module=CurlNoiseForce001}
   &GravityForce(Gravity=&MakeVector(
     X=0
     Y=0
     Z=&FloatFromCurve(
      FloatCurve=DataInterface() {propertyValues {Curve {keys [
          {tangentMode=RCTM_User, time=0, value=1}
          {arriveTangent=-0.0405, interpMode=RCIM_Cubic, leaveTangent=-0.0405, tangentMode=RCTM_User, time=0.7009, value=1}
          {arriveTangent=-8.0089, interpMode=RCIM_Cubic, leaveTangent=-8.0089, tangentMode=RCTM_User, time=1, value=0}
         ]}}}
      CurveIndex=$Emitter.NormalizedLoopAge
      Scale Curve=&RandomRangeFloat(Minimum=200, Maximum=500)
     )
    ))
   &SolveForcesAndVelocity(Clamp Velocity=false, Limit Acceleration=false)
   &ScaleColor(
    Scale RGB=&MakeVectorFromLinearColorRGB(Color=&ColorFromCurve(ColorCurve=DataInterface() {propertyValues {
        AlphaCurve {keys [<at 0 1 RCIM_Cubic>]}
        BlueCurve {keys [<at 0.0048 5>,<at 0.4666 3>,<at 1 0.5469 RCIM_Cubic>]}
        GreenCurve {keys [<at 0.0048 1.38>,<at 0.4666 0>,<at 1 0 RCIM_Cubic>]}
        RedCurve {keys [<at 0.0048 3.1297>,<at 0.4666 2.2667>,<at 1 0.1623 RCIM_Cubic>]}
       }}))
    ScaleA=true
    ScaleRGB=true
   )
   &ScaleSpriteSize(
    Uniform Curve Scale=&RandomRangeFloat(Minimum=1, Maximum=2)
    Uniform Curve Sprite Scale=DataInterface() {propertyValues {Curve {keys [
        {tangentMode=RCTM_User, time=0, value=1}
        {arriveTangent=-0.0405, interpMode=RCIM_Cubic, leaveTangent=-0.0405, tangentMode=RCTM_User, time=0.7009, value=1}
        {arriveTangent=-8.0089, interpMode=RCIM_Cubic, leaveTangent=-8.0089, tangentMode=RCTM_User, time=1, value=0}
       ]}}}
   )
   &ScaleSpriteSizeBySpeed(
    Max Scale Factor=<1.666 5>
    Min Scale Factor=<1.333 1.333>
    Sample Scale Factor By Curve=true
    Scale Factor Curve=DataInterface() {propertyValues {Curve {keys [<at 0 0>, {arriveTangent=1.4989, interpMode=RCIM_Cubic, tangentMode=RCTM_Break, time=1, value=1}]}}}
    Velocity Threshold=250
   )
  ]
  rendererClasses [&NiagaraSpriteRendererProperties]
  renderers [{rendererClass=&NiagaraSpriteRendererProperties}]
 }
 FlameSplashRing {
  emitterUpdateScript [
   &EmitterState(
    Life Cycle Mode=::NewEnumerator1
    Loop Behavior=::NewEnumerator1
    Loop Duration=&FN_FrontEndFloatModifier(Gameplay=0.5, Front End=1.25)
    UseLoopDelay=false
   )
   &SpawnBurst_Instantaneous(Spawn Count=2, Use Spawn Probability=false, UseLoopCountLimit=false)
   &SpawnRate(
    Spawn Probability=&FloatFromCurve(
     FloatCurve=DataInterface() {propertyValues {Curve {keys [
         {arriveTangent=0.0025, interpMode=RCIM_Cubic, leaveTangent=0.0025, tangentMode=RCTM_User, time=0, value=0.7}
         {arriveTangent=-0.5452, interpMode=RCIM_Cubic, leaveTangent=-0.5452, tangentMode=RCTM_User, time=1, value=0.2}
        ]}}}
     CurveIndex=$Emitter.NormalizedLoopAge
     Scale Curve=&FN_FrontEndFloatModifier(Gameplay=&Lerp_Float(A=0.25, B=1, Alpha=$System.Heart Burn Clamped), Front End=1)
    )
    SpawnRate=&FN_FrontEndFloatModifier(Gameplay=0, Front End=15)
    Use Spawn Probability=true
   )
  ]
  particleSpawnScript [
   &FN_SetSpawnGroup(Number of Spawn Groups=2)
   &InitializeParticle(
    Color=&MakeLinearColorFromVectorAndFloat(
     "Vector (RGB)"=&VectorFromFloat(Value=&FNTimeOfDayScalar(Night Scale=0.5, Morning Scale=1, Evening Scale=1))
     "Float (Alpha)"=1
    )
    Lifetime=&FN_FrontEndFloatModifier(
     Gameplay=&RandomRangeFloat(Minimum=1, Maximum=2)
     Front End=&RandomRangeFloat(Minimum=1.5, Maximum=1.75)
    )
    Sprite Rotation Angle=&RandomRangeFloat(Minimum=-0.05, Maximum=0.05)
    Sprite Rotation Mode=::NewEnumerator1
    Sprite Size=<40 50>
    Sprite Size Mode=::NewEnumerator4
    Sprite UV Mode=::NewEnumerator1
    UsePositionOffset=false
    Write Material Random=true
   )
   /SetVariables(
    Particles.Radius:NiagaraFloat=&Multiply_Float(A=12, B=$System.Heart Burn Clamped)
    Particles.Theta:NiagaraFloat=&RandomRangeFloat(Minimum=0, Maximum=1)
   ) {enabled=false, isSetParameters=true, module=SetVariables_#3}
   &SkeletalMeshLocation(
    Bone Sampling Mode=::NewEnumerator4
    MaskSpawnGroup=false
    Sampled Position Offset=&MakeVector(
     X=&MaskFloatBySpawnGroup(Masked Float=0, Passthrough Float=0)
     Y=&MaskFloatBySpawnGroup(Masked Float=-12, Passthrough Float=12)
     Z=&MaskFloatBySpawnGroup(Masked Float=0, Passthrough Float=0)
    )
    Skeletal Mesh=$User.SkeletalMesh
   ) {module=SkeletalMeshLocation001}
   &ShapeLocation(MaskSpawnGroup=false, Override Seed=false, Sphere Radius=1, UseOwnerScale=true) {enabled=false}
   &AddVelocity(
    Module.Use Falloff Distance=false
    Origin Offset=<0 0 -2>
    Override Seed=false
    Velocity Mode=::NewEnumerator1
    Velocity Speed=&RandomRangeFloat(Minimum=20, Maximum=25)
   ) {enabled=false}
   &CameraOffset(Camera Offset Amount=0.5)
  ]
  particleUpdateScript [
   &ParticleState()
   &SkeletalMeshLocation(
    Bone Sampling Mode=::NewEnumerator4
    MaskSpawnGroup=false
    Sampled Orientation Coordinate Space=::World
    Sampled Position Offset=&MakeVector(
     X=&MaskFloatBySpawnGroup(Masked Float=-2.5, Passthrough Float=-5)
     Y=&MaskFloatBySpawnGroup(Masked Float=-11, Passthrough Float=10)
     Z=&MaskFloatBySpawnGroup(Masked Float=2, Passthrough Float=0)
    )
    Skeletal Mesh=$User.SkeletalMesh
   )
   &ScaleSpriteSize(
    Non-Uniform Curve Sprite Scale=DataInterface() {propertyValues {
      XCurve {keys [
        {arriveTangent=9.0725, interpMode=RCIM_Cubic, leaveTangent=9.0725, tangentMode=RCTM_User, time=0, value=0}
        {arriveTangent=-2.5249, interpMode=RCIM_Cubic, leaveTangent=-2.5249, tangentMode=RCTM_User, time=0.1408, value=0.9847}
        {arriveTangent=-0.0405, interpMode=RCIM_Cubic, leaveTangent=-0.0405, tangentMode=RCTM_User, time=1, value=0}
       ]}
      YCurve {keys [
        {interpMode=RCIM_Cubic, tangentMode=RCTM_User, time=0, value=0}
        {interpMode=RCIM_Cubic, tangentMode=RCTM_User, time=0.5, value=1}
        {tangentMode=RCTM_User, time=1, value=0}
       ]}
     }}
    Scale Sprite Size Mode=::NewEnumerator3
   ) {enabled=false}
   &ScaleColor(
    Scale Alpha=&FloatFromCurve(FloatCurve=DataInterface() {propertyValues {Curve {keys [
         {arriveTangent=9.0725, interpMode=RCIM_Cubic, leaveTangent=9.0725, tangentMode=RCTM_User, time=0, value=0.5}
         {arriveTangent=-1.1456, interpMode=RCIM_Cubic, leaveTangent=-1.1456, tangentMode=RCTM_User, time=0.2, value=0.8}
         {arriveTangent=-0.7658, interpMode=RCIM_Cubic, leaveTangent=-0.7658, tangentMode=RCTM_User, time=1, value=0}
        ]}}})
    Scale RGB=&MakeVectorFromLinearColorRGB(Color=&ColorFromCurve(ColorCurve=DataInterface() {propertyValues {
        AlphaCurve {keys []}
        BlueCurve {keys [<at 0.169 5>,<at 0.4284 0.8648>,<at 0.9066 0>]}
        GreenCurve {keys [<at 0.169 0>,<at 0.4284 0>,<at 0.9066 0>]}
        RedCurve {keys [<at 0.169 3.9167>,<at 0.4284 1>,<at 0.9066 0>]}
       }}))
    ScaleA=true
    ScaleRGB=true
   )
   &Drag(Drag=&RandomRangeFloat(Minimum=8, Maximum=12), Use Linear Drag=true, Use Rotational Drag=true) {enabled=false}
   &CurlNoiseForce(Noise Strength=25, Pan Noise=false) {enabled=false}
   &GravityForce(Gravity=<0 0 100>) {enabled=false}
   &SolveForcesAndVelocity(Clamp Velocity=false, Limit Acceleration=false)
   &AlignSpriteToMeshOrientation(
    Mesh Orientation Relative Sprite Alignment Vector=<1 0 0>
    Mesh Orientation Relative Sprite Facing Vector:Vector3f=<0 -1 0>
    SetAlignment=true
    SetFacing=true
   )
   &DynamicMaterialParameters(
    Index 0 Param 1=&FloatFromCurve(
     FloatCurve=DataInterface() {propertyValues {Curve {keys [
         {arriveTangent=-0.015, interpMode=RCIM_Cubic, leaveTangent=-0.015, tangentMode=RCTM_User, time=0, value=0.4}
         {arriveTangent=0.4707, interpMode=RCIM_Cubic, leaveTangent=0.4707, tangentMode=RCTM_User, time=1, value=0.8}
        ]}}}
     Scale Curve=5
    )
    Index 0 Param 2=&FloatFromCurve(
     FloatCurve=DataInterface() {propertyValues {Curve {keys [
         {arriveTangent=1.1466, interpMode=RCIM_Cubic, leaveTangent=1.1466, tangentMode=RCTM_User, time=0, value=0.2}
         {arriveTangent=0.3228, interpMode=RCIM_Cubic, leaveTangent=0.3228, tangentMode=RCTM_User, time=1, value=1}
        ]}}}
     CurveIndex=$System.Heart Burn Clamped
     Scale Curve=0.27
    )
    Index 0 Param 3=&FloatFromCurve(
     FloatCurve=DataInterface() {propertyValues {Curve {keys [
         {interpMode=RCIM_Cubic, tangentMode=RCTM_User, time=0, value=0.25}
         {arriveTangent=1.7963, interpMode=RCIM_Cubic, leaveTangent=1.7963, tangentMode=RCTM_User, time=1, value=1}
        ]}}}
     Scale Curve=0.35
    )
    Index 0 Param 4=&FloatFromCurve(
     FloatCurve=DataInterface() {propertyValues {Curve {keys [
         {arriveTangent=9.0725, interpMode=RCIM_Cubic, leaveTangent=9.0725, tangentMode=RCTM_User, time=0, value=0.25}
         {arriveTangent=-2.5249, interpMode=RCIM_Cubic, leaveTangent=-2.5249, tangentMode=RCTM_User, time=0.2, value=0.875}
         {arriveTangent=-0.0405, interpMode=RCIM_Cubic, leaveTangent=-0.0405, tangentMode=RCTM_User, time=1, value=0.25}
        ]}}}
     Scale Curve=0.2
    )
    Module.Param0WriteEnabled=true
    Module.Param1WriteEnabled=true
    Module.Param2WriteEnabled=true
    Module.Param3WriteEnabled=true
   )
  ]
  rendererClasses [&NiagaraSpriteRendererProperties]
  renderers [{rendererClass=&NiagaraSpriteRendererProperties}]
 }
 FlameSplashSprite {
  emitterUpdateScript [
   &EmitterState(
    Life Cycle Mode=::NewEnumerator1
    Loop Behavior=::NewEnumerator1
    Loop Duration=&FN_FrontEndFloatModifier(Gameplay=0.3, Front End=1.5)
    UseLoopDelay=false
   )
   &SpawnBurst_Instantaneous(
    Spawn Count=&RandomRangeInt(Minimum=3, Maximum=6)
    Use Spawn Probability=false
    UseLoopCountLimit=false
   )
   &SpawnRate(
    Spawn Probability=&FloatFromCurve(
     FloatCurve=DataInterface() {propertyValues {Curve {keys [
         {arriveTangent=0.0025, interpMode=RCIM_Cubic, leaveTangent=0.0025, tangentMode=RCTM_User, time=0, value=0.7}
         {arriveTangent=-1.1662, interpMode=RCIM_Cubic, leaveTangent=-1.1662, tangentMode=RCTM_User, time=1, value=0.1}
        ]}}}
     CurveIndex=$Emitter.NormalizedLoopAge
     Scale Curve=&FN_FrontEndFloatModifier(Gameplay=&Lerp_Float(A=0.25, B=1, Alpha=$System.Heart Burn Clamped), Front End=1)
    )
    SpawnRate=25
    Use Spawn Probability=true
   )
  ]
  particleSpawnScript [
   &FN_SetSpawnGroup(Number of Spawn Groups=2)
   &InitializeParticle(
    Color=&MakeLinearColorFromVectorAndFloat(
     "Vector (RGB)"=&VectorFromFloat(Value=&FNTimeOfDayScalar(Night Scale=0.5, Morning Scale=1, Evening Scale=1))
     "Float (Alpha)"=1
    )
    Lifetime=&FN_FrontEndFloatModifier(
     Gameplay=&FloatFromCurve(
      FloatCurve=DataInterface() {propertyValues {Curve {keys [
          {interpMode=RCIM_Cubic, tangentMode=RCTM_User, time=0, value=1.5}
          {arriveTangent=-2.5332, interpMode=RCIM_Cubic, leaveTangent=-2.5332, tangentMode=RCTM_User, time=1, value=0.2}
         ]}}}
      CurveIndex=$Emitter.NormalizedLoopAge
      Scale Curve=&RandomRangeFloat(Minimum=0.6, Maximum=1.2)
     )
     Front End=&RandomRangeFloat(Minimum=1.5, Maximum=2)
    )
    Sprite Rotation Angle=&RandomRangeFloat(Minimum=-0.05, Maximum=0.05)
    Sprite Rotation Mode=::NewEnumerator1
    Sprite Size=&Multiply_Vector2D_ByFloat(
     Float=&Lerp_Float(A=0.5, B=0.75, Alpha=$System.Heart Burn Clamped)
     Vector2D=&RandomRangeVector2D(Minimum=<20 25>, Maximum=<35 40>)
    )
    Sprite Size Mode=::NewEnumerator4
    Sprite UV Mode=::NewEnumerator1
    UsePositionOffset=false
    Write Material Random=true
   )
   /SetVariables(
    Particles.Radius:NiagaraFloat=&FloatFromCurve(
     FloatCurve=DataInterface() {propertyValues {Curve {keys [
         {arriveTangent=19.5482, interpMode=RCIM_Cubic, leaveTangent=19.5482, tangentMode=RCTM_User, time=0, value=0}
         {interpMode=RCIM_Cubic, tangentMode=RCTM_User, time=1, value=10}
        ]}}}
     CurveIndex=$System.Heart Burn Clamped
     Scale Curve=&MaskFloatBySpawnGroup(Masked Float=1, Passthrough Float=0.8)
    )
    Particles.Theta:NiagaraFloat=&RandomRangeFloat(Minimum=0, Maximum=1)
   ) {isSetParameters=true, module=SetVariables_#3}
   &SkeletalMeshLocation(
    Bone Sampling Mode=::NewEnumerator4
    MaskSpawnGroup=false
    Sampled Position Offset=&MakeVector(
     X=&Add_Float(
      A=0
      B=&Multiply_Float(
       A=&Add_Float(A=$Particles.Radius, B=2)
       B=&Sine(Normalized Angle=$Particles.Theta, Period=1, Scale=1, Bias=0)
      )
     )
     Y=&MaskFloatBySpawnGroup(Masked Float=-12, Passthrough Float=12)
     Z=&Add_Float(
      A=2
      B=&Multiply_Float(A=$Particles.Radius, B=&Cosine(Normalized Angle=$Particles.Theta, Period=1, Scale=1, Bias=0))
     )
    )
    Skeletal Mesh=$User.SkeletalMesh
   ) {module=SkeletalMeshLocation001}
   &ShapeLocation(MaskSpawnGroup=false, Override Seed=false, Sphere Radius=1, UseOwnerScale=true) {enabled=false}
   &AddVelocity(
    Module.Use Falloff Distance=false
    Origin Offset=<0 0 -2>
    Override Seed=false
    Velocity Mode=::NewEnumerator1
    Velocity Speed=&RandomRangeFloat(Minimum=20, Maximum=25)
   ) {enabled=false}
   &CameraOffset(Camera Offset Amount=3)
  ]
  particleUpdateScript [
   &ParticleState()
   &SkeletalMeshLocation(
    Bone Sampling Mode=::NewEnumerator4
    Initialize Particle Position To Sampled Position=false
    MaskSpawnGroup=false
    Position Sampling=::NewEnumerator1
    Skeletal Mesh=$User.SkeletalMesh
    Write To Particle Position=false
   )
   &ScaleSpriteSize(
    Non-Uniform Curve Sprite Scale=DataInterface() {propertyValues {
      XCurve {keys [
        {arriveTangent=9.0725, interpMode=RCIM_Cubic, leaveTangent=9.0725, tangentMode=RCTM_User, time=0, value=0}
        {arriveTangent=-2.5249, interpMode=RCIM_Cubic, leaveTangent=-2.5249, tangentMode=RCTM_User, time=0.1408, value=0.9847}
        {arriveTangent=-0.0405, interpMode=RCIM_Cubic, leaveTangent=-0.0405, tangentMode=RCTM_User, time=1, value=0}
       ]}
      YCurve {keys [
        {interpMode=RCIM_Cubic, tangentMode=RCTM_User, time=0, value=0}
        {interpMode=RCIM_Cubic, tangentMode=RCTM_User, time=0.5, value=1}
        {tangentMode=RCTM_User, time=1, value=0}
       ]}
     }}
    Scale Sprite Size Mode=::NewEnumerator3
   ) {enabled=false}
   &ScaleColor(
    Scale Alpha=&FloatFromCurve(FloatCurve=DataInterface() {propertyValues {Curve {keys [
         {arriveTangent=9.0725, interpMode=RCIM_Cubic, leaveTangent=9.0725, tangentMode=RCTM_User, time=0, value=0.5}
         {arriveTangent=-2.5249, interpMode=RCIM_Cubic, leaveTangent=-2.5249, tangentMode=RCTM_User, time=0.1408, value=0.9847}
         {arriveTangent=-0.0405, interpMode=RCIM_Cubic, leaveTangent=-0.0405, tangentMode=RCTM_User, time=1, value=0}
        ]}}})
    Scale RGB=&MakeVectorFromLinearColorRGB(Color=&ColorFromCurve(ColorCurve=DataInterface() {propertyValues {
        AlphaCurve {keys []}
        BlueCurve {keys [<at 0.169 5>,<at 0.4284 0.8648>,<at 0.9066 0>]}
        GreenCurve {keys [<at 0.169 0>,<at 0.4284 0>,<at 0.9066 0>]}
        RedCurve {keys [<at 0.169 3.9167>,<at 0.4284 1>,<at 0.9066 0>]}
       }}))
    ScaleA=true
    ScaleRGB=true
   )
   &Drag(Drag=&RandomRangeFloat(Minimum=8, Maximum=12), Use Linear Drag=true, Use Rotational Drag=true)
   &CurlNoiseForce(Noise Strength=50, Pan Noise=false)
   &GravityForce(Gravity=&MakeVector(
     X=0
     Y=0
     Z=&Multiply_Float(
      A=&RandomRangeFloat(Minimum=50, Maximum=200)
      B=&Lerp_Float(A=0.2, B=1, Alpha=$System.Heart Burn Clamped)
     )
    ))
   &SolveForcesAndVelocity(Clamp Velocity=false, Limit Acceleration=false)
   &AlignSpriteToMeshOrientation(
    Mesh Orientation Relative Sprite Alignment Vector=<1 0 0>
    Mesh Orientation Relative Sprite Facing Vector:Vector3f=<0 -1 0>
    SetAlignment=true
    SetFacing=true
   ) {enabled=false}
   &DynamicMaterialParameters(
    Index 0 Param 1=&FloatFromCurve(
     FloatCurve=DataInterface() {propertyValues {Curve {keys [
         {arriveTangent=-0.015, interpMode=RCIM_Cubic, leaveTangent=-0.015, tangentMode=RCTM_User, time=0, value=0.4}
         {arriveTangent=0.4707, interpMode=RCIM_Cubic, leaveTangent=0.4707, tangentMode=RCTM_User, time=1, value=0.8}
        ]}}}
     Scale Curve=5
    )
    Index 0 Param 2=&FloatFromCurve(
     FloatCurve=DataInterface() {propertyValues {Curve {keys [
         {arriveTangent=1.1466, interpMode=RCIM_Cubic, leaveTangent=1.1466, tangentMode=RCTM_User, time=0, value=0.2}
         {arriveTangent=0.3228, interpMode=RCIM_Cubic, leaveTangent=0.3228, tangentMode=RCTM_User, time=1, value=1}
        ]}}}
     CurveIndex=$System.Heart Burn Clamped
     Scale Curve=0.5
    )
    Index 0 Param 3=&FloatFromCurve(
     FloatCurve=DataInterface() {propertyValues {Curve {keys [
         {interpMode=RCIM_Cubic, tangentMode=RCTM_User, time=0, value=0.25}
         {arriveTangent=1.7963, interpMode=RCIM_Cubic, leaveTangent=1.7963, tangentMode=RCTM_User, time=1, value=1}
        ]}}}
     Scale Curve=0.15
    )
    Index 0 Param 4=&FloatFromCurve(
     FloatCurve=DataInterface() {propertyValues {Curve {keys [
         {arriveTangent=9.0725, interpMode=RCIM_Cubic, leaveTangent=9.0725, tangentMode=RCTM_User, time=0, value=0.25}
         {arriveTangent=-2.5249, interpMode=RCIM_Cubic, leaveTangent=-2.5249, tangentMode=RCTM_User, time=0.2, value=0.875}
         {arriveTangent=-0.0405, interpMode=RCIM_Cubic, leaveTangent=-0.0405, tangentMode=RCTM_User, time=1, value=0.25}
        ]}}}
     Scale Curve=0.2
    )
    Module.Param0WriteEnabled=false
    Module.Param1WriteEnabled=true
    Module.Param2WriteEnabled=true
    Module.Param3WriteEnabled=true
   )
  ]
  rendererClasses [&NiagaraSpriteRendererProperties]
  renderers [{rendererClass=&NiagaraSpriteRendererProperties}]
 }
}
summary {
 systemName=NS_MED_Patches_RevealFX
 emitters [
  {emitterName=Body_Smoke, rendererClasses [&NiagaraSpriteRendererProperties, &NiagaraMeshRendererProperties]}
  {emitterName=FlameSplashSprite, rendererClasses [&NiagaraSpriteRendererProperties]}
  {emitterName=Embers, rendererClasses [&NiagaraSpriteRendererProperties]}
  {emitterName=FlameSplashRing, rendererClasses [&NiagaraSpriteRendererProperties]}
  {emitterName=BurningPieces, rendererClasses [&NiagaraMeshRendererProperties]}
 ]
 userVariables [
  User.bIsFrontEnd {type=NiagaraBool}
  User.bIsFrontEndPreview {type=NiagaraBool}
  User.bKillReaction {type=NiagaraBool}
  User.global Alpha {type=NiagaraFloat, defaultValue=0, description=""}
  User.HeartBurn {type=NiagaraFloat, defaultValue=1, description=""}
  User.ImpactColor {type=LinearColor, description="", defaultValue {struct=LinearColor, value=<rgba 1 1 1 1>}}
  User.IsSkydiving {type=NiagaraBool}
  User.IsWeakPoint {type=NiagaraBool}
  User.SkeletalMesh {
   type=NiagaraDataInterfaceSkeletalMesh
   defaultValue {
    struct=NiagaraExt_VariableValue_DataInterface
    value {dataInterface=&self:NiagaraDataInterfaceSkeletalMesh_0}
   }
  }
 ]
}
