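# Niagara idle chest effect for the Patches player character cosmetic, centred on a pulsing glow.
# Light_loop bursts one particle at a user position offset, drawn through light and sprite renderers.
# Its color multiplies the user BackPack_Color by a time-of-day scalar and a curve rising and falling to zero.
# Standard sphere and lightning emitters are disabled; LW_-prefixed counterparts exist, but show no modules or renderers here.
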NiagaraSystem /BRCosmetics/Characters/Player/Male/Medium/Bodies/M_MED_Patches/FX/NS_Patches_ChestIdle
dependencies=null
emitters {
 LW_CenterLightning {rendererClasses [], renderers []}
 LW_Lightning {rendererClasses [], renderers []}
 LW_Sphere_center {rendererClasses [], renderers []}
 Light_loop {
  emitterUpdateScript [
   &EmitterState()
   &SpawnBurst_Instantaneous(Spawn Count=1, Use Spawn Probability=false)
  ]
  particleSpawnScript [
   &InitializeParticle(Position Offset:NiagaraPosition=$User.Position Offset, UsePositionOffset=true, Write Material Random=true) {module=InitializeParticle001}
   /SetVariables(Particles.Light_fall_off=1.5) {isSetParameters=true}
   /SetVariables(Particles.LightRadius=15) {isSetParameters=true}
  ]
  particleUpdateScript [
   &ParticleState()
   &SolveForcesAndVelocity(Clamp Velocity=false, Limit Acceleration=false)
   &ScaleColor(
    Scale RGB=&Multiply_VectorByFloat(
     Vector=&Multiply_Vector(
      A=&MakeVectorFromLinearColorRGB(Color=$User.BackPack_Color)
      B=&Multiply_VectorByFloat(
       Vector=<10 10 10>
       Float=&FNTimeOfDayScalar(Night Scale=0.7, Morning Scale=0.3, Evening Scale=0.5)
      )
     )
     Float=&FloatFromCurve(FloatCurve=DataInterface() {propertyValues {Curve {keys [
          {arriveTangent=6.1807, interpMode=RCIM_Cubic, leaveTangent=6.1807, tangentMode=RCTM_User, time=0, value=0}
          {interpMode=RCIM_Cubic, tangentMode=RCTM_User, time=0.5, value=1}
          {arriveTangent=-5.937, interpMode=RCIM_Cubic, leaveTangent=-5.937, tangentMode=RCTM_User, time=1, value=0}
         ]}}})
    )
    ScaleA=true
    ScaleRGB=true
   )
  ]
  rendererClasses [&NiagaraLightRendererProperties, &NiagaraSpriteRendererProperties]
  renderers [
   {rendererClass=&NiagaraLightRendererProperties, rendererIndex=0}
   {rendererClass=&NiagaraSpriteRendererProperties, rendererIndex=1}
  ]
 }
}
summary {
 systemName=NS_Patches_ChestIdle
 emitters [
  {bEnabled=false, emitterName=Sphere_center, rendererClasses [&NiagaraMeshRendererProperties]}
  {bEnabled=false, emitterName=Lightning, rendererClasses [&NiagaraSpriteRendererProperties]}
  {emitterName=Light_loop, rendererClasses [&NiagaraLightRendererProperties, &NiagaraSpriteRendererProperties]}
  {bEnabled=false, emitterName=CenterLightning, rendererClasses [&NiagaraSpriteRendererProperties]}
  {emitterName=LW_CenterLightning, rendererClasses []}
  {emitterName=LW_Sphere_center, rendererClasses []}
  {emitterName=LW_Lightning, rendererClasses []}
 ]
 userVariables [
  User.BackPack_Color {type=LinearColor, description="", defaultValue {struct=LinearColor, value=<rgba 1 0 0.667 1>}}
  User.HoleRadius {type=NiagaraFloat, defaultValue=0.1, description=""}
  User.Position Offset {type=NiagaraPosition, description="", defaultValue {}}
 ]
}
