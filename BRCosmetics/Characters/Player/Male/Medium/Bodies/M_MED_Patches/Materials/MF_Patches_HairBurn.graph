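# Material function for the Patches character that adds a burnt-hair effect to incoming material attributes, faded in by "Hair burn Blend".
# The hair burn mask's green channel darkens BaseColor; its panning-noise UV distortion is multiplied by zero.
# Its blue channel, Fresnel and tiled noise form an edge glow added to Emissive, tinted pink or blue by static switch.
# Two panning noise layers vary the glow intensity between Min and Max parameters and beyond.
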MaterialFunction /BRCosmetics/Characters/Player/Male/Medium/Bodies/M_MED_Patches/Materials/MF_Patches_HairBurn
graph {nodes [
  n1 = GetMaterialAttributes(None=FunctionInput() {inputType=FunctionInput_MaterialAttributes}) {attributeGetTypes [#1, #2]}
  n2 = NamedRerouteUsage()
  n3 = NamedRerouteUsage()
  FunctionOutput(None=SetMaterialAttributes!(MaterialAttributes=@n1.MaterialAttributes, BaseColor=Multiply(A=@n2, B=@n1.BaseColor), EmissiveColor=Add(A=@n1.EmissiveColor, B=@n3)) {attributeSetTypes [#1, #2]})
  n4 = TextureCoordinate()
  n5 = TextureCoordinate()
  n6 = TextureSample(UVs=Panner(Coordinate=Multiply(A=@n5) {constB=15}) {speedX=-0.2, speedY=0.8}) {
   samplerSource=SSM_Wrap_WorldGroupSettings
   samplerType=SAMPLERTYPE_Grayscale
   texture=/Game/TimeOfDay/TODM_Sequence/Materials/Sky/Textures/LowResBlurredNoise_Radish_LF_Root
  }
  n7 = Constant()
  n8 = Multiply(A=QualitySwitch(Default=@n6.R, Low=@n7, High=@n6.R, Medium=@n7, Epic=@n6.R), B=Constant())
  n9 = Add(
   A=TextureSample(UVs=Add(A=@n4, B=@n8)) {texture=/BRCosmetics/Characters/Player/Male/Medium/Bodies/M_MED_Patches/Textures/T_HairBurn_Mask}.G
   B=Constant() {r=0.02}
  )
  n10 = Saturate(None=Power(Base=Saturate(None=@n9), Exp=Constant() {r=2}))
  n11 = Saturate(None=ScalarParameter() {group=Patches, parameterName=Hair burn Blend})
  NamedRerouteDeclaration(None=LinearInterpolate(B=LinearInterpolate(Alpha=@n10) {constA=0.01}, Alpha=@n11) {constA=1}) {name=Smoke, nodeColor=<rgba 0.9219 0 1 1>}
  n12 = Fresnel(ExponentIn=Constant() {r=2.6}, BaseReflectFractionIn=Constant() {r=-0.128})
  n13 = TextureSample() {texture=&T_HairBurn_Mask}
  n14 = Saturate(None=Multiply(A=@n13.B) {constB=2})
  n15 = Constant() {r=1.5}
  n16 = TextureSample(UVs=Multiply(A=TextureCoordinate(), B=ScalarParameter() {defaultValue=5, group=Patches, parameterName=Hair edge Emissive Tiling2})) {
   samplerSource=SSM_Wrap_WorldGroupSettings
   samplerType=SAMPLERTYPE_Grayscale
   texture=&LowResBlurredNoise_Radish_LF_Root
  }
  n17 = TextureSample(UVs=Multiply(A=TextureCoordinate(), B=ScalarParameter() {defaultValue=10, group=Patches, parameterName=hair edge Emissive Tiling})) {
   samplerSource=SSM_Wrap_WorldGroupSettings
   samplerType=SAMPLERTYPE_Grayscale
   texture=&LowResBlurredNoise_Radish_LF_Root
  }
  n18 = Multiply(A=@n13.B, B=Max(A=@n16.R, B=@n17.R))
  n19 = Constant() {r=1}
  n20 = Max(A=Add(A=Multiply(A=@n12, B=@n14), B=Power(Base=@n13.B, Exp=@n15)), B=Multiply(A=Power(Base=@n18, Exp=@n19)) {constB=0.2})
  n21 = Multiply(
   A=@n20
   B=StaticSwitchParameter(
    True=VectorParameter() {group=Patches, parameterName=hair PinkRed Edge, defaultValue=<rgba 1 0.1264 0.697 1>}.RGB
    False=VectorParameter() {group=Patches, parameterName=hairBlueedge, defaultValue=<rgba 0.4433 0 0.85 1>}.RGB
   ) {parameterName="blue or  pink Emissive edge"}
  )
  n22 = TextureCoordinate()
  n23 = Constant() {r=9.3168}
  n24 = TextureSample(UVs=Panner(Coordinate=Multiply(A=@n22, B=@n23)) {speedX=-0.1, speedY=0.1}) {
   samplerSource=SSM_Wrap_WorldGroupSettings
   samplerType=SAMPLERTYPE_Grayscale
   texture=&LowResBlurredNoise_Radish_LF_Root
  }
  n25 = TextureCoordinate()
  n26 = Constant() {r=10}
  n27 = TextureSample(UVs=Panner(Coordinate=Multiply(A=@n25, B=@n26)) {speedX=-0.2, speedY=0.2}) {
   samplerSource=SSM_Wrap_WorldGroupSettings
   samplerType=SAMPLERTYPE_Grayscale
   texture=&LowResBlurredNoise_Radish_LF_Root
  }
  n28 = Add(
   A=Power(Base=TextureSample() {texture=&T_HairBurn_Mask}.B)
   B=Max(A=Power(Base=@n24.R), B=@n27.R)
  )
  n29 = Multiply(A=@n28, B=Constant() {r=0.1562})
  n30 = Power(Base=@n20, Exp=LinearInterpolate(A=Constant() {r=1.167}, B=Constant() {r=0.773}, Alpha=@n28))
  n31 = LinearInterpolate(
   A=ScalarParameter() {defaultValue=0.5, group=Patches, parameterName=Hair edge Emissive Min}
   B=LinearInterpolate(A=ScalarParameter() {defaultValue=10, group=Patches, parameterName=Hair edge Emissive Max}, Alpha=Saturate(None=@n28)) {constB=50}
   Alpha=Saturate(None=Multiply(A=@n29, B=@n30))
  )
  NamedRerouteDeclaration(None=LinearInterpolate(B=Multiply(A=@n21, B=@n31), Alpha=@n11)) {name=Emissive color1, nodeColor=<rgba 0.9219 0 1 1>}
  ScalarParameter() {parameterName=Hair burn Blend}
 ]}
