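# Material function for the M_MED_Patches character body that burns holes into clothing, outputting base color, emissive, normal and opacity-style masks.
# BurnRadius sets animation strength for noise-distorted sphere masks in a flame UV texture, combined with hole and hair-burn mask textures.
# Panning noise forms smoke and blend masks, while threshold bands create sparky emissive edges with switchable blue or pink color.
# Quality switches simplify smoke on lower settings, and FlattenNormal modulates a burnt-cloth normal.
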MaterialFunction /BRCosmetics/Characters/Player/Male/Medium/Bodies/M_MED_Patches/Materials/MF_HoleMask
graph {nodes [
  n1 = Saturate(None=Subtract(A=NamedRerouteUsage(), B=Constant() {r=0.5}))
  n2 = Constant() {r=5}
  n3 = LinearInterpolate(
   A=NamedRerouteUsage()
   B=LinearInterpolate(B=NamedRerouteUsage(), Alpha=NamedRerouteUsage()) {constA=1}
   Alpha=Saturate(None=Multiply(A=@n1, B=@n2) {constB=9})
  )
  n4 = FunctionInput() {inputName=BaseColor}
  FunctionOutput(None=LinearInterpolate(A=Multiply(A=@n3, B=@n4), B=@n4, Alpha=Constant())) {outputName=base color}
  FunctionOutput(None=NamedRerouteUsage()) {outputName=EmissiveColor}
  n5 = TextureSample() {samplerSource=SSM_Wrap_WorldGroupSettings, texture=/BRCosmetics/Characters/Player/Male/Medium/Bodies/M_MED_Patches/Textures/T_hole_mask_all}
  n6 = NamedRerouteUsage()
  FunctionOutput(None=If(A=NamedRerouteUsage(), "A > B"=@n5.A, "A == B"=@n5.A, "A < B"=LinearInterpolate(A=@n6, Alpha=@n5.A)) {constB=1}) {outputName=Mask}
  n7 = TextureSample() {
   samplerSource=SSM_Wrap_WorldGroupSettings
   samplerType=SAMPLERTYPE_Normal
   texture=/BRCosmetics/Characters/Player/Male/Medium/Bodies/M_MED_Patches/Textures/T_BurnCloth_Normal
  }
  n8 = Constant() {r=0.1}
  n9 = Subtract(A=NamedRerouteUsage(), B=Constant() {r=0.7})
  n10 = Saturate(None=Multiply(A=Saturate(None=@n9), B=Constant() {r=5}) {constB=9})
  FunctionOutput(None=MaterialFunctionCall(Normal=@n7.RGB, Flatness=LinearInterpolate(B=@n8, Alpha=@n10) {constA=1}) {materialFunction=&FlattenNormal}.Result) {outputName=Normal}
  n11 = TextureSample() {samplerSource=SSM_Wrap_WorldGroupSettings, texture=&T_hole_mask_all}
  n12 = Power(Base=@n11.G, Exp=Constant() {r=0.8})
  FunctionOutput(None=LinearInterpolate(B=Saturate(None=@n12), Alpha=NamedRerouteUsage()) {constA=1}) {outputName=Mask_2}
  n13 = TextureCoordinate()
  n14 = TextureSample(UVs=Panner(Coordinate=Multiply(A=@n13) {constB=24}) {speedX=-0.2, speedY=0.8}) {
   samplerSource=SSM_Wrap_WorldGroupSettings
   samplerType=SAMPLERTYPE_Grayscale
   texture=/Game/TimeOfDay/TODM_Sequence/Materials/Sky/Textures/LowResBlurredNoise_Radish_LF_Root
  }
  n15 = Constant()
  n16 = Add(A=TextureCoordinate(), B=Multiply(A=QualitySwitch(Default=@n14.R, Low=@n15, High=@n14.R, Medium=@n15, Epic=@n14.R), B=Constant() {r=0.004}))
  n17 = Saturate(None=Add(
    A=TextureSample(UVs=@n16) {texture=&T_hole_mask_all}.R
    B=Constant() {r=0.02}
   ))
  n18 = Constant() {r=2}
  n19 = LinearInterpolate(Alpha=Saturate(None=Power(Base=@n17, Exp=@n18))) {constA=0.01}
  n20 = Constant() {r=1.5}
  NamedRerouteDeclaration(None=LinearInterpolate(A=Min(A=@n19, B=@n12), Alpha=Power(Base=@n11.G, Exp=@n20) {constExponent=3})) {name=Smoke, nodeColor=<rgba 0.9219 0 1 1>}
  n21 = TextureSample() {samplerSource=SSM_Wrap_WorldGroupSettings, texture=/BRCosmetics/Characters/Player/Male/Medium/Bodies/M_MED_Patches/Textures/T_flameUV}
  n22 = ComponentMask(None=@n21.RGB) {g=true}
  n23 = Constant() {r=0.6557}
  n24 = TextureCoordinate()
  n25 = Constant() {r=14}
  n26 = TextureSample(UVs=Panner(Coordinate=Multiply(A=@n24, B=@n25)) {speedX=-0.2, speedY=0.8}) {
   samplerSource=SSM_Wrap_WorldGroupSettings
   samplerType=SAMPLERTYPE_Grayscale
   texture=&LowResBlurredNoise_Radish_LF_Root
  }
  n27 = Add(A=AppendVector(A=ComponentMask(None=@n21.RGB) {r=true}, B=Multiply(A=@n22, B=@n23)), B=Multiply(A=@n26.R, B=Constant() {r=0.008}))
  n28 = AppendVector(A=Constant() {r=0.156}, B=Constant() {r=0.14})
  n29 = Subtract(A=LinearInterpolate(Alpha=NamedRerouteUsage()) {constB=0.2}) {constB=0.003}
  n30 = Constant() {r=0.1}
  n31 = Power(Base=OneMinus(None=SphereMask(A=@n27, B=@n28, Radius=@n29, Hardness=@n30))) {constExponent=3}
  n32 = AppendVector(A=Constant() {r=0.156}, B=Constant() {r=0.14})
  n33 = Subtract(A=LinearInterpolate(Alpha=NamedRerouteUsage()) {constB=0.2}, B=Constant() {r=0.05}) {constB=0.003}
  n34 = Constant() {r=0.9}
  n35 = Multiply(A=OneMinus(None=SphereMask(A=@n27, B=@n32, Radius=@n33, Hardness=@n34)), B=@n31)
  n36 = TextureSample() {samplerSource=SSM_Wrap_WorldGroupSettings, texture=&T_hole_mask_all}
  n37 = Add(A=Multiply(A=OneMinus(None=@n36.R)) {constB=3}, B=OneMinus(None=@n36.A))
  NamedRerouteDeclaration(None=LinearInterpolate(B=QualitySwitch(Default=@n31, Low=@n31, High=@n31, Medium=@n31, Epic=@n35), Alpha=Saturate(None=@n37)) {constA=1}) {name=Smoke ani, nodeColor=<rgba 0.0078 0 1 1>}
  n38 = TextureSample() {samplerSource=SSM_Wrap_WorldGroupSettings, texture=&T_hole_mask_all}
  n39 = Subtract(A=NamedRerouteUsage(), B=Constant() {r=0.8})
  n40 = Saturate(None=Multiply(A=Saturate(None=@n39), B=Constant() {r=10}) {constB=9})
  n41 = LinearInterpolate(A=Multiply(A=NamedRerouteUsage()) {constB=0.3}, B=@n38.B, Alpha=@n40)
  n42 = Add(A=Power(Base=@n41, Exp=Constant() {r=1.8372}), B=Multiply(A=@n38.B, B=LinearInterpolate(Alpha=@n40) {constB=0.2}) {constB=0.2})
  n43 = StaticSwitchParameter(
   True=VectorParameter() {group=Patches, parameterName=Body PinkRed Edge, defaultValue=<rgba 1 0.1264 0.697 1>}.RGB
   False=LinearInterpolate(
    A=VectorParameter() {group=Patches, parameterName=Body Blue Edge, defaultValue=<rgba 0.5774 0.0425 0.85 1>}.RGB
    B=VectorParameter() {group=Patches, parameterName="Body  Edge color2", defaultValue=<rgba 0.2501 0.0425 0.85 1>}.RGB
    Alpha=NamedRerouteUsage()
   )
  ) {group=Patches, parameterName="blue or  pink Emissive edge"}
  n44 = ScalarParameter() {defaultValue=4, group=Patches, parameterName=Clothing edge Emissive Min}
  n45 = ScalarParameter() {defaultValue=3000, group=Patches, parameterName=Clothing edge Emissive Max}
  n46 = NamedRerouteUsage()
  n47 = Constant() {r=0.1562}
  n48 = LinearInterpolate(A=Constant() {r=1.167}, B=Constant() {r=0.773}, Alpha=NamedRerouteUsage())
  n49 = Saturate(None=Multiply(A=Multiply(A=@n46, B=@n47), B=Power(Base=@n41, Exp=@n48)))
  NamedRerouteDeclaration(None=Multiply(A=Multiply(A=@n42, B=@n43), B=LinearInterpolate(A=@n44, B=@n45, Alpha=@n49))) {name=Emissive color1, nodeColor=<rgba 0.9219 0 1 1>}
  n50 = TextureSample() {samplerSource=SSM_Wrap_WorldGroupSettings, texture=/BRCosmetics/Characters/Player/Male/Medium/Bodies/M_MED_Patches/Textures/T_HairBurn_Mask}
  n51 = NamedRerouteUsage()
  n52 = NamedRerouteUsage()
  n53 = Constant() {r=0.7}
  n54 = Multiply(A=Saturate(None=Subtract(A=@n52, B=@n53)), B=Constant() {r=5}) {constB=9}
  n55 = Multiply(
   A=LinearInterpolate(B=ScalarParameter() {defaultValue=2, group=Patches, parameterName=Clothing edge noise}, Alpha=Saturate(None=@n54)) {constA=0.5}
   B=NamedRerouteUsage()
  )
  n56 = Saturate(None=Power(Base=@n50.R, Exp=Add(A=@n51, B=@n55)))
  n57 = LinearInterpolate(
   A=@n56
   B=LinearInterpolate(A=@n56, Alpha=ScalarParameter() {defaultValue=0.95, group=Patches, parameterName=Tie burning offset}) {constAlpha=0.8}
   Alpha=@n50.A
  )
  n58 = TextureSample() {samplerSource=SSM_Wrap_WorldGroupSettings, texture=&T_flameUV}
  n59 = AppendVector(A=ComponentMask(None=@n58.RGB) {r=true}, B=Multiply(A=ComponentMask(None=@n58.RGB) {g=true}, B=Constant() {r=0.6557}))
  n60 = TextureSample(UVs=Multiply(A=TextureCoordinate(), B=Constant() {r=33.4965})) {
   samplerSource=SSM_Wrap_WorldGroupSettings
   samplerType=SAMPLERTYPE_Grayscale
   texture=&LowResBlurredNoise_Radish_LF_Root
  }
  n61 = Multiply(A=@n60.R, B=Constant() {r=0.008})
  n62 = Constant() {r=0.156}
  n63 = Constant() {r=0.14}
  n64 = NamedRerouteUsage()
  n65 = LinearInterpolate(Alpha=@n64) {constB=0.2}
  n66 = LinearInterpolate(B=ScalarParameter() {defaultValue=0.25, group=Patches, parameterName=Tie mixing radius offset}, Alpha=@n64) {constB=0.32}
  n67 = TextureSample() {samplerSource=SSM_Wrap_WorldGroupSettings, texture=&T_HairBurn_Mask}
  n68 = OneMinus(None=SphereMask(A=Add(A=@n59, B=@n61), B=AppendVector(A=@n62, B=@n63), Radius=LinearInterpolate(A=@n65, B=@n66, Alpha=@n67.A), Hardness=Constant() {r=0.1}))
  n69 = TextureSample() {samplerSource=SSM_Wrap_WorldGroupSettings, texture=&T_hole_mask_all}
  n70 = Add(A=Multiply(A=OneMinus(None=@n69.R)) {constB=3}, B=OneMinus(None=@n69.A))
  NamedRerouteDeclaration(None=LinearInterpolate(B=Multiply(A=@n57, B=@n68) {constA=1}, Alpha=Saturate(None=@n70)) {constA=1}) {name=Animation Mask, nodeColor=<rgba 0 1 0.3672 1>}
  n71 = TextureCoordinate()
  n72 = Constant() {r=20.2022}
  n73 = TextureSample(UVs=Panner(Coordinate=Multiply(A=@n71, B=@n72)) {speedX=-0.2, speedY=0.2}) {
   samplerSource=SSM_Wrap_WorldGroupSettings
   samplerType=SAMPLERTYPE_Grayscale
   texture=&LowResBlurredNoise_Radish_LF_Root
  }
  n74 = TextureCoordinate()
  n75 = Constant() {r=30}
  n76 = TextureSample(UVs=Panner(Coordinate=Multiply(A=@n74, B=@n75)) {speedX=-0.4, speedY=0.8}) {
   samplerSource=SSM_Wrap_WorldGroupSettings
   samplerType=SAMPLERTYPE_Grayscale
   texture=&LowResBlurredNoise_Radish_LF_Root
  }
  NamedRerouteDeclaration(None=Max(A=Power(Base=@n73.R), B=@n76.R)) {name=Blend mask, nodeColor=<rgba 0.3828 0 1 1>}
  n77 = NamedRerouteUsage()
  n78 = Multiply(A=Power(Base=OneMinus(None=@n77)) {constExponent=3.2})
  n79 = NamedRerouteUsage()
  n80 = Constant() {r=0.338}
  n81 = Add(A=ScalarParameter() {defaultValue=0.015, group=Patches, parameterName=Clothing edge Emissive insideWidth}, B=@n80)
  n82 = Constant()
  n83 = Constant() {r=1}
  n84 = If(A=NamedRerouteUsage(), B=@n80, "A > B"=@n83, "A == B"=@n83, "A < B"=Constant()) {constB=0.01}
  n85 = TextureSample(UVs=Multiply(A=TextureCoordinate(), B=ScalarParameter() {defaultValue=10, group=Patches, parameterName=Clothing edge Emissive InsideTiling})) {
   samplerSource=SSM_Wrap_WorldGroupSettings
   samplerType=SAMPLERTYPE_Grayscale
   texture=&LowResBlurredNoise_Radish_LF_Root
  }
  n86 = NamedRerouteUsage()
  n87 = Constant() {r=0.338}
  n88 = Add(A=@n87, B=ScalarParameter() {defaultValue=0.05, group=Patches, parameterName=Clothing edge Emissive Width})
  n89 = Constant()
  n90 = Constant() {r=1}
  n91 = If(A=NamedRerouteUsage(), B=@n87, "A > B"=@n90, "A == B"=@n90, "A < B"=Constant()) {constB=0.01}
  n92 = TextureSample(UVs=Multiply(A=TextureCoordinate(), B=ScalarParameter() {defaultValue=45, group=Patches, parameterName=Clothing edge Emissive Tiling})) {
   samplerSource=SSM_Wrap_WorldGroupSettings
   samplerType=SAMPLERTYPE_Grayscale
   texture=&LowResBlurredNoise_Radish_LF_Root
  }
  n93 = Add(
   A=Multiply(A=If(A=@n79, B=@n81, "A > B"=@n82, "A == B"=@n84, "A < B"=@n84) {constB=0.02}, B=@n85.R)
   B=Multiply(A=If(A=@n86, B=@n88, "A > B"=@n89, "A == B"=@n91, "A < B"=@n91) {constB=0.02}, B=@n92.R)
  )
  n94 = TextureSample() {samplerSource=SSM_Wrap_WorldGroupSettings, texture=&T_hole_mask_all}
  n95 = Multiply(A=Add(A=@n94.B, B=OneMinus(None=@n94.A))) {constB=10}
  NamedRerouteDeclaration(None=LinearInterpolate(B=Add(A=@n78, B=@n93), Alpha=Saturate(None=@n95))) {name=Sparky2, nodeColor=<rgba 1 0.2812 0 1>}
  NamedRerouteDeclaration(None=Saturate(None=ScalarParameter() {group=Patches, parameterName=BurnRadius})) {name=Animation mixing strength, nodeColor=<rgba 0.3594 0 1 1>}
 ]}
