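# Material function for the Patches character body that layers an animated glowing heart and optional burnt-cloth holes onto input material attributes.
# Heart masks, Fresnel, a sine pulse and panning, normal-distorted smoke on radial UVs build HDR glow, with extra smoke at Epic quality.
# Smoke samples take mip derivatives from unscrolled UVs to avoid seams.
# Static switches select MF_HoleMask outputs, while masks adjust normal, specular, roughness and metallic.
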MaterialFunction /BRCosmetics/Characters/Player/Male/Medium/Bodies/M_MED_Patches/Materials/MF_Patches_burn
graph {nodes [
  n1 = GetMaterialAttributes(None=FunctionInput() {inputType=FunctionInput_MaterialAttributes}) {attributeGetTypes [
    #1
    #2
    #3
    #4
    #5
    #6
   ]}
  n2 = NamedRerouteUsage()
  n3 = SetMaterialAttributes!(
   MaterialAttributes=@n1.MaterialAttributes
   EmissiveColor=NamedRerouteUsage()
   OpacityMask=StaticSwitchParameter(True=NamedRerouteUsage(), False=@n1.OpacityMask) {group=Patches, parameterName=use Opacity mask}
   Normal=NamedRerouteUsage()
   Specular=NamedRerouteUsage()
   WorldPositionOffset=Constant()
   BaseColor=StaticSwitchParameter(True=NamedRerouteUsage(), False=NamedRerouteUsage()) {group=Patches, parameterName=use cloth burn Base color}
   Roughness=LinearInterpolate(A=@n1.Roughness, Alpha=Saturate(None=@n2)) {constB=0}
   Metallic=LinearInterpolate(A=@n1.Metallic, Alpha=NamedRerouteUsage()) {constB=0}
  ) {attributeSetTypes [
    #7
    #4
    #1
    #2
    #8
    #3
    #5
    #6
   ]}
  FunctionOutput(None=@n3)
  LinearInterpolate(B=NamedRerouteUsage(), Alpha=NamedRerouteUsage())
  n4 = TextureSample() {
   samplerSource=SSM_Wrap_WorldGroupSettings
   texture=/BRCosmetics/Characters/Player/Male/Medium/Bodies/M_MED_Patches/Textures/T_heartmask_flame
  }
  n5 = ComponentMask(None=LinearInterpolate(A=Constant3Vector() {constant=<rgba -3.29 0.2 0 1>}, B=Constant3Vector() {constant=<rgba -2.02 -0.72 0 1>}, Alpha=@n4.G)) {g=true, r=true}
  n6 = MaterialFunctionCall(UVs=TextureCoordinate(), Texture Scale=Constant() {r=0.1544}) {materialFunction=&ScaleUVsByCenter}
  NamedRerouteDeclaration(None=MaterialFunctionCall(Vector or UVs=Add(A=@n5, B=@n6.UVs)) {materialFunction=&VectorToRadialValue}.Radial Coordinates) {name=HeartSmokeUV, nodeColor=<rgba 0.2422 0 1 1>}
  n7 = MaterialFunctionCall!(BaseColor=NamedRerouteUsage()) {materialFunction=&/MF_HoleMask}
  NamedRerouteDeclaration(None=@n7.Mask) {name=Opacity Mask, nodeColor=<rgba 1 0 0.9141 1>}
  n8 = Constant()
  n9 = TextureSample() {texture=/BRCosmetics/Characters/Player/Male/Medium/Bodies/M_MED_Patches/Textures/T_heartBG_FX}
  n10 = LinearInterpolate(B=NamedRerouteUsage(), Alpha=Add(A=@n9.R, B=@n9.R))
  NamedRerouteDeclaration(None=StaticSwitchParameter(True=Add(A=@n8, B=@n10), False=@n7.EmissiveColor) {defaultValue=true, group=Patches, parameterName=use heart emissive}) {name=Emissive color, nodeColor=<rgba 0.6172 0 1 1>}
  n11 = NamedRerouteUsage()
  n12 = Constant()
  n13 = OneMinus(None=@n7.Mask_2)
  n14 = Constant()
  n15 = TextureSample() {samplerSource=SSM_Wrap_WorldGroupSettings, texture=&T_heartBG_FX}
  n16 = Saturate(None=Add(A=StaticSwitchParameter(True=@n13, False=@n14) {group=Patches, parameterName=useCloth Specular}, B=Add(A=@n15.R, B=@n15.R)))
  NamedRerouteDeclaration(None=LinearInterpolate(A=LinearInterpolate(A=@n11, B=@n12, Alpha=@n16) {constA=0.5}, Alpha=NamedRerouteUsage()) {constB=0}) {name=Specular, nodeColor=<rgba 0.875 0 1 1>}
  NamedRerouteDeclaration(None=@n1.BaseColor) {name=Input BaseColor, nodeColor=<rgba 1 0 0.9609 1>}
  NamedRerouteDeclaration(None=@n7.base color) {name=burn cloth base color, nodeColor=<rgba 1 0 0.7266 1>}
  NamedRerouteDeclaration(None=@n1.Normal) {name=input normal, nodeColor=<rgba 0.3125 0 1 1>}
  NamedRerouteDeclaration(None=@n1.Specular) {name=input Specular, nodeColor=<rgba 1 0 0.4922 1>}
  n17 = Saturate(None=Fresnel(ExponentIn=Constant() {r=1.3}, BaseReflectFractionIn=Constant() {r=-0.104}) {exponent=2})
  n18 = Constant() {r=1.5}
  n19 = NamedRerouteUsage()
  n20 = NamedRerouteUsage()
  n21 = Constant2Vector() {g=0.08, r=0.1}
  n22 = TextureCoordinate()
  n23 = TextureSample!(
   UVs=Panner(Coordinate=@n20, Speed=@n21) {speedX=0.025, speedY=0.01}
   Tex
   "DDX(UVs)"=DDX(Value=@n22)
   "DDY(UVs)"=DDY(Value=@n22)
   Apply View MipBias
  ) {
   mipValueMode=TMVM_Derivative
   samplerSource=SSM_Wrap_WorldGroupSettings
   samplerType=SAMPLERTYPE_Normal
   texture=/BRCosmetics/Characters/Player/Male/Medium/Bodies/M_MED_Patches/Textures/T_heartnoise_N
  }
  n24 = Multiply(A=ComponentMask(None=@n23.RGB) {g=true, r=true}, B=Constant() {r=0.3})
  n25 = Constant2Vector() {g=0.04, r=0.02}
  n26 = TextureCoordinate()
  n27 = TextureSample(UVs=Panner(Coordinate=Add(A=@n19, B=@n24), Speed=@n25) {speedX=-0.05, speedY=0.25}, "DDX(UVs)"=DDX(Value=@n26), "DDY(UVs)"=DDY(Value=@n26)) {
   mipValueMode=TMVM_Derivative
   samplerSource=SSM_Wrap_WorldGroupSettings
   texture=/BRCosmetics/Characters/Player/Male/Medium/Bodies/M_MED_Patches/Textures/T_Smoke_01
  }
  n28 = NamedRerouteUsage()
  n29 = TextureCoordinate()
  n30 = TextureSample(UVs=Panner(Coordinate=@n20, Speed=@n21) {speedX=-0.015, speedY=0.01}, "DDX(UVs)"=DDX(Value=@n29), "DDY(UVs)"=DDY(Value=@n29)) {
   mipValueMode=TMVM_Derivative
   samplerSource=SSM_Wrap_WorldGroupSettings
   samplerType=SAMPLERTYPE_Normal
   texture=&T_heartnoise_N
  }
  n31 = Multiply(A=ComponentMask(None=@n30.RGB) {g=true, r=true}, B=Constant() {r=2})
  n32 = TextureCoordinate()
  n33 = TextureSample(UVs=Panner(Coordinate=Add(A=@n28, B=@n31), Speed=@n25) {speedX=0.01, speedY=0.25}, "DDX(UVs)"=DDX(Value=@n32), "DDY(UVs)"=DDY(Value=@n32)) {
   mipValueMode=TMVM_Derivative
   samplerSource=SSM_Wrap_WorldGroupSettings
   texture=&T_Smoke_01
  }
  n34 = Power(Base=QualitySwitch(Default=@n27.R, Low=@n27.R, High=@n27.R, Medium=@n27.R, Epic=Add(A=@n27.R, B=@n33.R)), Exp=Constant() {r=1.2}) {constExponent=1.2}
  n35 = Constant()
  n36 = TextureSample!(UVs, Tex, Apply View MipBias) {
   samplerSource=SSM_Wrap_WorldGroupSettings
   texture=&T_heartmask_flame
  }
  n37 = TextureCoordinate()
  n38 = Constant() {r=29.2896}
  n39 = TextureSample(UVs=Panner(Coordinate=Multiply(A=@n37, B=@n38)) {speedX=0.6, speedY=0.1}) {
   samplerSource=SSM_Wrap_WorldGroupSettings
   samplerType=SAMPLERTYPE_Grayscale
   texture=/Game/TimeOfDay/TODM_Sequence/Materials/Sky/Textures/LowResBlurredNoise_Radish_LF_Root
  }
  n40 = Constant() {r=0.006}
  n41 = TextureSample(UVs=Add(A=TextureCoordinate(), B=Multiply(A=@n39.R, B=@n40))) {
   samplerSource=SSM_Wrap_WorldGroupSettings
   texture=&T_heartmask_flame
  }
  n42 = Saturate(None=Power(Base=@n41.A, Exp=Constant() {r=2}))
  n43 = NamedRerouteDeclaration(None=LinearInterpolate(A=@n42, B=Multiply(A=@n42) {constB=0}, Alpha=@n36.B)) {name=heart, nodeColor=<rgba 1 0.1406 0 1>}
  n44 = Constant() {r=0.1}
  n45 = Clamp(None=Max(A=LinearInterpolate(A=@n34, B=@n35, Alpha=@n36.B), B=Multiply(A=@n43, B=@n44)))
  n46 = Constant() {r=0.27}
  n47 = Constant() {r=8}
  n48 = Constant() {r=0.98}
  n49 = NamedRerouteUsage()
  n50 = LinearInterpolate(
   A=Constant3Vector() {constant=<rgba 0 0 0 1>}
   B=Constant() {r=2.1625}
   Alpha=Power(Base=Fresnel(ExponentIn=@n46) {exponent=2}, Exp=LinearInterpolate(A=@n47, B=@n48, Alpha=@n49))
  )
  NamedRerouteDeclaration(None=Add(A=Multiply(A=@n17, B=@n18) {constB=0.5}, B=Multiply(A=@n45, B=@n50))) {name=Edge flame, nodeColor=<rgba 0.9453 0 1 1>}
  n51 = Constant()
  n52 = Constant3Vector() {constant=<rgba 13.1649 0 15 1>}
  n53 = NamedRerouteUsage()
  n54 = Constant3Vector() {constant=<rgba 0.5171 0 1 1>}
  n55 = Constant3Vector() {constant=<rgba 12.7067 0.9 18 1>}
  n56 = Constant() {r=0.45}
  n57 = Constant() {r=0.6}
  n58 = MaterialFunctionCall(Value=Multiply(A=Time()) {constB=0.75}) {materialFunction=&LinearSine}
  n59 = Power(
   Base=Fresnel(ExponentIn=LinearInterpolate(A=@n56, B=@n57, Alpha=@n58.Linear Sine) {constB=0.6}) {exponent=2}
   Exp=LinearInterpolate(A=Constant() {r=5.4999}, B=Constant() {r=10}, Alpha=NamedRerouteUsage())
  )
  n60 = Saturate(None=Power(Base=OneMinus(None=@n59)) {constExponent=50})
  n61 = Add(A=Add(A=LinearInterpolate(A=@n51, B=@n52, Alpha=@n53), B=LinearInterpolate(A=@n54, B=@n55, Alpha=@n60))) {constB=0}
  n62 = TextureCoordinate()
  n63 = Constant() {r=10}
  n64 = TextureSample(UVs=Panner(Coordinate=Multiply(A=@n62, B=@n63)) {speedX=0.6, speedY=0.1}) {
   samplerSource=SSM_Wrap_WorldGroupSettings
   samplerType=SAMPLERTYPE_Grayscale
   texture=&LowResBlurredNoise_Radish_LF_Root
  }
  n65 = LinearInterpolate(A=Constant() {r=0.5}, B=Constant() {r=3}, Alpha=@n64.R)
  NamedRerouteDeclaration(None=MaterialFunctionCall(LDR Input=Clamp(None=@n61), HDR Tint=Multiply(A=@n61, B=@n65)) {materialFunction=&DeriveHDRfromLDR}.RemappedRange) {name=Heart fx, nodeColor=<rgba 0 0.5547 1 1>}
  n66 = ComponentMask(None=LinearInterpolate(
    A=Constant3Vector() {constant=<rgba -2.9347 -1.802 0 1>}
    B=Constant3Vector() {constant=<rgba -2.9284 -0.7695 0 1>}
    Alpha=TextureSample() {texture=&T_heartBG_FX}.G
   )) {g=true, r=true}
  n67 = MaterialFunctionCall(UVs=TextureCoordinate(), Texture Scale=Constant() {r=0.1624}) {materialFunction=&ScaleUVsByCenter}
  NamedRerouteDeclaration(None=MaterialFunctionCall(Vector or UVs=Add(A=@n66, B=@n67.UVs)) {materialFunction=&VectorToRadialValue}.Radial Coordinates) {name=HeartBGSmokeUV_1, nodeColor=<rgba 0.2422 0 1 1>}
  n68 = Constant3Vector() {constant=<rgba 1.5465 0.1 5 1>}
  n69 = Constant3Vector() {constant=<rgba 47.5428 0.5 50 1>}
  n70 = Constant() {r=-0.008}
  n71 = NamedRerouteUsage()
  n72 = NamedRerouteUsage()
  n73 = Constant3Vector() {constant=<rgba 0.02 -0.08 0 1>}
  n74 = TextureCoordinate()
  n75 = TextureSample!(
   UVs=Panner(Coordinate=@n72, Speed=@n73) {speedX=0.025, speedY=0.01}
   Tex
   "DDX(UVs)"=DDX(Value=@n74)
   "DDY(UVs)"=DDY(Value=@n74)
   Apply View MipBias
  ) {
   mipValueMode=TMVM_Derivative
   samplerType=SAMPLERTYPE_Normal
   texture=&T_heartnoise_N
  }
  n76 = Multiply(A=ComponentMask(None=@n75.RGB) {g=true, r=true}, B=Constant() {r=0.5})
  n77 = Constant3Vector() {constant=<rgba 0.02 -0.1 0 1>}
  n78 = TextureCoordinate()
  n79 = TextureSample(UVs=Panner(Coordinate=Add(A=@n71, B=@n76), Speed=@n77) {speedX=-0.05, speedY=0.25}, "DDX(UVs)"=DDX(Value=@n78), "DDY(UVs)"=DDY(Value=@n78)) {mipValueMode=TMVM_Derivative, samplerType=SAMPLERTYPE_LinearColor, texture=/Game/Global/Textures/Default/Noise/T_Curie_NoisePack}
  n80 = TextureSample!(UVs=Add(A=TextureCoordinate(), B=Multiply(A=@n79.R, B=@n70)), Tex, Apply View MipBias) {samplerSource=SSM_Wrap_WorldGroupSettings, texture=&T_heartBG_FX}
  n81 = NamedRerouteUsage()
  n82 = TextureCoordinate()
  n83 = TextureSample!(
   UVs=Panner(Coordinate=@n72, Speed=@n73) {speedX=-0.015, speedY=0.01}
   Tex
   "DDX(UVs)"=DDX(Value=@n82)
   "DDY(UVs)"=DDY(Value=@n82)
   Apply View MipBias
  ) {
   mipValueMode=TMVM_Derivative
   samplerType=SAMPLERTYPE_Normal
   texture=&T_heartnoise_N
  }
  n84 = Multiply(A=ComponentMask(None=@n83.RGB) {g=true, r=true}, B=Constant() {r=2})
  n85 = TextureCoordinate()
  n86 = TextureSample(UVs=Panner(Coordinate=Add(A=@n81, B=@n84), Speed=@n77) {speedX=0.01, speedY=0.25}, "DDX(UVs)"=DDX(Value=@n85), "DDY(UVs)"=DDY(Value=@n85)) {mipValueMode=TMVM_Derivative, samplerType=SAMPLERTYPE_LinearColor, texture=&T_Curie_NoisePack}
  n87 = QualitySwitch(Default=@n79.R, Low=@n79.R, High=@n79.R, Medium=@n79.R, Epic=Add(A=Max(A=@n79.R, B=@n79.R), B=Max(A=@n86.R, B=@n86.R)))
  n88 = Constant() {r=2}
  n89 = Multiply(
   A=Multiply(A=Power(Base=@n87, Exp=@n88) {constExponent=1.2}, B=Constant() {r=0.5})
   B=LinearInterpolate(B=Constant() {r=0.042}, Alpha=@n80.B)
  )
  n90 = Multiply(A=Power(Base=@n80.B) {constExponent=0.5}, B=Constant() {r=0.32}) {constB=0.02}
  NamedRerouteDeclaration(None=LinearInterpolate(B=LinearInterpolate(A=@n68, B=@n69, Alpha=@n80.B), Alpha=Add(A=@n89, B=@n90))) {name=heart BG, nodeColor=<rgba 0.9219 0 1 1>}
  n91 = TextureSample!(UVs, Tex, Apply View MipBias) {texture=&T_heartmask_flame}
  NamedRerouteDeclaration(None=Add(A=@n91.R, B=@n91.R)) {name=HeartMask, nodeColor=<rgba 0.4297 1 0 1>}
  n92 = NamedRerouteUsage()
  n93 = NamedRerouteUsage()
  n94 = TextureSample() {texture=&T_heartmask_flame}
  n95 = LinearInterpolate(Alpha=Add(A=@n94.R, B=@n94.R)) {constB=0.8}
  NamedRerouteDeclaration(None=StaticSwitchParameter(
    True=MaterialFunctionCall(BaseNormal=@n92, AdditionalNormal=@n7.Normal) {materialFunction=&BlendAngleCorrectedNormals}.Result
    False=MaterialFunctionCall(Normal=@n93, Flatness=@n95) {materialFunction=&FlattenNormal}.Result
   ) {group=Patches, parameterName=use cloth burn normal}) {name=Normal, nodeColor=<rgba 0 1 0.2031 1>}
 ]}
