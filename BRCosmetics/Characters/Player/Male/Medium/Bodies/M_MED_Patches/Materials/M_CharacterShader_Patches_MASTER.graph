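# Master character material for the Patches medium male body, using material attributes with an expression-driven shading model.
# Base PreFX output passes through switchable thin-film, metal LUT and silk layers, with masks from the SkinFX texture, then anisotropic shading.
# Patches-specific burn, on by default, and optional hair burn follow.
# The result then runs through base character PostFX and cinematic PostFX functions.
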Material /BRCosmetics/Characters/Player/Male/Medium/Bodies/M_MED_Patches/Materials/M_CharacterShader_Patches_MASTER
graph {
 properties {
  bAutomaticallySetUsageInEditor=false
  bUseMaterialAttributes=true
  bUsedWithClothing=true
  bUsedWithMeshDeformer=true
  bUsedWithMorphTargets=true
  bUsedWithNanite=true
  bUsedWithSkeletalMesh=true
  pixelDepthOffsetMode=PDOM_Legacy
  refractionMethod=RM_IndexOfRefraction
  shadingModel=MSM_FromMaterialExpression
  subsurfaceProfile=&SS_HeroSkin_02
  usedShadingModels=Default Lit | Subsurface | Clear Coat | Subsurface Profile | Cloth | Eye
 }
 nodes [
  n1 = MaterialFunctionCall() {materialFunction=&BaseFN_CharacterFunction_PreFX}
  n2 = StaticComponentMaskParameter(None=NamedRerouteUsage()) {defaultR=true, group=Skin FX, parameterName=ThinSilk_Channel}
  n3 = NamedRerouteUsage()
  n4 = StaticSwitchParameter(
   True=MaterialFunctionCall!(Alpha=StaticComponentMaskParameter(None=@n3) {defaultR=true, group=Skin FX, parameterName=ThinFilm_Channel}, Material Attributes=@n1.Result) {materialFunction=/Game/Weapons/WeaponSkins/MaterialFunction/ThinFilmFunction}.Result
   False=@n1.Result
  ) {group=Skin FX, parameterName=Use Thin Film}
  n5 = StaticSwitchParameter(
   True=MaterialFunctionCall(Mask=NamedRerouteUsage(), Materials=@n4) {materialFunction=&MF_MetalsLUT}.Result
   False=@n4
  ) {group=Skin FX, parameterName=UseMetalLUT}
  n6 = MaterialFunctionCall!(In=StaticSwitchParameter(True=MaterialFunctionCall!(Alpha=@n2, Material=@n5) {materialFunction=/Game/Weapons/WeaponSkins/MaterialFunction/MF_Silk}.Result, False=@n5) {group=Skin FX, parameterName=Use Silk}) {materialFunction=/Game/Weapons/WeaponSkins/MaterialFunction/MF_AnisotropicShading}
  n7 = MaterialFunctionCall!(In=@n6.Result) {materialFunction=&/MF_Patches_burn}
  n8 = StaticSwitchParameter(True=ShadingPathSwitch(Default=@n7.Result, Mobile=@n7.Result), False=@n6.Result) {defaultValue=true, group=Patches, parameterName=Use Patches burn}
  n9 = MaterialFunctionCall(
   AO=@n1.AO
   Distance Blend=@n1.AO
   Fresnel=@n1.AO
   Gmap=@n1.AO
   Pre Skinned Local Position=@n1.AO
   Material=StaticSwitchParameter(
    True=MaterialFunctionCall!(In=@n8) {materialFunction=&/MF_Patches_HairBurn}.Result
    False=@n8
   ) {group=Patches, parameterName=Use Patches Hair burn}
  ) {materialFunction=&BaseFN_CharacterFunction_PostFX}
  MaterialFunctionCall(In=@n9.Result) {materialFunction=&CINE_MF_CharPostFX}
  MaterialFunctionCall(
   PreskinnedPosition=@n1.Pre skinned Local Position
   Gmap=@n1.Pre skinned Local Position
   MaterialAttributes=@n1.Pre skinned Local Position
   AO=@n1.Pre skinned Local Position
  ) {materialFunction=&CharacterSkinFX}
  n10 = TextureObjectParameter() {group=Skin FX, parameterName=SkinFX_Mask}
  NamedRerouteDeclaration(None=TextureSample(UVs=TextureCoordinate(), Tex=NamedRerouteDeclaration(None=@n10) {name=SkinFX_Mask, nodeColor=<rgba 0.9842 0 1 1>}).RGBA) {name=SkinFX_Mask_Sampled, nodeColor=<rgba 0.7475 0 1 1>}
 ]
}
