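# Corrective Control Rig for the M_MED_Patches male body cosmetic, adjusting jacket, right upper-arm and watch bones.
# A sequence runs spherical pose readers on the left thigh, right upper arm and left hand.
# Reader outputs, gated by movement, crouch and pistol variables and often smoothed or inverted with AlphaInterp, weight additional bone offsets.
# Some offsets ignore the readers, including the targeting-driven shirt shift and constant watch nudge; two upper-arm nodes are unconnected.
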ControlRigBlueprint /BRCosmetics/Characters/Player/Male/Medium/Bodies/M_MED_Patches/Meshes/M_MED_Patches_CtrlRig
graphs [
 ContainedGraph {nodes [
   n1 = RigVMFunctionEntryNode()
   SequenceAggregate = &RigVMFunction_Sequence(ExecuteContext=@n1.ExecuteContext)
   SequenceAggregate_1 = &RigVMFunction_Sequence(ExecuteContext=@SequenceAggregate.B)
   SequenceAggregate_2 = &RigVMFunction_Sequence(ExecuteContext=@SequenceAggregate_1.B)
   SequenceAggregate_3 = &RigVMFunction_Sequence(ExecuteContext=@SequenceAggregate_2.B)
   SequenceAggregate_4 = &RigVMFunction_Sequence(ExecuteContext=@SequenceAggregate_3.B)
   SequenceAggregate_5 = &RigVMFunction_Sequence(ExecuteContext=@SequenceAggregate_4.B)
   SequenceAggregate_6 = &RigVMFunction_Sequence(ExecuteContext=@SequenceAggregate_5.B)
   SequenceAggregate_7 = &RigVMFunction_Sequence(ExecuteContext=@SequenceAggregate_6.B)
   SequenceAggregate_8 = &RigVMFunction_Sequence(ExecuteContext=@SequenceAggregate_7.B)
   SequenceAggregate_9 = &RigVMFunction_Sequence(ExecuteContext=@SequenceAggregate_8.B)
   RigVMFunctionReturnNode(
    A=@SequenceAggregate.A
    B=@SequenceAggregate_1.A
    C=@SequenceAggregate_2.A
    D=@SequenceAggregate_3.A
    E=@SequenceAggregate_4.A
    F=@SequenceAggregate_5.A
    G=@SequenceAggregate_6.A
    H=@SequenceAggregate_7.A
    I=@SequenceAggregate_8.A
    J=@SequenceAggregate_9.A
    K=@SequenceAggregate_9.B
   )
  ]}
 RigVMFunctionLibrary {links [], nodes []}
 RigVMModel {nodes [
   RigUnit_BeginExecution = &RigUnit_BeginExecution()
   SequenceAggregate_1_1_1 = RigVMAggregateNode(ExecuteContext=@RigUnit_BeginExecution.ExecutePin)
   n1 = &RigUnit_SphericalPoseReader(
    ActiveRegionScaleFactors=(PositiveWidth=1,NegativeWidth=1,PositiveHeight=1,NegativeHeight=1)
    ActiveRegionSize=0.1
    Debug=(bDrawDebug=false,bDraw2D=false,bDrawLocalAxes=false,DebugScale=25,DebugSegments=20,DebugThickness=0.25)
    DriverAxis=(X=-1,Y=0,Z=0)
    DriverItem=(Type=Bone,Name="thigh_l")
    FalloffRegionScaleFactors=(PositiveWidth=0.176,NegativeWidth=0.184,PositiveHeight=0.216,NegativeHeight=0.104)
    FalloffSize=1
    FlipWidthScaling=False
    OptionalParentItem=(Type=Bone)
    RotationOffset=(X=120,Y=0,Z=70)
    ExecutePin=@SequenceAggregate_1_1_1.A
   )
   AlphaInterp = &RigVMFunction_AlphaInterp(InterpSpeedDecreasing=3, InterpSpeedIncreasing=5, Scale=1, bClampResult=False, Value=RigVMVariableNode() {variable=leftMove}.Value)
   Multiply = &RigVMFunction_MathDoubleMul(B=@AlphaInterp.Result, A=@n1.OutputParam)
   > &RigUnit_ModifyTransforms(
    ItemToModify=((Item=(Type=Bone,Name="L_back_BASE_jacket"),Transform=(Rotation=(X=0,Y=0,Z=0,W=1),Translation=(X=0,Y=0,Z=0),Scale3D=(X=1,Y=1,Z=1))))
    ItemToModify.0.Transform.Rotation=&RigVMFunction_MathQuaternionFromEuler(Euler=(X=0,Y=-26,Z=3.283)).Result
    Weight=@Multiply.Result
   )
   > &RigUnit_ModifyTransforms(
    ItemToModify=((Item=(Type=Bone,Name="L_back_BASE_jacket"),Transform=(Rotation=(X=0,Y=0,Z=0,W=1),Translation=(X=0,Y=0,Z=0),Scale3D=(X=1,Y=1,Z=1))))
    Weight=@AlphaInterp.Result
    ItemToModify.0.Transform.Rotation=&RigVMFunction_MathQuaternionFromEuler(Euler=(X=0,Y=-5,Z=1)).Result
   )
   > &RigUnit_ModifyTransforms(
    ItemToModify=((Item=(Type=Bone,Name="L_back_BASE_jacket"),Transform=(Rotation=(X=0,Y=0,Z=0,W=1),Translation=(X=0,Y=0,Z=0),Scale3D=(X=1,Y=1,Z=1))))
    Weight=1
    ItemToModify.0.Transform.Rotation=&RigVMFunction_MathQuaternionFromEuler(Euler=(X=0,Y=-5,Z=1)).Result
   )
   n2 = &RigVMFunction_AlphaInterp(
    InterpSpeedDecreasing=10
    InterpSpeedIncreasing=3
    Scale=1
    bClampResult=False
    Value=RigVMVariableNode() {variable=targeting_behaviour}.Value
   )
   > &RigUnit_ModifyTransforms(
    ItemToModify=((Item=(Type=Bone,Name="dyn_R_frontShirt_BASE_jacket"),Transform=(Rotation=(X=0,Y=0,Z=0,W=1),Translation=(X=0,Y=1,Z=0),Scale3D=(X=1,Y=1,Z=1))))
    Weight=@n2.Result
   )
   n3 = &RigUnit_SphericalPoseReader(
    ActiveRegionScaleFactors=(PositiveWidth=1,NegativeWidth=1,PositiveHeight=1,NegativeHeight=1)
    ActiveRegionSize=0.1
    Debug=(bDrawDebug=false,bDraw2D=false,bDrawLocalAxes=false,DebugScale=25,DebugSegments=20,DebugThickness=0.25)
    DriverAxis=(X=-1,Y=0,Z=0)
    DriverItem=(Type=Bone,Name="thigh_l")
    FalloffRegionScaleFactors=(PositiveWidth=0.176,NegativeWidth=0.184,PositiveHeight=0.216,NegativeHeight=0.104)
    FalloffSize=1
    FlipWidthScaling=False
    OptionalParentItem=(Type=Bone)
    RotationOffset=(X=140,Y=30,Z=70)
    ExecutePin=@SequenceAggregate_1_1_1.B
   )
   > RigVMVariableNode(Value=@n3.OutputParam) {variable=SP_output_thigh_l}
   n4 = &RigVMFunction_AlphaInterp(InterpSpeedDecreasing=3, InterpSpeedIncreasing=5, Scale=1, bClampResult=False, Value=RigVMVariableNode() {variable=crouchingLeftMove}.Value)
   > &RigUnit_ModifyTransforms(
    ItemToModify=((Item=(Type=Bone,Name="L_back_BASE_jacket"),Transform=(Rotation=(X=0,Y=0,Z=0,W=1),Translation=(X=0,Y=0,Z=0),Scale3D=(X=1,Y=1,Z=1))))
    ItemToModify.0.Transform.Rotation=&RigVMFunction_MathQuaternionFromEuler(Euler=(X=10,Y=-10,Z=-20)).Result
    Weight=&RigVMFunction_MathDoubleMul(B=@n4.Result, A=RigVMVariableNode() {variable=SP_output_thigh_l}.Value).Result
   )
   n5 = &RigVMFunction_AlphaInterp(
    InterpSpeedDecreasing=3
    InterpSpeedIncreasing=5
    OutRange=(Min=1,Max=0)
    Scale=1
    bClampResult=true
    bMapRange=true
    Value=&RigVMFunction_MathDoubleMul(B=RigVMVariableNode() {variable=crouchingLeftMove}.Value, A=RigVMVariableNode() {variable=SP_output_thigh_l}.Value).Result
   )
   n6 = &RigVMFunction_AlphaInterp(
    InterpSpeedDecreasing=10
    InterpSpeedIncreasing=5
    Scale=1
    bClampResult=False
    Value=&RigVMFunction_MathDoubleMul(A=@n5.Result, B=RigVMVariableNode() {variable=crouchingLeftMove}.Value).Result
   )
   > &RigUnit_ModifyTransforms(
    ItemToModify=((Item=(Type=Bone,Name="R_back_BASE_jacket"),Transform=(Rotation=(X=0,Y=0,Z=0,W=1),Translation=(X=0,Y=0,Z=0),Scale3D=(X=1,Y=1,Z=1))))
    Weight=@n6.Result
    ItemToModify.0.Transform.Rotation=&RigVMFunction_MathQuaternionFromEuler(Euler=(X=10,Y=-20,Z=-20)).Result
   )
   n7 = &RigVMFunction_AlphaInterp(
    InterpSpeedDecreasing=3
    InterpSpeedIncreasing=5
    Scale=1
    bClampResult=False
    Value=&RigVMFunction_MathDoubleMul(B=RigVMVariableNode() {variable=crouchingLeftMove}.Value, A=RigVMVariableNode() {variable=SP_output_thigh_l}.Value).Result
   )
   > &RigUnit_ModifyTransforms(
    ItemToModify=((Item=(Type=Bone,Name="R_back_BASE_jacket"),Transform=(Rotation=(X=0,Y=0,Z=0,W=1),Translation=(X=0,Y=0,Z=0),Scale3D=(X=1,Y=1,Z=1))))
    Weight=@n7.Result
    ItemToModify.0.Transform.Rotation=&RigVMFunction_MathQuaternionFromEuler(Euler=(X=-10,Y=10,Z=-10)).Result
   )
   n8 = RigVMVariableNode() {variable=crouchingRightMove}
   n9 = &RigVMFunction_AlphaInterp(
    InterpSpeedDecreasing=6
    InterpSpeedIncreasing=10
    OutRange=(Min=1,Max=0)
    Scale=1
    bClampResult=true
    bMapRange=true
    Value=&RigVMFunction_MathDoubleMul(B=@n8.Value, A=RigVMVariableNode() {variable=SP_output_thigh_l}.Value).Result
   )
   n10 = &RigVMFunction_AlphaInterp(
    InterpSpeedDecreasing=10
    InterpSpeedIncreasing=5
    Scale=1
    bClampResult=False
    Value=&RigVMFunction_MathDoubleMul(B=@n9.Result, A=@n8.Value).Result
   )
   > &RigUnit_ModifyTransforms(
    ItemToModify=((Item=(Type=Bone,Name="R_back_BASE_jacket"),Transform=(Rotation=(X=0,Y=0,Z=0,W=1),Translation=(X=0,Y=0,Z=0),Scale3D=(X=1,Y=1,Z=1))))
    Weight=@n10.Result
    ItemToModify.0.Transform.Rotation=&RigVMFunction_MathQuaternionFromEuler(Euler=(X=10,Y=-10,Z=-20)).Result
   )
   > &RigUnit_ModifyTransforms(
    ItemToModify=((Item=(Type=Bone,Name="L_back_BASE_jacket"),Transform=(Rotation=(X=0,Y=0,Z=0,W=1),Translation=(X=0,Y=0,Z=0),Scale3D=(X=1,Y=1,Z=1))))
    ItemToModify.0.Transform.Rotation=&RigVMFunction_MathQuaternionFromEuler(Euler=(X=7,Y=-7,Z=-12)).Result
    Weight=&RigVMFunction_MathDoubleMul(B=RigVMVariableNode() {variable=crouchingRightMove}.Value, A=RigVMVariableNode() {variable=SP_output_thigh_l}.Value).Result
   )
   n11 = &RigUnit_SphericalPoseReader(
    ActiveRegionScaleFactors=(PositiveWidth=1,NegativeWidth=1,PositiveHeight=1,NegativeHeight=1)
    ActiveRegionSize=0.1
    Debug=(bDrawDebug=false,bDraw2D=false,bDrawLocalAxes=false,DebugScale=25,DebugSegments=20,DebugThickness=0.25)
    DriverAxis=(X=-1,Y=5,Z=-1)
    DriverItem=(Type=Bone,Name="thigh_l")
    FalloffRegionScaleFactors=(PositiveWidth=0.176,NegativeWidth=0.184,PositiveHeight=0.216,NegativeHeight=0.104)
    FalloffSize=1
    FlipWidthScaling=False
    OptionalParentItem=(Type=Bone)
    RotationOffset=(X=60,Y=-40,Z=-5)
    ExecutePin=@SequenceAggregate_1_1_1.C
   )
   > &RigUnit_ModifyTransforms(
    ItemToModify=((Item=(Type=Bone,Name="dyn_L_front_BASE_jacket"),Transform=(Rotation=(X=0,Y=0,Z=0,W=1),Translation=(X=0,Y=2,Z=0),Scale3D=(X=1,Y=1,Z=1))))
    ItemToModify.0.Transform.Rotation=&RigVMFunction_MathQuaternionFromEuler(Euler=(X=20,Y=0,Z=0)).Result
    Weight=@n11.OutputParam
   )
   > &RigUnit_ModifyTransforms(
    ItemToModify=((Item=(Type=Bone,Name="dyn_R_frontShirt_BASE_jacket"),Transform=(Rotation=(X=0,Y=0,Z=0,W=1),Translation=(X=0,Y=5,Z=0),Scale3D=(X=1,Y=1,Z=1))))
    ItemToModify.0.Transform.Rotation=&RigVMFunction_MathQuaternionFromEuler(Euler=(X=-20,Y=0,Z=0)).Result
    Weight=@n11.OutputParam
   )
   n12 = &RigUnit_SphericalPoseReader(
    ActiveRegionScaleFactors=(PositiveWidth=1,NegativeWidth=1,PositiveHeight=1,NegativeHeight=1)
    ActiveRegionSize=0.1
    Debug=(bDrawDebug=false,bDraw2D=false,bDrawLocalAxes=false,DebugScale=25,DebugSegments=20,DebugThickness=0.25)
    DriverAxis=(X=-1,Y=0,Z=0)
    DriverItem=(Type=Bone,Name="thigh_l")
    FalloffRegionScaleFactors=(PositiveWidth=0.176,NegativeWidth=0.184,PositiveHeight=0.216,NegativeHeight=0.104)
    FalloffSize=1
    FlipWidthScaling=False
    OptionalParentItem=(Type=Bone)
    RotationOffset=(X=60,Y=-40,Z=-5)
    ExecutePin=@SequenceAggregate_1_1_1.D
   )
   n13 = &RigVMFunction_MathDoubleMul(A=@n12.OutputParam, B=RigVMVariableNode() {variable=static_crouching_rotation}.Value)
   > &RigUnit_ModifyTransforms(
    ItemToModify=((Item=(Type=Bone,Name="R_back_BASE_jacket"),Transform=(Rotation=(X=0,Y=0,Z=0,W=1),Translation=(X=0,Y=0,Z=0),Scale3D=(X=1,Y=1,Z=1))))
    ItemToModify.0.Transform.Rotation=&RigVMFunction_MathQuaternionFromEuler(Euler=(X=20,Y=-10,Z=-10)).Result
    Weight=@n13.Result
   )
   n14 = &RigVMFunction_AlphaInterp(InterpSpeedDecreasing=6, InterpSpeedIncreasing=10, OutRange=(Min=1,Max=0), Scale=1, bClampResult=true, bMapRange=true, Value=@n13.Result)
   n15 = &RigVMFunction_AlphaInterp(
    InterpSpeedDecreasing=10
    InterpSpeedIncreasing=5
    Scale=1
    bClampResult=False
    Value=&RigVMFunction_MathDoubleMul(B=@n14.Result, A=RigVMVariableNode() {variable=static_crouching_rotation}.Value).Result
   )
   > &RigUnit_ModifyTransforms(
    ItemToModify=((Item=(Type=Bone,Name="L_back_BASE_jacket"),Transform=(Rotation=(X=0,Y=0,Z=0,W=1),Translation=(X=0,Y=0,Z=0),Scale3D=(X=1,Y=1,Z=1))))
    Weight=@n15.Result
    ItemToModify.0.Transform.Rotation=&RigVMFunction_MathQuaternionFromEuler(Euler=(X=10,Y=-10,Z=-20)).Result
   )
   n16 = &RigUnit_SphericalPoseReader(
    ActiveRegionScaleFactors=(PositiveWidth=1,NegativeWidth=1,PositiveHeight=1,NegativeHeight=1)
    ActiveRegionSize=0.1
    Debug=(bDrawDebug=false,bDraw2D=false,bDrawLocalAxes=false,DebugScale=25,DebugSegments=20,DebugThickness=0.25)
    DriverAxis=(X=-1,Y=0.5,Z=0)
    DriverItem=(Type=Bone,Name="upperarm_r")
    FalloffRegionScaleFactors=(PositiveWidth=0.176,NegativeWidth=0.184,PositiveHeight=0.216,NegativeHeight=0.104)
    FalloffSize=1
    FlipWidthScaling=False
    OptionalParentItem=(Type=Bone)
    RotationOffset=(X=280,Y=-150,Z=30)
    ExecutePin=@SequenceAggregate_1_1_1.E
   )
   n17 = &RigVMFunction_MathDoubleMul(A=@n16.OutputParam, B=RigVMVariableNode() {variable=pistol}.Value)
   > &RigUnit_ModifyTransforms(
    ItemToModify=((Item=(Type=Bone,Name="extra_upperarm_r_01"),Transform=(Rotation=(X=0,Y=0,Z=0,W=1),Translation=(X=20.353,Y=19.421,Z=-50.469),Scale3D=(X=1,Y=1,Z=1))))
    Weight=@n17.Result
   )
   > &RigUnit_ModifyTransforms(
    ItemToModify=((Item=(Type=Bone,Name="extra_upperarm_r_02"),Transform=(Rotation=(X=0,Y=0,Z=0,W=1),Translation=(X=19.704,Y=51.669,Z=8.576),Scale3D=(X=1,Y=1,Z=1))))
    Weight=@n17.Result
   )
   &RigUnit_SphericalPoseReader(
    ActiveRegionScaleFactors=(PositiveWidth=0.7,NegativeWidth=0.5,PositiveHeight=0.5,NegativeHeight=0.5)
    ActiveRegionSize=0.1
    Debug=(bDrawDebug=true,DebugScale=25,DebugSegments=20,DebugThickness=0.25)
    DriverAxis=(X=1,Y=0,Z=0)
    DriverItem=(Type=Bone,Name="hand_l")
    FalloffRegionScaleFactors=(PositiveWidth=1,NegativeWidth=0.5,PositiveHeight=1,NegativeHeight=1)
    FalloffSize=0.2
    OptionalParentItem=(Type=Bone)
    RotationOffset=(X=105,Y=0,Z=-10)
    ExecutePin=@SequenceAggregate_1_1_1.F
   )
   > &RigUnit_ModifyTransforms(
    ItemToModify=((Item=(Type=Bone,Name="extra_watch_01"),Transform=(Rotation=(X=0,Y=0,Z=0,W=1),Translation=(X=-0.7,Y=0,Z=0),Scale3D=(X=1,Y=1,Z=1))))
    Weight=1
   )
   &RigUnit_ModifyTransforms(
    ExecutePin=()
    ItemToModify=((Item=(Type=Bone,Name="extra_upperarm_r_02"),Transform=(Rotation=(X=0,Y=0,Z=0,W=1),Translation=(X=19.704,Y=-51.669,Z=8.576),Scale3D=(X=1,Y=1,Z=1))))
    Weight=1
   )
   &RigUnit_ModifyTransforms(
    ExecutePin=()
    ItemToModify=((Item=(Type=Bone,Name="extra_upperarm_r_01"),Transform=(Rotation=(X=0,Y=0,Z=0,W=1),Translation=(X=20.353,Y=-19.421,Z=-50.469),Scale3D=(X=1,Y=1,Z=1))))
    Weight=1
   )
  ]}
]
hierarchy [
 root {loc=<0 0 0>, rot=<0 0 0 1>}
 pelvis {parent=root, loc=<0 -0.7071 93.1191>, rot=<0.0127 0.707 -0.0127 -0.707>}
 spine_01 {parent=pelvis, loc=<6.7494 -0.6388 0>, rot=<0 0 0.034 -0.9994>}
 spine_02 {parent=spine_01, loc=<7.2518 -0.035 0>, rot=<0 0 0.0238 -0.9997>}
 spine_03 {parent=spine_02, loc=<6.0104 0 0>, rot=<0 0 0.0149 -0.9999>}
 spine_04 {parent=spine_03, loc=<8.1024 0 0>, rot=<0 0 -0.0132 -0.9999>}
 spine_05 {parent=spine_04, loc=<10.0146 0 0>, rot=<0 0 -0.0015 -1>}
 neck_01 {parent=spine_05, loc=<11.3138 0.1529 0>, rot=<0 0 -0.2003 -0.9797>}
 neck_02 {parent=neck_01, loc=<4.5627 0 0>, rot=<0 0 0.0504 -0.9987>}
 head {parent=neck_02, loc=<5.3914 0 0>, rot=<0 0 0.115 -0.9934>}
 attach_cape {parent=spine_05, loc=<9.017 -11.714 0>, rot=<0.0616 -0.7044 -0.0616 -0.7044>}
 clavicle_l {parent=spine_05, loc=<4.3575 -0.1466 -2.3223>, rot=<0.0465 -0.7289 -0.009 -0.6829>}
 upperarm_l {parent=clavicle_l, loc=<15.5111 0 0>, rot=<0.0226 -0.3234 -0.0088 -0.9459>}
 lowerarm_l {parent=upperarm_l, loc=<26.6514 0 0>, rot=<0.0391 -0.0099 -0.2453 -0.9686>}
 hand_l {parent=lowerarm_l, loc=<22.3491 0 0>, rot=<-0.7081 -0.0507 0.0975 -0.6975>}
 thumb_01_l {parent=hand_l, loc=<1.3118 -0.3191 -2.459>, rot=<0.6546 -0.4042 0.0664 -0.6354>}
 thumb_02_l {parent=thumb_01_l, loc=<5.9336 -0.0448 0.0401>, rot=<-0.0037 0.0213 0.298 -0.9543>}
 thumb_03_l {parent=thumb_02_l, loc=<2.2223 0.3682 -0.0907>, rot=<0.0118 -0.013 0.0066 -0.9998>}
 index_metacarpal_l {parent=hand_l, loc=<1.6535 0.0183 -2.96>, rot=<0.0158 -0.0124 -0.0173 -0.9996>}
 index_01_l {parent=index_metacarpal_l, loc=<8.3008 0.0357 -0.0524>, rot=<-0.0044 0.0485 0.0924 -0.9945>}
 index_02_l {parent=index_01_l, loc=<3.3587 -0.3004 0.0452>, rot=<-0.002 0.0027 0.3225 -0.9466>}
 index_03_l {parent=index_02_l, loc=<2.5578 -0.123 -0.0252>, rot=<-0.002 0.0067 0.2558 -0.9667>}
 middle_metacarpal_l {parent=hand_l, loc=<2.1856 -0.1725 -0.5984>, rot=<0.006 0.0168 -0.0329 -0.9993>}
 middle_01_l {parent=middle_metacarpal_l, loc=<7.3275 -0.2447 0.0129>, rot=<-0.0208 0.0042 0.0858 -0.9961>}
 middle_02_l {parent=middle_01_l, loc=<4.0123 0.1401 0.1175>, rot=<-0.0218 0.0258 0.3512 -0.9357>}
 middle_03_l {parent=middle_02_l, loc=<2.6077 0.0534 0.1373>, rot=<0.0082 0.0188 0.1708 -0.9851>}
 ring_metacarpal_l {parent=hand_l, loc=<2.2074 -0.3625 1.3638>, rot=<-0.0284 0.0722 -0.0458 -0.9959>}
 ring_01_l {parent=ring_metacarpal_l, loc=<7.0295 -0.6197 -0.1446>, rot=<-0.0489 -0.0252 0.1892 -0.9804>}
 ring_02_l {parent=ring_01_l, loc=<3.6174 0.4647 -0.1027>, rot=<0 0 0.2806 -0.9598>}
 ring_03_l {parent=ring_02_l, loc=<2.4152 0.114 0.0199>, rot=<0 0 0.1189 -0.9929>}
 pinky_metacarpal_l {parent=hand_l, loc=<1.7616 -0.9753 3.3814>, rot=<-0.0397 0.0814 -0.069 -0.9935>}
 pinky_01_l {parent=pinky_metacarpal_l, loc=<7.1709 -0.4831 -0.1094>, rot=<-0.0654 -0.0035 0.2242 -0.9723>}
 pinky_02_l {parent=pinky_01_l, loc=<2.8274 0.0893 -0.1043>, rot=<0.0037 0.0011 0.2854 -0.9584>}
 pinky_03_l {parent=pinky_02_l, loc=<2.2046 0.0307 -0.0078>, rot=<-0.0002 0.0098 0.1445 -0.9895>}
 lowerarm_twist_02_l {parent=lowerarm_l, loc=<9.6199 0 0>, rot=<0 0 0 1>}
 lowerarm_twist_01_l {parent=lowerarm_l, loc=<17.8549 0 0>, rot=<0 0 0 1>}
 lowerarm_twist_01_l1 {parent=lowerarm_twist_01_l, loc=<3.5461 0 0>, rot=<0 0 0 1>}
 extra_watch_01 {parent=lowerarm_twist_01_l, loc=<3.5461 0 0>, rot=<0 0 0 1>}
 upperarm_twist_01_l {parent=upperarm_l, loc=<0.0041 0 0>, rot=<0 0 0 1>}
 upperarm_twist_02_l {parent=upperarm_l, loc=<13.7941 0 0>, rot=<0 0 0 1>}
 attach_backpack {parent=spine_05, loc=<-0.123 -13.473 0>, rot=<-0.0093 -0.707 0.0093 -0.707>}
 clavicle_r {parent=spine_05, loc=<4.3575 -0.1466 2.3223>, rot=<0.7289 0.0465 -0.6829 0.009>}
 upperarm_r {parent=clavicle_r, loc=<-15.5111 0 0>, rot=<0.0226 -0.3234 -0.0088 -0.9459>}
 lowerarm_r {parent=upperarm_r, loc=<-26.6514 0 0>, rot=<0.0391 -0.0099 -0.2453 -0.9686>}
 hand_r {parent=lowerarm_r, loc=<-22.3491 0 0>, rot=<-0.7081 -0.0507 0.0975 -0.6975>}
 thumb_01_r {parent=hand_r, loc=<-1.3118 0.3191 2.459>, rot=<0.6546 -0.4042 0.0664 -0.6354>}
 thumb_02_r {parent=thumb_01_r, loc=<-5.9336 0.0448 -0.0401>, rot=<-0.0037 0.0213 0.298 -0.9543>}
 thumb_03_r {parent=thumb_02_r, loc=<-2.2223 -0.3682 0.0907>, rot=<0.0118 -0.013 0.0066 -0.9998>}
 index_metacarpal_r {parent=hand_r, loc=<-1.6535 -0.0183 2.96>, rot=<0.0158 -0.0124 -0.0173 -0.9996>}
 index_01_r {parent=index_metacarpal_r, loc=<-8.3008 -0.0357 0.0524>, rot=<-0.0044 0.0485 0.0924 -0.9945>}
 index_02_r {parent=index_01_r, loc=<-3.3587 0.3004 -0.0452>, rot=<-0.002 0.0027 0.3225 -0.9466>}
 index_03_r {parent=index_02_r, loc=<-2.5578 0.123 0.0252>, rot=<-0.002 0.0067 0.2558 -0.9667>}
 middle_metacarpal_r {parent=hand_r, loc=<-2.1856 0.1725 0.5984>, rot=<0.006 0.0168 -0.0329 -0.9993>}
 middle_01_r {parent=middle_metacarpal_r, loc=<-7.3275 0.2447 -0.0129>, rot=<-0.0208 0.0042 0.0858 -0.9961>}
 middle_02_r {parent=middle_01_r, loc=<-4.0123 -0.1401 -0.1175>, rot=<-0.0218 0.0258 0.3512 -0.9357>}
 middle_03_r {parent=middle_02_r, loc=<-2.6077 -0.0534 -0.1373>, rot=<0.0082 0.0188 0.1708 -0.9851>}
 ring_metacarpal_r {parent=hand_r, loc=<-2.2074 0.3625 -1.3638>, rot=<-0.0284 0.0722 -0.0458 -0.9959>}
 ring_01_r {parent=ring_metacarpal_r, loc=<-7.0295 0.6197 0.1446>, rot=<-0.0489 -0.0252 0.1892 -0.9804>}
 ring_02_r {parent=ring_01_r, loc=<-3.6174 -0.4647 0.1027>, rot=<0 0 0.2806 -0.9598>}
 ring_03_r {parent=ring_02_r, loc=<-2.4152 -0.114 -0.0199>, rot=<0 0 0.1189 -0.9929>}
 pinky_metacarpal_r {parent=hand_r, loc=<-1.7616 0.9753 -3.3814>, rot=<-0.0397 0.0814 -0.069 -0.9935>}
 pinky_01_r {parent=pinky_metacarpal_r, loc=<-7.1709 0.4831 0.1094>, rot=<-0.0654 -0.0035 0.2242 -0.9723>}
 pinky_02_r {parent=pinky_01_r, loc=<-2.8274 -0.0893 0.1043>, rot=<0.0037 0.0011 0.2854 -0.9584>}
 pinky_03_r {parent=pinky_02_r, loc=<-2.2046 -0.0307 0.0078>, rot=<-0.0002 0.0098 0.1445 -0.9895>}
 lowerarm_twist_02_r {parent=lowerarm_r, loc=<-9.6199 0 0>, rot=<0 0 0 1>}
 lowerarm_twist_01_r {parent=lowerarm_r, loc=<-17.8549 0 0>, rot=<0 0 0 1>}
 upperarm_twist_01_r {parent=upperarm_r, loc=<-0.0041 0 0>, rot=<0 0 0 1>}
 upperarm_twist_02_r {parent=upperarm_r, loc=<-13.7941 0 0>, rot=<0 0 0 1>}
 extra_upperarm_r {parent=upperarm_r, loc=<0 0 0>, rot=<0 0 0 1>}
 extra_upperarm_r_01 {parent=upperarm_r, loc=<0 0 0>, rot=<0 0 0 1>}
 extra_upperarm_r_02 {parent=upperarm_r, loc=<0 0 0>, rot=<0 0 0 1>}
 dyn_tai_02 {parent=spine_04, loc=<7.0588 12.4304 0.7366>, rot=<-0.0166 -0.9992 -0.0037 -0.0374>}
 dyn_tai_03 {parent=dyn_tai_02, loc=<3.6943 0 0>, rot=<-0.0023 -0.018 -0.0123 -0.9998>}
 dyn_tai_04 {parent=dyn_tai_03, loc=<4.5093 0 0>, rot=<-0.0005 -0.0052 0.0355 -0.9994>}
 dyn_tai_05 {parent=dyn_tai_04, loc=<4.641 0 0>, rot=<-0.0013 -0.0196 0.0059 -0.9998>}
 dyn_tai_06 {parent=dyn_tai_05, loc=<4.8637 0 0>, rot=<0.001 0.0127 -0.0169 -0.9998>}
 dyn_tai_07 {parent=dyn_tai_06, loc=<5.1658 0 0>, rot=<-0.0012 -0.0204 0.0324 -0.9993>}
 dyn_tai_08 {parent=dyn_tai_07, loc=<5.2501 0 0>, rot=<0 0 0 1>}
 thigh_l {parent=pelvis, loc=<-2.5693 -0.0109 -7.7501>, rot=<0.0012 0.0572 0.0155 -0.9982>}
 calf_l {parent=thigh_l, loc=<-39.0484 0 0>, rot=<0.0006 -0.0161 -0.0391 -0.9991>}
 foot_l {parent=calf_l, loc=<-42.561 0.1077 0.4292>, rot=<0.0418 -0.0455 0.0343 -0.9975>}
 ball_l {parent=foot_l, loc=<-6.1599 13.8264 0.6806>, rot=<0.0001 -0.0001 -0.7063 -0.7079>}
 calf_twist_02_l {parent=calf_l, loc=<-18.8064 -0.0967 0.0397>, rot=<0.0237 0.0095 -0.0035 -0.9997>}
 calf_twist_01_l {parent=calf_l, loc=<-35.813 -0.0967 0.0397>, rot=<0.0237 0.0095 -0.0035 -0.9997>}
 thigh_twist_01_l {parent=thigh_l, loc=<-21.6303 0 0>, rot=<0.0013 0 0 -1>}
 dyn_chainFront_BASE {parent=thigh_twist_01_l, loc=<12.3076 8.2682 -4.1013>, rot=<0.1773 -0.8899 -0.418 -0.0435>}
 dyn_chain_00 {parent=dyn_chainFront_BASE, loc=<0 0 0>, rot=<0 0 0 1>}
 dyn_chain_01 {parent=dyn_chain_00, loc=<5.6349 0 0>, rot=<0 0 0 1>}
 dyn_chain_02 {parent=dyn_chain_01, loc=<5.059 0 0>, rot=<-0.1644 -0.1145 0.1537 -0.9676>}
 dyn_chain_03 {parent=dyn_chain_02, loc=<5.6076 0 0>, rot=<-0.0916 -0.2146 0.3803 -0.8949>}
 dyn_chain_04 {parent=dyn_chain_03, loc=<4.8024 0 0>, rot=<0.1039 -0.2272 0.4039 -0.88>}
 dyn_chain_05 {parent=dyn_chain_04, loc=<5.53 0 0>, rot=<0.2367 -0.1333 0.1903 -0.9434>}
 dyn_chain_06 {parent=dyn_chain_05, loc=<5.6753 0 0>, rot=<0.0758 -0.0221 0.0301 -0.9964>}
 dyn_chain_07 {parent=dyn_chain_06, loc=<6.1932 0 0>, rot=<0 0 0 1>}
 dyn_L_front_BASE_jacket {parent=thigh_l, loc=<-6.5834 9.5296 -1.5074>, rot=<-0.0395 -0.9954 0.0838 -0.0248>}
 dyn_L_front_01_jacket {parent=dyn_L_front_BASE_jacket, loc=<0 0 0>, rot=<0 0 0 1>}
 dyn_L_front_02_jacket {parent=dyn_L_front_01_jacket, loc=<0 0 0>, rot=<0 0 0 1>}
 dyn_L_front_01_jacket1 {parent=thigh_l, loc=<-6.8869 10.7269 5.4816>, rot=<-0.0395 -0.9954 0.0838 -0.0248>}
 dyn_L_front_01_jacket2 {parent=thigh_l, loc=<-6.5843 6.68 -5.2815>, rot=<-0.0462 -0.8724 0.4865 -0.0063>}
 thigh_r {parent=pelvis, loc=<-2.5693 -0.0109 7.7501>, rot=<0.0572 -0.0012 0.9982 0.0155>}
 calf_r {parent=thigh_r, loc=<39.0484 0 0>, rot=<0.0006 -0.0161 -0.0391 -0.9991>}
 foot_r {parent=calf_r, loc=<42.561 -0.1077 -0.4292>, rot=<0.0418 -0.0455 0.0343 -0.9975>}
 ball_r {parent=foot_r, loc=<6.1599 -13.8264 -0.6806>, rot=<0.0001 -0.0001 -0.7063 -0.7079>}
 calf_twist_02_r {parent=calf_r, loc=<18.8064 0.0967 -0.0397>, rot=<0.0237 0.0095 -0.0035 -0.9997>}
 calf_twist_01_r {parent=calf_r, loc=<35.813 0.0967 -0.0397>, rot=<0.0237 0.0095 -0.0035 -0.9997>}
 thigh_twist_01_r {parent=thigh_r, loc=<21.6303 0 0>, rot=<0.0013 0 0 -1>}
 dyn_R_frontShirt_BASE_jacket {parent=thigh_r, loc=<6.5834 -9.5296 1.5074>, rot=<0.9954 -0.0395 0.0248 0.0838>}
 dyn_R_frontShirt_01_jacket {parent=dyn_R_frontShirt_BASE_jacket, loc=<0 0 0>, rot=<0 0 0 1>}
 dyn_R_frontShirt_02_jacket {parent=dyn_R_frontShirt_01_jacket, loc=<0 0 0>, rot=<0 0 0 1>}
 dyn_R_frontShirt_01_jacket_1 {parent=thigh_r, loc=<6.8869 -10.7269 -5.4816>, rot=<0.9954 -0.0395 0.0248 0.0838>}
 dyn_R_frontShirt_01_jacket_2 {parent=thigh_r, loc=<6.5843 -6.68 5.2815>, rot=<0.8724 -0.0462 0.0063 0.4865>}
 R_back_BASE_jacket {parent=pelvis, loc=<-5.9219 -10.2508 9.7972>, rot=<0.0348 -0.995 -0.0929 0.0116>}
 R_back_jacket_01 {parent=R_back_BASE_jacket, loc=<0 0 0>, rot=<0 0 0 1>}
 M_back_BASE_jacket {parent=pelvis, loc=<-6.3358 -11.0398 0>, rot=<0.0316 -0.9995 0 0>}
 M_back_jacket_01 {parent=M_back_BASE_jacket, loc=<0 0 0>, rot=<0 0 0 1>}
 L_back_BASE_jacket {parent=pelvis, loc=<-5.9219 -10.2508 -9.7972>, rot=<0.0348 -0.995 0.0929 -0.0116>}
 L_back_jacket_01 {parent=L_back_BASE_jacket, loc=<0 0 0>, rot=<0 0 0 1>}
 extra_groin {parent=pelvis, loc=<0.2102 0 0>, rot=<0 0 0 1>, scale=<1 0.263 1>}
 extra_back {parent=pelvis, loc=<-0.2167 0 0>, rot=<0 -1 0 0>, scale=<-1 -0.3 -1>}
]
variables [
 SP_output_thigh_l {type=double, default=0}
 crouchingLeftMove {type=double, default=0}
 crouchingRightMove {type=double, default=0}
 leftMove {type=double, default=0}
 pistol {type=double, default=0}
 rightMove {type=double, default=0}
 static_crouching_rotation {type=double, default=0}
 targeting_behaviour {type=double, default=0}
 test {type=double, default=-1.44}
]
