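# Physics asset for the M_MED_Patches male character body, built around a dangling tie (dyn_tai bones) and the torso and legs around it.
# Three tie capsules form a chain from spine_04 through every few tie bones, with narrow swing and very tight twist limits.
# Larger capsules cover pelvis, three spine bones and both thighs, linked by constraints that leave the upper two spine joints free in swing and twist.
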PhysicsAsset /BRCosmetics/Characters/Player/Male/Medium/Bodies/M_MED_Patches/Meshes/M_MED_Patches_Physics_Tai
bodies {
 dyn_tai_02 [{
   shapeType=Capsule
   length=8.4991
   radius=1.7122
   shapeName=dyn_tie_01_joint_capsule
   center=<6.9237 0.1925 -0.0169>
   rotation=<rot -0.0019 -91.5575 -89.9004>
  }]
 dyn_tai_05 [{
   shapeType=Capsule
   length=5.4099
   radius=2.2764
   shapeName=dyn_tie_03_joint_capsule
   center=<5.4668 -0.0294 -0.0545>
   rotation=<rot 0.0104 -89.9999 -89.4573>
  }]
 dyn_tai_07 [{
   shapeType=Capsule
   length=7.8971
   radius=2.2634
   shapeName=dyn_tie_04_joint_capsule
   center=<5.8255 -0.0506 -0.0672>
   rotation=<rot 0 -90.7352 -89.2667>
  }]
 pelvis [{shapeType=Capsule, length=8.3439, radius=11.4084, rotation=<rot 0 -90 0>}]
 spine_01 [{shapeType=Capsule, length=11.466, radius=10.2711, center=<0 0.8012 0>, rotation=<rot -2.2203 -90 1.5431>}]
 spine_02 [{shapeType=Capsule, length=11.5547, radius=9.828, shapeName=spine_02_capsule, center=<4.3734 0.9243 -0.085>, rotation=<rot 0.6864 -90 1.3425>}]
 spine_04 [{
   shapeType=Capsule
   length=13.7346
   radius=10.6282
   shapeName=spine_04_capsule
   center=<-1.1401 0.5411 0.0364>
   rotation=<rot -0.4398 -90 0.5187>
  }]
 thigh_l [{shapeType=Capsule, length=23.1036, radius=8.7183, center=<-19.7521 -0.3002 0.8412>, rotation=<rot 3.1635 -87.4829 -91.464>}]
 thigh_r [{shapeType=Capsule, length=23.1036, radius=8.7183, center=<19.7521 0.3002 -0.8412>, rotation=<rot 3.1635 -87.4829 88.536>}]
}
constraints [
 {bone1Name=spine_01, bone2Name=pelvis, swing1LimitDegrees=45, swing2LimitDegrees=45, twistLimitDegrees=45}
 {bone1Name=thigh_l, bone2Name=pelvis, swing1LimitDegrees=45, swing2LimitDegrees=45, twistLimitDegrees=45}
 {bone1Name=thigh_r, bone2Name=pelvis, swing1LimitDegrees=45, swing2LimitDegrees=45, twistLimitDegrees=45}
 {
  bone1Name=spine_04
  bone2Name=spine_02
  swing1LimitDegrees=45
  swing1Motion=Free
  swing2LimitDegrees=45
  swing2Motion=Free
  twistLimitDegrees=45
  twistMotion=Free
 }
 {
  bone1Name=spine_02
  bone2Name=spine_01
  swing1LimitDegrees=45
  swing1Motion=Free
  swing2LimitDegrees=45
  swing2Motion=Free
  twistLimitDegrees=45
  twistMotion=Free
 }
 {bone1Name=dyn_tai_05, bone2Name=dyn_tai_02, swing1LimitDegrees=25, swing2LimitDegrees=20, twistLimitDegrees=5}
 {bone1Name=dyn_tai_07, bone2Name=dyn_tai_05, swing1LimitDegrees=25, swing2LimitDegrees=20, twistLimitDegrees=5}
 {bone1Name=dyn_tai_02, bone2Name=spine_04, swing1LimitDegrees=25, swing2LimitDegrees=10, twistLimitDegrees=5}
]
