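# Physics asset for the dangling front chain on the male medium Patches character body.
# Capsules on the dyn_chain bones form a serial constraint chain from dyn_chainFront_BASE, which pivots freely against a larger left-thigh capsule.
# Links have tight twist limits and moderate swing limits, which widen toward the tip so the end swings more loosely.
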PhysicsAsset /BRCosmetics/Characters/Player/Male/Medium/Bodies/M_MED_Patches/Meshes/M_MED_Patches_Physics_chain
bodies {
 dyn_chainFront_BASE [{shapeType=Capsule, length=0.505, radius=0.505, shapeName=dyn_chainFront_BASE_capsule, rotation=<rot 0 -90 0>}]
 dyn_chain_00 [{
   shapeType=Capsule
   length=3.5723
   radius=0.5
   shapeName=dyn_chain_00_capsule
   center=<2.2935 -0.127 -0.3293>
   rotation=<rot 0 95.6594 -84.1043>
  }]
 dyn_chain_01 [{shapeType=Capsule, length=2.8124, radius=0.5, shapeName=dyn_chain_01_capsule, center=<2.6 0.0436 0.0132>, rotation=<rot 0 90.0992 -89.077>}]
 dyn_chain_02 [{
   shapeType=Capsule
   length=2.9303
   radius=0.9937
   shapeName=dyn_chain_02_capsule
   center=<3.1686 0.3058 0.1205>
   rotation=<rot 0 -89.59 -88.8048>
  }]
 dyn_chain_03 [{
   shapeType=Capsule
   length=2.5294
   radius=1.2263
   shapeName=dyn_chain_03_capsule
   center=<2.6883 0.4809 0.0785>
   rotation=<rot 0 89.5481 -89.5032>
  }]
 dyn_chain_04 [{
   shapeType=Capsule
   length=2.3204
   radius=1.2834
   shapeName=dyn_chain_04_capsule
   center=<2.6748 0.2721 0.0838>
   rotation=<rot 0 -92.2186 -89.2308>
  }]
 dyn_chain_05 [{
   shapeType=Capsule
   length=2.5081
   radius=1.1154
   shapeName=dyn_chain_05_capsule
   center=<2.8047 0.1284 0.0024>
   rotation=<rot 0 89.1319 -89.2199>
  }]
 dyn_chain_06 [{
   shapeType=Capsule
   length=3.3754
   radius=0.8785
   shapeName=dyn_chain_06_capsule
   center=<3.3787 0.1158 -0.0038>
   rotation=<rot 0 90.8479 -89.4753>
  }]
 dyn_chain_07 [{shapeType=Capsule, length=0.505, radius=0.505, shapeName=dyn_chain_07_capsule, rotation=<rot 0 -90 0>}]
 thigh_l [{shapeType=Capsule, length=23.2635, radius=8.5584, center=<-19.751 -0.183 0.622>, rotation=<rot 3.1635 -87.4829 -91.6637>}]
}
constraints [
 {
  bone1Name=dyn_chainFront_BASE
  bone2Name=thigh_l
  swing1LimitDegrees=45
  swing1Motion=Free
  swing2LimitDegrees=45
  swing2Motion=Free
  twistLimitDegrees=45
  twistMotion=Free
 }
 {bone1Name=dyn_chain_00, bone2Name=dyn_chainFront_BASE, swing1LimitDegrees=45, swing2LimitDegrees=45, twistLimitDegrees=2}
 {bone1Name=dyn_chain_01, bone2Name=dyn_chain_00, swing1LimitDegrees=45, swing2LimitDegrees=45, twistLimitDegrees=2}
 {bone1Name=dyn_chain_02, bone2Name=dyn_chain_01, swing1LimitDegrees=45, swing2LimitDegrees=45, twistLimitDegrees=2}
 {bone1Name=dyn_chain_03, bone2Name=dyn_chain_02, swing1LimitDegrees=45, swing2LimitDegrees=45, twistLimitDegrees=2}
 {bone1Name=dyn_chain_04, bone2Name=dyn_chain_03, swing1LimitDegrees=45, swing2LimitDegrees=45, twistLimitDegrees=2}
 {bone1Name=dyn_chain_05, bone2Name=dyn_chain_04, swing1LimitDegrees=45, swing2LimitDegrees=45, twistLimitDegrees=2}
 {bone1Name=dyn_chain_06, bone2Name=dyn_chain_05, swing1LimitDegrees=100, swing2LimitDegrees=100, twistLimitDegrees=2}
 {bone1Name=dyn_chain_07, bone2Name=dyn_chain_06, swing1LimitDegrees=100, swing2LimitDegrees=100, twistLimitDegrees=20}
]
