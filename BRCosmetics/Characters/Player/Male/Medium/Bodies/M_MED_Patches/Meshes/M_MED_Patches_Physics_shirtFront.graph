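# Physics asset for the M_MED_Patches male player body, covering the front and back flaps of its shirt or jacket.
# Pelvis and thigh capsules form the body base, with three back flap chains hung from the pelvis and left and right front chains from the thighs.
# Tiny BASE bodies join with free angular motion, while end segments mostly use limited swing and very tight twist.
# Each front side also has a second 01 segment constrained directly to its thigh.
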PhysicsAsset /BRCosmetics/Characters/Player/Male/Medium/Bodies/M_MED_Patches/Meshes/M_MED_Patches_Physics_shirtFront
bodies {
 L_back_BASE_jacket [{shapeType=Capsule, length=0.505, radius=0.505, shapeName=L_back_BASE_jacket_capsule, rotation=<rot 0 -90 0>}]
 L_back_jacket_01 [{
   shapeType=Capsule
   length=7.6757
   radius=5.0373
   shapeName=L_back_jacket_01_capsule
   center=<9.7758 -0.6369 0.6235>
   rotation=<rot 0.0001 -159.9697 -39.163>
  }]
 M_back_BASE_jacket [{shapeType=Capsule, length=0.505, radius=0.505, shapeName=M_back_BASE_jacket_capsule, rotation=<rot 0 -90 0>}]
 M_back_jacket_01 [{shapeType=Capsule, length=8.0277, radius=5.8162, shapeName=M_back_jacket_01_capsule, center=<8.5 0 0>, rotation=<rot 0 -90 0>}]
 R_back_BASE_jacket [{shapeType=Capsule, length=0.505, radius=0.505, shapeName=R_back_BASE_jacket_capsule, rotation=<rot 0 -90 0>}]
 R_back_jacket_01 [{
   shapeType=Capsule
   length=7.1418
   radius=5.5712
   shapeName=L_back_jacket_01_capsule
   center=<8.8066 0.1314 -1.5572>
   rotation=<rot 0 -159.9701 37.229>
  }]
 dyn_L_front_01_jacket [{shapeType=Sphere, radius=2.4529, shapeName=dyn_L_front_01_jacket_sphere}]
 dyn_L_front_01_jacket2 [{shapeType=Capsule, length=7.7668, radius=2.4621, shapeName=dyn_L_front_01_jacket2_capsule, center=<8.711 0.8932 -0.0253>}]
 dyn_L_front_02_jacket [{shapeType=Capsule, length=6.9003, radius=3.4302, shapeName=dyn_L_front_02_jacket_capsule, center=<7.2635 1.9119 0>}]
 dyn_L_front_BASE_jacket [{shapeType=Capsule, length=0.505, radius=0.505, shapeName=dyn_L_front_BASE_jacket_capsule, rotation=<rot 0 -90 0>}]
 dyn_R_frontShirt_01_jacket [{shapeType=Capsule, length=0.6574, radius=0.7516, shapeName=dyn_R_frontShirt_01_jacket_capsule, rotation=<rot 7.3289 -90 14.8826>}]
 dyn_R_frontShirt_01_jacket_2 [{shapeType=Capsule, length=7.6324, radius=2.7741, shapeName=dyn_R_frontShirt_01_jacket_2_capsule, center=<8.3647 0.8094 0>}]
 dyn_R_frontShirt_02_jacket [{shapeType=Capsule, length=7.8952, radius=3.1521, shapeName=dyn_R_frontShirt_02_jacket_capsule, center=<8.4908 0 0>}]
 dyn_R_frontShirt_BASE_jacket [{shapeType=Capsule, length=0.505, radius=0.505, shapeName=dyn_R_frontShirt_BASE_jacket_capsule, rotation=<rot 0 -90 0>}]
 pelvis [{shapeType=Capsule, length=8.3439, radius=11.4084, rotation=<rot 0 -90 0>}]
 thigh_l [{shapeType=Capsule, length=23.1036, radius=8.7183, center=<-19.7521 -0.3002 0.8412>, rotation=<rot 3.1635 -87.4829 -91.464>}]
 thigh_r [{shapeType=Capsule, length=23.1036, radius=8.7183, center=<19.7521 0.3002 -0.8412>, rotation=<rot 3.1635 -87.4829 88.536>}]
}
constraints [
 {bone1Name=thigh_l, bone2Name=pelvis, swing1LimitDegrees=45, swing2LimitDegrees=45, twistLimitDegrees=45}
 {
  bone1Name=L_back_BASE_jacket
  bone2Name=pelvis
  swing1LimitDegrees=45
  swing1Motion=Free
  swing2LimitDegrees=45
  swing2Motion=Free
  twistLimitDegrees=45
  twistMotion=Free
 }
 {bone1Name=L_back_jacket_01, bone2Name=L_back_BASE_jacket, swing1LimitDegrees=60, swing2LimitDegrees=25, twistLimitDegrees=3}
 {
  bone1Name=M_back_BASE_jacket
  bone2Name=pelvis
  swing1LimitDegrees=45
  swing1Motion=Free
  swing2LimitDegrees=45
  swing2Motion=Free
  twistLimitDegrees=45
  twistMotion=Free
 }
 {bone1Name=M_back_jacket_01, bone2Name=M_back_BASE_jacket, swing1LimitDegrees=60, swing2LimitDegrees=25, twistLimitDegrees=3}
 {bone1Name=thigh_r, bone2Name=pelvis, swing1LimitDegrees=45, swing2LimitDegrees=45, twistLimitDegrees=45}
 {
  bone1Name=dyn_L_front_BASE_jacket
  bone2Name=thigh_l
  swing1LimitDegrees=45
  swing1Motion=Free
  swing2LimitDegrees=5
  swing2Motion=Free
  twistLimitDegrees=2
  twistMotion=Free
 }
 {
  bone1Name=dyn_L_front_01_jacket
  bone2Name=dyn_L_front_BASE_jacket
  swing1LimitDegrees=56.52
  swing1Motion=Free
  swing2LimitDegrees=5
  swing2Motion=Free
  twistLimitDegrees=2
  twistMotion=Free
 }
 {bone1Name=dyn_L_front_02_jacket, bone2Name=dyn_L_front_01_jacket, swing1LimitDegrees=30, swing2LimitDegrees=5, twistLimitDegrees=2}
 {bone1Name=dyn_L_front_01_jacket2, bone2Name=thigh_l, swing1LimitDegrees=30, swing2LimitDegrees=5, twistLimitDegrees=2}
 {bone1Name=dyn_R_frontShirt_01_jacket_2, bone2Name=thigh_r, swing1LimitDegrees=60, swing2LimitDegrees=5, twistLimitDegrees=2}
 {
  bone1Name=dyn_R_frontShirt_BASE_jacket
  bone2Name=thigh_r
  swing1LimitDegrees=45
  swing1Motion=Free
  swing2LimitDegrees=45
  swing2Motion=Free
  twistLimitDegrees=45
  twistMotion=Free
 }
 {
  bone1Name=dyn_R_frontShirt_01_jacket
  bone2Name=dyn_R_frontShirt_BASE_jacket
  swing1LimitDegrees=45
  swing1Motion=Free
  swing2LimitDegrees=45
  swing2Motion=Free
  twistLimitDegrees=45
  twistMotion=Free
 }
 {bone1Name=dyn_R_frontShirt_02_jacket, bone2Name=dyn_R_frontShirt_01_jacket, swing1LimitDegrees=60, swing2LimitDegrees=5, twistLimitDegrees=2}
 {
  bone1Name=R_back_BASE_jacket
  bone2Name=pelvis
  swing1LimitDegrees=45
  swing1Motion=Free
  swing2LimitDegrees=45
  swing2Motion=Free
  twistLimitDegrees=45
  twistMotion=Free
 }
 {bone1Name=R_back_jacket_01, bone2Name=R_back_BASE_jacket, swing1LimitDegrees=60, swing2LimitDegrees=25, twistLimitDegrees=3}
]
