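# Post-process style cosmetic AnimBlueprint for the Patches player body, adding secondary motion to its jacket, shirt and chain.
# It copies the parent mesh pose, then layers bone constraints, state-driven additive offsets and RigidBody simulations before a final Control Rig.
# Chain physics swaps its base bone from pelvis to left thigh when crouching.
# EventGraph derives the rig's strafe, targeting, crouch-turn and pistol inputs; front_shirt_alpha and backShirt_in_sky go unused in the AnimGraph.
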AnimBlueprint /BRCosmetics/Characters/Player/Male/Medium/Bodies/M_MED_Patches/Meshes/M_MED_Patches_AnimBP
target_skeleton=&/M_MED_Patches_Skeleton
graphs [
 AnimGraph {nodes [
   n1 = "Constraint - dyn_chain_07 (PPP)"!(
    ComponentPose=Local To Component(LocalPose=Copy Pose From Mesh() {bUseAttachedParent=true})
    bAlphaBoolEnabled:bool=NOT Boolean(A=Get static_crouching() -> (static_crouching:bool))
    ConstraintWeights_0=0.4933
    ConstraintWeights_1:float=0.4763
    ConstraintWeights_2:float=0.0304
   ) -> (Pose) {
    alphaInputType=Bool
    alphaBoolBlend {blendInTime=0.2, blendOption=Linear, blendOutTime=0.2}
    boneToModify {boneName=dyn_chain_07}
    constraintSetup [{targetBone {boneName=thigh_twist_01_l}}, {targetBone {boneName=pelvis}}, {targetBone {boneName=thigh_l}}]
    constraintWeights [1,1,1]
   }
   n2 = Get bIsCrouching()
   n3 = "Transform (Modify) Bone - Bone: dyn_L_front_BASE_jacket"!(ComponentPose=@n1, bAlphaBoolEnabled:bool=@n2, Translation="-0.647,3.573,2.239", Rotation="6.704,19.939,-1.277", Scale="1,1,1") -> (Pose) {
    alphaInputType=Bool
    rotationMode=BMM_Additive
    rotationSpace=BCS_ParentBoneSpace
    translationMode=BMM_Additive
    translationSpace=BCS_ParentBoneSpace
    alphaBoolBlend {blendInTime=0.24, blendOption=Linear, blendOutTime=0.24}
    boneToModify {boneName=dyn_L_front_BASE_jacket}
    rotation=<rot 6.704 19.939 -1.277>
    translation=<-0.647 3.573 2.239>
   }
   n4 = "Transform (Modify) Bone - Bone: dyn_chainFront_BASE"!(ComponentPose=@n3, bAlphaBoolEnabled:bool=@n2, Translation="0.547,-2.084,0.201", Rotation="1.159,-9.294,0.662", Scale="1,1,1") -> (Pose) {
    alphaInputType=Bool
    rotationMode=BMM_Additive
    rotationSpace=BCS_ParentBoneSpace
    translationMode=BMM_Additive
    translationSpace=BCS_ParentBoneSpace
    alphaBoolBlend {blendInTime=1, blendOption=Linear, blendOutTime=1}
    boneToModify {boneName=dyn_chainFront_BASE}
    rotation=<rot 1.159 -9.294 0.662>
    translation=<0.547 -2.084 0.201>
   }
   n5 = "Transform (Modify) Bone - Bone: dyn_chain_07"!(ComponentPose=@n4, bAlphaBoolEnabled:bool=Get static_crouching(), Translation="2,-6,-4", Rotation="0,0,0", Scale="1,1,1") -> (Pose) {
    alphaInputType=Bool
    rotationMode=BMM_Additive
    rotationSpace=BCS_ParentBoneSpace
    translationMode=BMM_Additive
    translationSpace=BCS_ParentBoneSpace
    alphaBoolBlend {blendInTime=0.2, blendOption=Linear, blendOutTime=0.2}
    boneToModify {boneName=dyn_chain_07}
    translation=<2 -6 -4>
   }
   n6 = RigidBody(ComponentPose=@n5) {
    bClampLinearTranslationLimitToRefPose=true
    bOverrideWorldGravity=true
    overridePhysicsAsset=&/M_MED_Patches_Physics_shirtFront
    simulationSpace=BaseBoneSpace
    alphaBoolBlend {blendOption=Linear}
    alphaScaleBiasClamp {bInterpResult=true}
    baseBoneRef {boneName=pelvis}
    componentAppliedLinearAccClamp=<10 10 10>
    overrideWorldGravity {z=-980}
   }
   n7 = RigidBody(ComponentPose=@n6) {
    bClampLinearTranslationLimitToRefPose=true
    bOverrideWorldGravity=true
    overridePhysicsAsset=&/M_MED_Patches_Physics_Tai
    simulationSpace=BaseBoneSpace
    alphaBoolBlend {blendOption=Linear}
    baseBoneRef {boneName=spine_03}
    componentAppliedLinearAccClamp=<2000 2000 2000>
    componentLinearAccScale=<0.3 0.3 0.3>
    componentLinearVelScale=<0.3 0.3 0.3>
    overrideWorldGravity {z=-980}
   }
   n8 = Get bIsCrouching()
   n9 = RigidBody(ComponentPose=@n7, Alpha="To Float (Boolean)"(InBool=NOT Boolean(A=@n8))) {
    bClampLinearTranslationLimitToRefPose=true
    bOverrideWorldGravity=true
    overridePhysicsAsset=&/M_MED_Patches_Physics_chain
    simulationSpace=BaseBoneSpace
    alphaBoolBlend {blendInTime=0.2, blendOption=Linear, blendOutTime=0.2}
    alphaScaleBiasClamp {bClampResult=true, bInterpResult=true, clampMax=0.7, outRange {max=10}}
    baseBoneRef {boneName=pelvis}
    componentAppliedLinearAccClamp=<10 10 10>
    overrideWorldGravity {z=-980}
   }
   n10 = RigidBody!(ComponentPose=@n9, bAlphaBoolEnabled:bool=Get bIsCrouching(), ExternalForce) -> (Pose) {
    alphaInputType=Bool
    bClampLinearTranslationLimitToRefPose=true
    bOverrideWorldGravity=true
    overridePhysicsAsset=&M_MED_Patches_Physics_chain
    simulationSpace=BaseBoneSpace
    alphaBoolBlend {blendInTime=0.2, blendOption=Linear, blendOutTime=0.2}
    baseBoneRef {boneName=thigh_l}
    componentAppliedLinearAccClamp=<10 10 10>
    overrideWorldGravity {z=-980}
   }
   n11 = "Constraint - dyn_chain_07 (PPP)"!(ComponentPose=@n10, Alpha:float=1, ConstraintWeights_0=0.4933, ConstraintWeights_1=0.4763, ConstraintWeights_2=0.0304) -> (Pose) {
    alphaBoolBlend {blendOption=Linear}
    boneToModify {boneName=dyn_chain_07}
    constraintSetup [{targetBone {boneName=thigh_twist_01_l}}, {targetBone {boneName=pelvis}}, {targetBone {boneName=thigh_l}}]
    constraintWeights [1,1,1]
   }
   n12 = "Constraint - dyn_chain_06 (PPP)"!(ComponentPose=@n11, Alpha=0, ConstraintWeights_0=0.6078, ConstraintWeights_1:float=0.2649, ConstraintWeights_2:float=0.1273) -> (Pose) {
    alphaBoolBlend {blendOption=Linear}
    boneToModify {boneName=dyn_chain_06}
    constraintSetup [{targetBone {boneName=thigh_twist_01_l}}, {targetBone {boneName=pelvis}}, {targetBone {boneName=thigh_l}}]
    constraintWeights [1,1,1]
   }
   n13 = "Transform (Modify) Bone - Bone: dyn_R_frontShirt_BASE_jacket"!(ComponentPose=@n12, bAlphaBoolEnabled:bool=Get bIsFalling(), Translation="0,0,0", Rotation="14.275,-22.775,-3.923", Scale="1,1,1") -> (Pose) {
    alphaInputType=Bool
    rotationMode=BMM_Additive
    rotationSpace=BCS_ParentBoneSpace
    translationMode=BMM_Additive
    translationSpace=BCS_ParentBoneSpace
    alphaBoolBlend {blendInTime=0.24, blendOption=Linear, blendOutTime=0.24}
    boneToModify {boneName=dyn_R_frontShirt_BASE_jacket}
    rotation=<rot 14.275 -22.775 -3.923>
   }
   n14 = "Transform (Modify) Bone - Bone: dyn_R_frontShirt_01_jacket_1"!(ComponentPose=@n13, bAlphaBoolEnabled:bool=Get bIsCrouching(), Translation="0,-3,0", Rotation="0,0,0", Scale="1,1,1") -> (Pose) {
    alphaInputType=Bool
    rotationMode=BMM_Additive
    rotationSpace=BCS_ParentBoneSpace
    translationMode=BMM_Additive
    translationSpace=BCS_ParentBoneSpace
    alphaBoolBlend {blendInTime=0.24, blendOption=Linear, blendOutTime=0.24}
    boneToModify {boneName=dyn_R_frontShirt_01_jacket_1}
    translation {y=-3}
   }
   n15 = "Transform (Modify) Bone - Bone: dyn_chain_04"!(ComponentPose=@n14, bAlphaBoolEnabled:bool=Get bIsCrouchSprinting(), Translation="-3,1,0", Rotation="0,0,0", Scale="1,1,1") -> (Pose) {
    alphaInputType=Bool
    rotationMode=BMM_Additive
    rotationSpace=BCS_ParentBoneSpace
    translationMode=BMM_Additive
    translationSpace=BCS_ParentBoneSpace
    alphaBoolBlend {blendInTime=0.2, blendOption=Linear, blendOutTime=0.2}
    boneToModify {boneName=dyn_chain_04}
    translation=<-3 1>
   }
   n16 = Get static_crouching()
   n17 = "Transform (Modify) Bone - Bone: dyn_L_front_BASE_jacket"(ComponentPose=@n15, bAlphaBoolEnabled=@n16, Translation="0,0,0", Rotation="0,10,0", Scale="1,1,1") {
    alphaInputType=Bool
    rotationMode=BMM_Additive
    rotationSpace=BCS_ParentBoneSpace
    translationMode=BMM_Additive
    translationSpace=BCS_ParentBoneSpace
    alphaBoolBlend {blendInTime=0.2, blendOption=Linear, blendOutTime=0.2}
    boneToModify {boneName=dyn_L_front_BASE_jacket}
    rotation {yaw=10}
   }
   n18 = "Transform (Modify) Bone - Bone: dyn_L_front_01_jacket2"!(ComponentPose=@n17, bAlphaBoolEnabled:bool=@n16, Translation="0,0,0", Rotation="0,12,0", Scale="1,1,1") -> (Pose) {
    alphaInputType=Bool
    rotationMode=BMM_Additive
    rotationSpace=BCS_ParentBoneSpace
    translationMode=BMM_Additive
    translationSpace=BCS_ParentBoneSpace
    alphaBoolBlend {blendInTime=0.2, blendOption=Linear, blendOutTime=0.2}
    boneToModify {boneName=dyn_L_front_01_jacket2}
    rotation {yaw=12}
   }
   n19 = "Transform (Modify) Bone - Bone: M_back_BASE_jacket"!(ComponentPose=@n18, bAlphaBoolEnabled:bool=Get bIsSkyDiving(), Translation="0,0,0", Rotation="-3.373,25.994,0.807", Scale="1,1,1") -> (Pose) {
    alphaInputType=Bool
    rotationMode=BMM_Additive
    rotationSpace=BCS_ParentBoneSpace
    translationMode=BMM_Additive
    translationSpace=BCS_ParentBoneSpace
    alphaBoolBlend {blendInTime=0.2, blendOption=Linear, blendOutTime=0.2}
    boneToModify {boneName=M_back_BASE_jacket}
    rotation=<rot -3.373 25.994 0.807>
   }
   n20 = Control Rig!(
    Source=Component To Local(ComponentPose=@n19)
    Alpha
    leftMove:double=Get Left_move() -> (Left_move:double)
    rightMove:double=Get Right_move() -> (Right_move:double)
    crouchingLeftMove:double=Get crouchingLeft_move() -> (crouchingLeft_move:double)
    crouchingRightMove:double=Get crouchingRight_move() -> (crouchingRight_move:double)
    targeting_behaviour:double=Get is_targeting_alpha() -> (is_targeting_alpha:double)
    static_crouching_rotation:double=Get static_crouching_rotation() -> (static_crouching_rotation:double)
    pistol:double=Get pistol() -> (pistol:double)
   ) -> (Pose) {
    alphaBoolBlend {blendOption=Linear}
    controlRigAssetReference {blueprintRigClass=&/M_MED_Patches_CtrlRig.M_MED_Patches_CtrlRig_C}
   }
   Output Pose(Result=@n20) {layerGroup=None}
  ]}
 EventGraph {nodes [
   Event BlueprintUpdateAnimation()
   > n1 = Sequence!(execute) -> (then_0, then_1, then_2:Exec, then_3:Exec, then_4:Exec, then_5:Exec, then_6:Exec, then_7:Exec)
   > n2 = Branch(Condition=Get bIsSprinting())
   > Set front_shirt_alpha(front_shirt_alpha:double=0.27) -> (Output_Get:double)
   Set front_shirt_alpha(execute=@n2.else, front_shirt_alpha=1)
   K2Node_Composite_2 = Left/Right move(execute=@n1.then_1) {is=title K2Node_Composite}
   n3 = Branch(execute=@n1.then_2, Condition=Get bIsFalling())
   > Set backShirt_in_sky(backShirt_in_sky:bool=false) -> (Output_Get:bool)
   K2Node_Composite_1 = condition_logic!() -> (ReturnValue:bool=~) {is=title K2Node_Composite}
   Set backShirt_in_sky(execute=@n3.else, backShirt_in_sky=@K2Node_Composite_1)
   n4 = NOT Boolean(A=Get bIsCrouching())
   n5 = Get bIsTargeting()
   Set is_targeting_alpha(execute=@n1.then_3, is_targeting_alpha:double="To Float (Boolean)"(InBool=AND Boolean(A=@n5, B=@n4))) -> (Output_Get:double)
   n6 = Get bIsMoving2D()
   > Set static_crouching(static_crouching:bool=AND Boolean(A=Get bIsCrouching(), B=NOT Boolean(A=@n6))) -> (Output_Get:bool)
   K2Node_Composite_0 = static_crouching_rotation_!() -> (ReturnValue:float=~) {is=title K2Node_Composite}
   Set static_crouching_rotation(execute=@n1.then_4, static_crouching_rotation:double=@K2Node_Composite_0) -> (Output_Get:double)
   K2Node_Composite_3 = "check if it's pistol "!() -> (ReturnValue:bool=~) {is=title K2Node_Composite}
   > Set pistol(pistol:double="To Float (Boolean)"(InBool=@K2Node_Composite_3)) -> (Output_Get:double)
  ]}
 static_crouching_rotation_ {
  path=EventGraph.K2Node_Composite_0.static_crouching_rotation_
  nodes [
   Inputs!()
   n1 = NOT Boolean(A=Get bIsMoving2D())
   n2 = Get PawnDeltaRotation() -> (PawnDeltaRotation:Rotator)
   n3 = "float > float"(A="Absolute (Float)"(A=BreakRotator(InRot=@n2).Yaw), B=2)
   n4 = Get bIsCrouching()
   Outputs!(ReturnValue:float="To Float (Boolean)"(InBool=AND Boolean!(A=@n3, B=@n4, C:bool=@n1) -> (ReturnValue)))
  ]
 }
 condition_logic {
  path=EventGraph.K2Node_Composite_1.condition_logic
  nodes [
   Inputs()
   n1 = NOT Boolean(A=Get bIsSkyDiving())
   n2 = NOT Boolean(A=Get bIsParachuteOpened())
   n3 = Get bIsMoving2D()
   n4 = NOT Boolean(A=Get bIsSkyDiving())
   n5 = Get bIsMoving2D()
   Outputs(ReturnValue:bool=OR Boolean!(A=AND Boolean!(A=@n5, B=@n4) -> (ReturnValue), B=AND Boolean!(A=@n3, B=@n2, C:bool=@n1) -> (ReturnValue), C:bool) -> (ReturnValue))
   Get bIsFalling()
  ]
 }
 Left/Right move {
  path=EventGraph.K2Node_Composite_2.Left/Right move
  nodes [
   Inputs!() -> (execute:Exec)
   n1 = Get bIsMoving2D()
   > n2 = Branch(Condition=@n1)
   n3 = Enum to String!(Enumerator:EFortCardinalDirection Enum=GetCardinalDirectionFromAngle!(
     self:Fort Anim Instance Object Reference
     Angle:float=Get PawnMovementDirectionAzimuth() -> (PawnMovementDirectionAzimuth:float)
     DeadZoneAngle:float=0
    ) -> (ReturnValue:EFortCardinalDirection Enum=North)) -> (ReturnValue:String)
   n4 = "To Float (Boolean)"(InBool="Equal Exactly (String)"(A=@n3, B=West))
   > Set Left_move(Left_move:double=@n4) -> (Output_Get:double)
   n5 = "To Float (Boolean)"(InBool="Equal Exactly (String)"(A=@n3, B=East))
   > Set Right_move(Right_move:double=@n5) -> (Output_Get:double)
   > n6 = Branch(Condition=Get bIsCrouching())
   > Set crouchingRight_move(crouchingRight_move:double=@n5) -> (Output_Get:double)
   > Set crouchingLeft_move(crouchingLeft_move:double=@n4) -> (Output_Get:double)
   > Set Left_move(Left_move=0)
   > Set Right_move(Right_move=0)
   Set crouchingRight_move(execute=@n6.else, crouchingRight_move=0)
   > Set crouchingLeft_move(crouchingLeft_move=0)
   Set Left_move(execute=@n2.else, Left_move=0)
   > Set Right_move(Right_move=0)
   > Set crouchingRight_move(crouchingRight_move=0)
   > Set crouchingLeft_move(crouchingLeft_move=0)
   AND Boolean!(A=Get bIsCrouching(), B=@n1) -> (ReturnValue)
   Outputs!()
  ]
 }
 "check if it's pistol " {
  path="EventGraph.K2Node_Composite_3.check if it's pistol "
  nodes [
   Inputs!()
   n1 = GetOwnerAnimationInterface!(self:Custom Character Part Anim Instance Object Reference) -> (ReturnValue:Fort Player Animation Interface Object Reference)
   Outputs!(ReturnValue:bool="Equal (Enum)"(
     A:EFortWeaponCoreAnimation Enum=GetWeaponCoreAnim!(self:Fort Player Animation Interface Object Reference=@n1) -> (ReturnValue:EFortWeaponCoreAnimation Enum)
     B:EFortWeaponCoreAnimation Enum=Pistol
    ))
  ]
 }
]
variables [
 backShirt_in_sky {type=boolean}
 crouchingLeft_move {type=number}
 crouchingRight_move {type=number}
 front_shirt_alpha {type=number}
 gravityBackShirt {type=Vector}
 is_targeting_alpha {type=number}
 Left_move {type=number}
 pistol {type=number}
 Right_move {type=number}
 static_crouching {type=boolean}
 static_crouching_rotation {type=number}
]
