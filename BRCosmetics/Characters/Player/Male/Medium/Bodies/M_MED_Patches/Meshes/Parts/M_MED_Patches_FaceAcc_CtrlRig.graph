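# Control Rig for the Patches character's face accessory mesh that applies pose-driven offsets to hair, ear and helper bones.
# Sequenced spherical pose readers on the head weight transform offsets to front and back hair bases and extra spine, neck and clavicle bones.
# Interpolated crouch, targeting and pistol variables scale some of those weights.
# A final branch blends a front-hair shift by melee state and always rotates the ear joint.
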ControlRigBlueprint /BRCosmetics/Characters/Player/Male/Medium/Bodies/M_MED_Patches/Meshes/Parts/M_MED_Patches_FaceAcc_CtrlRig
graphs [
 ContainedGraph {nodes [
   n1 = RigVMFunctionEntryNode()
   SequenceAggregate = &RigVMFunction_Sequence(ExecuteContext=@n1.ExecuteContext)
   SequenceAggregate_1 = &RigVMFunction_Sequence(ExecuteContext=@SequenceAggregate.B)
   SequenceAggregate_2 = &RigVMFunction_Sequence(ExecuteContext=@SequenceAggregate_1.B)
   SequenceAggregate_3 = &RigVMFunction_Sequence(ExecuteContext=@SequenceAggregate_2.B)
   SequenceAggregate_4 = &RigVMFunction_Sequence(ExecuteContext=@SequenceAggregate_3.B)
   SequenceAggregate_5 = &RigVMFunction_Sequence(ExecuteContext=@SequenceAggregate_4.B)
   RigVMFunctionReturnNode(
    A=@SequenceAggregate.A
    B=@SequenceAggregate_1.A
    C=@SequenceAggregate_2.A
    D=@SequenceAggregate_3.A
    E=@SequenceAggregate_4.A
    F=@SequenceAggregate_5.A
    G=@SequenceAggregate_5.B
   )
  ]}
 RigVMFunctionLibrary {links [], nodes []}
 RigVMModel {nodes [
   RigUnit_BeginExecution = &RigUnit_BeginExecution()
   SequenceAggregate_1 = RigVMAggregateNode(ExecuteContext=@RigUnit_BeginExecution.ExecutePin)
   n1 = &RigUnit_SphericalPoseReader(
    ActiveRegionScaleFactors=(PositiveWidth=1,NegativeWidth=0.5,PositiveHeight=0.75,NegativeHeight=0.75)
    ActiveRegionSize=0.1
    Debug=(bDrawDebug=false,DebugScale=25,DebugSegments=20,DebugThickness=0.25)
    DriverAxis=(X=0,Y=1,Z=0)
    DriverItem=(Type=Bone,Name="head")
    FalloffRegionScaleFactors=(PositiveWidth=1,NegativeWidth=0.5,PositiveHeight=1,NegativeHeight=1)
    FalloffSize=0.2
    OptionalParentItem=(Type=Bone)
    RotationOffset=(X=90,Y=0,Z=-30)
    ExecutePin=@SequenceAggregate_1.A
   )
   > RigVMVariableNode(Value=@n1.OutputParam) {variable=head_up_SP_result}
   > &RigUnit_ModifyTransforms(
    ItemToModify=((Item=(Type=Bone,Name="dyn_hair_back_r_BASE"),Transform=(Rotation=(X=0,Y=0,Z=0,W=1),Translation=(X=0,Y=5,Z=0),Scale3D=(X=1,Y=1,Z=1))))
    Weight=0
   )
   AlphaInterp = &RigVMFunction_AlphaInterp(
    InterpSpeedDecreasing=3
    InterpSpeedIncreasing=5
    Scale=1
    bClampResult=False
    Value=RigVMVariableNode() {variable=notCrouchingTargeting}.Value
   )
   n2 = &RigVMFunction_MathDoubleMul(B=@AlphaInterp.Result, A=@n1.OutputParam)
   > &RigUnit_ModifyTransforms(
    ItemToModify=((Item=(Type=Bone,Name="extra_hair_spine_05"),Transform=(Rotation=(X=0,Y=0,Z=0,W=1),Translation=(X=0,Y=0,Z=0),Scale3D=(X=1,Y=1,Z=1))))
    ItemToModify.0.Transform.Rotation=&RigVMFunction_MathQuaternionFromEuler(Euler=(X=-15,Y=0,Z=0)).Result
    Weight=@n2.Result
   )
   > &RigUnit_ModifyTransforms(
    ItemToModify=((Item=(Type=Bone,Name="extra_hair_spine_04"),Transform=(Rotation=(X=0,Y=0,Z=0,W=1),Translation=(X=0,Y=3,Z=0),Scale3D=(X=1,Y=1,Z=1))))
    Weight=@n2.Result
   )
   n3 = &RigVMFunction_AlphaInterp(InterpSpeedDecreasing=3, InterpSpeedIncreasing=5, Scale=1, bClampResult=False, Value=RigVMVariableNode() {variable=not_pistol}.Value)
   > &RigUnit_ModifyTransforms(
    ItemToModify=((Item=(Type=Bone,Name="extra_clavicle_l"),Transform=(Rotation=(X=0,Y=0,Z=0,W=1),Translation=(X=0,Y=-2,Z=0),Scale3D=(X=1,Y=1,Z=1))))
    ItemToModify.0.Transform.Rotation=&RigVMFunction_MathQuaternionFromEuler(Euler=(X=0,Y=40,Z=0)).Result
    Weight=&RigVMFunction_MathDoubleMul(B=@n3.Result, A=RigVMVariableNode() {variable=head_up_SP_result}.Value).Result
   )
   > &RigUnit_ModifyTransforms(
    ItemToModify=((Item=(Type=Bone,Name="extra_clavicle_r"),Transform=(Rotation=(X=0,Y=0,Z=0,W=1),Translation=(X=0,Y=-2,Z=0),Scale3D=(X=1,Y=1,Z=1))))
    ItemToModify.0.Transform.Rotation=&RigVMFunction_MathQuaternionFromEuler(Euler=(X=0,Y=40,Z=0)).Result
    Weight=RigVMVariableNode() {variable=head_up_SP_result}.Value
   )
   n4 = &RigUnit_SphericalPoseReader(
    ActiveRegionScaleFactors=(PositiveWidth=1,NegativeWidth=1,PositiveHeight=1.4,NegativeHeight=1.2)
    ActiveRegionSize=0.1
    Debug=(bDrawDebug=false,DebugScale=25,DebugSegments=20,DebugThickness=0.25)
    DriverAxis=(X=0,Y=1,Z=0)
    DriverItem=(Type=Bone,Name="head")
    FalloffRegionScaleFactors=(PositiveWidth=1,NegativeWidth=1.5,PositiveHeight=1,NegativeHeight=1)
    FalloffSize=0.2
    OptionalParentItem=(Type=Bone)
    RotationOffset=(X=90,Y=0,Z=-30)
    ExecutePin=@SequenceAggregate_1.B
   )
   > &RigUnit_ModifyTransforms(
    ItemToModify=((Item=(Type=Bone,Name="hair_front_r_BASE"),Transform=(Rotation=(X=0,Y=0,Z=0,W=1),Translation=(X=0,Y=-4,Z=0),Scale3D=(X=1,Y=1,Z=1))))
    ItemToModify.0.Transform.Rotation=&RigVMFunction_MathQuaternionFromEuler(Euler=(X=30,Y=0,Z=0)).Result
    Weight=@n4.OutputParam
   )
   > &RigUnit_ModifyTransforms(
    ItemToModify=((Item=(Type=Bone,Name="hair_front_l_BASE"),Transform=(Rotation=(X=0,Y=0,Z=0,W=1),Translation=(X=0,Y=4,Z=-3),Scale3D=(X=1,Y=1,Z=1))))
    ItemToModify.0.Transform.Rotation=&RigVMFunction_MathQuaternionFromEuler(Euler=(X=30,Y=0,Z=0)).Result
    Weight=@n4.OutputParam
   )
   n5 = &RigUnit_SphericalPoseReader(
    ActiveRegionScaleFactors=(PositiveWidth=0.5,NegativeWidth=0.2,PositiveHeight=1.4,NegativeHeight=1.2)
    ActiveRegionSize=0.1
    Debug=(bDrawDebug=false,DebugScale=25,DebugSegments=20,DebugThickness=0.25)
    DriverAxis=(X=0,Y=1,Z=0)
    DriverItem=(Type=Bone,Name="head")
    FalloffRegionScaleFactors=(PositiveWidth=1,NegativeWidth=0.2,PositiveHeight=1,NegativeHeight=1)
    FalloffSize=0.2
    OptionalParentItem=(Type=Bone)
    RotationOffset=(X=90,Y=0,Z=0)
    ExecutePin=@SequenceAggregate_1.C
   )
   n6 = &RigVMFunction_AlphaInterp(
    InterpSpeedDecreasing=3
    InterpSpeedIncreasing=5
    Scale=1
    bClampResult=False
    Value=&RigVMFunction_MathDoubleMul(A=@n5.OutputParam, B=RigVMVariableNode() {variable=not_crouching}.Value).Result
   )
   > &RigUnit_ModifyTransforms(
    ItemToModify=((Item=(Type=Bone,Name="extra_hair_spine_05_02"),Transform=(Rotation=(X=0,Y=0,Z=0,W=1),Translation=(X=0,Y=5,Z=0),Scale3D=(X=1,Y=1,Z=1))))
    Weight=@n6.Result
   )
   > &RigUnit_ModifyTransforms(
    ItemToModify=((Item=(Type=Bone,Name="extra_hair_neck_02_01"),Transform=(Rotation=(X=0,Y=0,Z=0,W=1),Translation=(X=0,Y=5,Z=0),Scale3D=(X=1,Y=1,Z=1))))
    Weight=@n6.Result
   )
   n7 = &RigUnit_SphericalPoseReader(
    ActiveRegionScaleFactors=(PositiveWidth=0.5,NegativeWidth=1,PositiveHeight=0.75,NegativeHeight=0.75)
    ActiveRegionSize=0.1
    Debug=(bDrawDebug=true,bDraw2D=false,bDrawLocalAxes=False,DebugScale=25,DebugSegments=20,DebugThickness=0.25)
    DriverAxis=(X=0,Y=1,Z=0)
    DriverItem=(Type=Bone,Name="head")
    FalloffRegionScaleFactors=(PositiveWidth=1,NegativeWidth=1,PositiveHeight=1,NegativeHeight=1)
    FalloffSize=0.1
    OptionalParentItem=(Type=Bone)
    RotationOffset=(X=90,Y=0,Z=10)
    ExecutePin=@SequenceAggregate_1.D
   )
   > &RigUnit_ModifyTransforms(
    ItemToModify=((Item=(Type=Bone,Name="dyn_hair_back_l_BASE"),Transform=(Rotation=(X=0,Y=0,Z=0,W=1),Translation=(X=0,Y=0,Z=0),Scale3D=(X=1,Y=1,Z=1))))
    ItemToModify.0.Transform.Rotation=&RigVMFunction_MathQuaternionFromEuler(Euler=(X=0,Y=0,Z=-30)).Result
    Weight=@n7.OutputParam
   )
   > &RigUnit_ModifyTransforms(
    ItemToModify=((Item=(Type=Bone,Name="dyn_hair_back_m_BASE"),Transform=(Rotation=(X=0,Y=0,Z=0,W=1),Translation=(X=0,Y=0,Z=0),Scale3D=(X=1,Y=1,Z=1))))
    ItemToModify.0.Transform.Rotation=&RigVMFunction_MathQuaternionFromEuler(Euler=(X=0,Y=0,Z=-50)).Result
    Weight=@n7.OutputParam
   )
   > &RigUnit_ModifyTransforms(
    ItemToModify=((Item=(Type=Bone,Name="dyn_hair_back_r_BASE"),Transform=(Rotation=(X=0,Y=0,Z=0,W=1),Translation=(X=0,Y=0,Z=0),Scale3D=(X=1,Y=1,Z=1))))
    ItemToModify.0.Transform.Rotation=&RigVMFunction_MathQuaternionFromEuler(Euler=(X=0,Y=0,Z=-30)).Result
    Weight=@n7.OutputParam
   )
   n8 = &RigVMFunction_AlphaInterp(InterpSpeedDecreasing=3, InterpSpeedIncreasing=5, Scale=1, bClampResult=False, Value=RigVMVariableNode() {variable=melee_float}.Value)
   &RigUnit_ModifyTransforms(
    ItemToModify=((Item=(Type=Bone,Name="hair_front_l_BASE"),Transform=(Rotation=(X=0,Y=0,Z=0,W=1),Translation=(X=0,Y=-3,Z=0),Scale3D=(X=1,Y=1,Z=1))))
    Weight=@n8.Result
    ExecutePin=@SequenceAggregate_1.G
   )
   > &RigUnit_ModifyTransforms(
    ItemToModify=((Item=(Type=Bone,Name="extra_ear_joint"),Transform=(Rotation=(X=0,Y=0,Z=0,W=1),Translation=(X=0,Y=0,Z=0),Scale3D=(X=1,Y=1,Z=1))))
    Weight=1
    ItemToModify.0.Transform.Rotation=&RigVMFunction_MathQuaternionFromEuler(Euler=(X=0,Y=30,Z=0)).Result
   )
   &RigUnit_SphericalPoseReader(
    ActiveRegionScaleFactors=(PositiveWidth=0.75,NegativeWidth=0.75,PositiveHeight=1,NegativeHeight=0.5)
    ActiveRegionSize=0.1
    Debug=(bDrawDebug=false,DebugScale=25,DebugSegments=20,DebugThickness=0.25)
    DriverAxis=(X=0,Y=1,Z=0)
    DriverItem=(Type=Bone,Name="head")
    ExecutePin=()
    FalloffRegionScaleFactors=(PositiveWidth=1,NegativeWidth=1,PositiveHeight=1,NegativeHeight=0.5)
    FalloffSize=0.3
    OptionalParentItem=(Type=Bone)
    RotationOffset=(X=120,Y=0,Z=0)
   )
   &RigUnit_SphericalPoseReader(
    ActiveRegionScaleFactors=(PositiveWidth=0.75,NegativeWidth=0.75,PositiveHeight=0.5,NegativeHeight=1)
    ActiveRegionSize=0.1
    Debug=(bDrawDebug=True,DebugScale=25,DebugSegments=20,DebugThickness=0.25)
    DriverAxis=(X=0,Y=1,Z=0)
    DriverItem=(Type=Bone,Name="head")
    ExecutePin=()
    FalloffRegionScaleFactors=(PositiveWidth=1,NegativeWidth=1,PositiveHeight=0.5,NegativeHeight=1)
    FalloffSize=0.3
    OptionalParentItem=(Type=Bone)
    RotationOffset=(X=60,Y=0,Z=0)
   )
  ]}
]
hierarchy [
 root {loc=<0 0 0>, rot=<0 0 0 1>}
 pelvis {parent=root, loc=<0 -0.7071 93.1191>, rot=<0.0127 0.707 -0.0127 -0.707>}
 spine_01 {parent=pelvis, loc=<6.7494 -0.6388 0>, rot=<0 0 0.034 -0.9994>}
 spine_02 {parent=spine_01, loc=<7.2518 -0.035 0>, rot=<0 0 0.0238 -0.9997>}
 spine_03 {parent=spine_02, loc=<6.0104 0 0>, rot=<0 0 0.0149 -0.9999>}
 spine_04 {parent=spine_03, loc=<8.1024 0 0>, rot=<0 0 -0.0132 -0.9999>}
 spine_05 {parent=spine_04, loc=<10.0146 0 0>, rot=<0 0 -0.0015 -1>}
 neck_01 {parent=spine_05, loc=<11.3138 0.1529 0>, rot=<0 0 -0.2003 -0.9797>}
 neck_02 {parent=neck_01, loc=<4.5627 0 0>, rot=<0 0 0.0504 -0.9987>}
 head {parent=neck_02, loc=<5.3914 0 0>, rot=<0 0 0.115 -0.9934>}
 dyn_hair_back_l_BASE {parent=head, loc=<8.1454 -1.6936 -5.9596>, rot=<0.192 -0.969 -0.153 -0.029>}
 dyn_hair_back_l_00 {parent=dyn_hair_back_l_BASE, loc=<0 0 0>, rot=<0 0 -0.1131 -0.9936>}
 dyn_hair_back_l_01 {parent=dyn_hair_back_l_00, loc=<7.5331 0 0>, rot=<0 0 0.1131 -0.9936>}
 dyn_hair_back_l_02 {parent=dyn_hair_back_l_01, loc=<4.8381 0 0>, rot=<0.0374 0.0185 0.0757 -0.9963>}
 dyn_hair_back_l_03 {parent=dyn_hair_back_l_02, loc=<3.7715 0 0>, rot=<0.0175 0.0118 0.0592 -0.998>}
 dyn_hair_back_l_04 {parent=dyn_hair_back_l_03, loc=<3.5522 0 0>, rot=<-0.2719 0.0157 -0.0062 -0.9622>}
 dyn_hair_back_m_BASE {parent=head, loc=<8.1443 -5.0204 -0.0835>, rot=<0.1255 -0.9915 0.0343 0.0041>}
 dyn_hair_back_m_00 {parent=dyn_hair_back_m_BASE, loc=<0 0 0>, rot=<0 0 -0.0391 -0.9992>}
 dyn_hair_back_m_01 {parent=dyn_hair_back_m_00, loc=<7.34 0 0>, rot=<0 0 0.0391 -0.9992>}
 dyn_hair_back_m_02 {parent=dyn_hair_back_m_01, loc=<4.6355 0 0>, rot=<0.0092 0.0027 0.0437 -0.999>}
 dyn_hair_back_m_03 {parent=dyn_hair_back_m_02, loc=<4.5286 0 0>, rot=<0.0017 0.0006 0.012 -0.9999>}
 dyn_hair_back_m_04 {parent=dyn_hair_back_m_03, loc=<4.1333 0 0>, rot=<0.0246 0.0016 0.063 -0.9977>}
 dyn_hair_back_r_BASE {parent=head, loc=<8.0164 -1.5814 5.9229>, rot=<0.969 0.192 -0.029 0.153>}
 dyn_hair_back_r_00 {parent=dyn_hair_back_r_BASE, loc=<0 0 0>, rot=<0 0 0 1>}
 dyn_hair_back_r_01 {parent=dyn_hair_back_r_00, loc=<-7.2659 -1.5353 0>, rot=<0 0 0 1>}
 dyn_hair_back_r_02 {parent=dyn_hair_back_r_01, loc=<-4.8384 -0.0001 0>, rot=<0.0374 0.0185 0.0757 -0.9963>}
 dyn_hair_back_r_03 {parent=dyn_hair_back_r_02, loc=<-3.7713 0.0001 0>, rot=<0.0175 0.0118 0.0592 -0.998>}
 dyn_hair_back_r_04 {parent=dyn_hair_back_r_03, loc=<-3.5523 0 0>, rot=<-0.2719 0.0157 -0.0062 -0.9622>}
 extra_ear_joint {parent=head, loc=<11.9933 0.9644 5.8796>, rot=<0 -0.1644 0 -0.9864>}
 hair_front_r_BASE {parent=head, loc=<-1.1748 4.7775 6.3739>, rot=<-0.101 -0.0215 0.073 -0.992>}
 dyn_hair_front_r_01 {parent=hair_front_r_BASE, loc=<0 0 0>, rot=<0 0 0 1>}
 dyn_hair_front_r_02 {parent=dyn_hair_front_r_01, loc=<-4.9137 0.0001 0>, rot=<0.0613 0.006 0.0387 -0.9974>}
 dyn_hair_front_r_03 {parent=dyn_hair_front_r_02, loc=<-5.1838 -0.0001 0>, rot=<0 0 0 1>}
 hair_front_l_BASE {parent=head, loc=<-1.1745 4.7775 -6.3739>, rot=<-0.0215 0.101 0.992 0.073>}
 dyn_hair_front_l_01 {parent=hair_front_l_BASE, loc=<0 0 0>, rot=<-0.0826 0 0 -0.9966>}
 dyn_hair_front_l_02 {parent=dyn_hair_front_l_01, loc=<4.9145 0 0>, rot=<0.0613 0.0122 0.0372 -0.9974>}
 dyn_hair_front_l_03 {parent=dyn_hair_front_l_02, loc=<5.1832 0 0>, rot=<0 0 0 1>}
 extra_hair_neck_02_01 {parent=neck_02, loc=<0 0 0>, rot=<0 0 0 1>}
 clavicle_l {parent=spine_05, loc=<4.3575 -0.1466 -2.3223>, rot=<0.0465 -0.7289 -0.009 -0.6829>}
 extra_clavicle_l {parent=clavicle_l, loc=<0 0 0>, rot=<0 0 0 1>}
 clavicle_r {parent=spine_05, loc=<4.3575 -0.1466 2.3223>, rot=<0.7289 0.0465 -0.6829 0.009>}
 extra_clavicle_r {parent=clavicle_r, loc=<0 0 0>, rot=<0 0 0 1>}
 extra_hair_spine_05 {parent=spine_05, loc=<0 0 0>, rot=<0 0 0 1>}
 extra_hair_spine_05_02 {parent=spine_05, loc=<0 0 0>, rot=<0 0 0 1>}
 extra_hair_spine_04 {parent=spine_04, loc=<0 0 0>, rot=<0 0 0 1>}
]
variables [
 Target {type=double, default=0}
 head_up_SP_result {type=double, default=0}
 melee_float {type=double, default=0}
 notCrouchingTargeting {type=double, default=0}
 not_crouching {type=double, default=0}
 not_pistol {type=double, default=0}
 pistol {type=double, default=0}
]
