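# Physics asset for the Patches medium male face accessory mesh, setting up dynamic hair.
# Three back and two front capsule chains hang from head-linked base bodies, with limited swing and twist along each chain.
# Free-angular constraints join the back chain tips left to middle to right.
# Head, neck, spine and clavicle shapes, two collar boxes and extra oversized or offset capsules on helper bones form the upper-body collision volumes.
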PhysicsAsset /BRCosmetics/Characters/Player/Male/Medium/Bodies/M_MED_Patches/Meshes/Parts/M_MED_Patches_FaceAcc_Physics
bodies {
 clavicle_l [{
   shapeType=Capsule
   length=12.848
   radius=6.777
   shapeName=clavicle_l_capsule
   center=<11.0637 1.2076 2.093>
   rotation=<rot 83.6478 -164.1804 -174.1237>
  }]
 clavicle_r [
  {
   shapeType=Capsule
   length=14
   radius=5.625
   shapeName=clavicle_r_capsule_01
   center=<-11.6326 0.1117 -1.7248>
   rotation=<rot 83.6478 -164.1807 5.876>
  }
  {
   shapeType=Capsule
   length=14
   radius=5.625
   shapeName=clavicle_r_capsule_02
   center=<-10.9935 -2.9097 -3.7429>
   rotation=<rot 83.6478 -164.1807 5.876>
  }
 ]
 dyn_hair_back_l_00 [{
   shapeType=Capsule
   length=3.0925
   radius=2.3948
   shapeName=dyn_hair_back_l_00_capsule
   center=<4.2988 -0.1757 -0.3631>
   rotation=<rot 27.5105 -89.9999 -86.9923>
  }]
 dyn_hair_back_l_01 [{shapeType=Capsule, length=1.5676, radius=2.0003, center=<2.2057 -0.5796 -0.0678>, rotation=<rot 0 0.0001 33.6178>}]
 dyn_hair_back_l_02 [{shapeType=Capsule, length=2.4059, radius=1.7804, center=<1.5 0 0.7>, rotation=<rot 0 0 17.3936>}]
 dyn_hair_back_l_03 [{shapeType=Capsule, length=1.0701, radius=2.1616, center=<1.7 0 0.7>, rotation=<rot 0 0 12.7973>}]
 dyn_hair_back_l_04 [{shapeType=Capsule, length=2.7653, radius=1.6441, center=<1.5 0 0.5>, rotation=<rot 0 0 44.6267>}]
 dyn_hair_back_l_BASE [{shapeType=Capsule, length=0.505, radius=0.505, shapeName=dyn_hair_back_l_BASE_capsule, rotation=<rot 0 -90 0>}]
 dyn_hair_back_m_00 [{shapeType=Capsule, length=2.9397, radius=2.2581, shapeName=dyn_hair_back_m_00_capsule, center=<2.4009 -0.9696 0>, rotation=<rot 0 -90 0>}]
 dyn_hair_back_m_01 [{shapeType=Capsule, length=6.4257, radius=2.9941, shapeName=dyn_hair_back_m_01_capsule, center=<2.1513 -2.0033 0>, rotation=<rot 0 -90 0>}]
 dyn_hair_back_m_02 [{
   shapeType=Capsule
   length=3.2185
   radius=3.3324
   shapeName=dyn_hair_back_m_02_capsule
   center=<2.3058 -2.0033 0.0327>
   rotation=<rot 0 -90 -0.8122>
  }]
 dyn_hair_back_m_03 [{shapeType=Capsule, length=6.1384, radius=3.2082, shapeName=dyn_hair_back_m_03_capsule, center=<2.2417 -2.0033 0>, rotation=<rot 0 -90 0>}]
 dyn_hair_back_m_04 [{shapeType=Capsule, length=7.6327, radius=3.527, shapeName=dyn_hair_back_m_04_capsule, center=<6.4275 -1.0798 0>, rotation=<rot 0 -90 0>}]
 dyn_hair_back_m_BASE [{shapeType=Capsule, length=0.505, radius=0.505, shapeName=dyn_hair_back_m_BASE_capsule, rotation=<rot 0 -90 0>}]
 dyn_hair_back_r_00 [{
   shapeType=Capsule
   length=3.3697
   radius=2.4275
   shapeName=dyn_hair_back_r_00_capsule
   center=<-4.1427 -0.5046 0.7928>
   rotation=<rot 22.7063 -86.8002 85.9192>
  }]
 dyn_hair_back_r_01 [{shapeType=Capsule, length=2.2731, radius=1.7965, center=<-2.4898 0.0859 0.3186>, rotation=<rot 0 0 -159.9417>}]
 dyn_hair_back_r_02 [{shapeType=Capsule, length=2.8141, radius=1.8012, center=<-1.6136 0.2713 1.9756>, rotation=<rot 0 0.0001 -167.994>}]
 dyn_hair_back_r_03 [{shapeType=Capsule, length=1.0701, radius=2.1616, center=<-1.7 0.5116 1.5523>, rotation=<rot 0 0 -167.2027>}]
 dyn_hair_back_r_04 [{shapeType=Capsule, length=3.3692, radius=2.2627, center=<-2.9865 1.2407 -0.4768>, rotation=<rot 22.1853 13.5557 -158.081>}]
 dyn_hair_back_r_BASE [{shapeType=Capsule, length=0.505, radius=0.505, shapeName=dyn_hair_back_r_BASE_capsule, rotation=<rot 0 -90 0>}]
 dyn_hair_front_l_01 [{shapeType=Capsule, length=2.25, radius=2, center=<2.6302 -0.3625 -0.5269>, rotation=<rot 90 0 0>}]
 dyn_hair_front_l_02 [{shapeType=Capsule, length=3.1606, radius=1.8394, center=<3.4512 0.5618 0.191>, rotation=<rot 90 0 0>}]
 dyn_hair_front_l_03 [{shapeType=Capsule, length=7.0178, radius=2.0257, center=<5.2521 0.7529 -0.0001>, rotation=<rot 90 -20.1626 -21.0369>}]
 dyn_hair_front_r_01 [{shapeType=Capsule, length=1.4243, radius=1.2624, center=<-2.9525 0 -0.5968>, rotation=<rot 86.628 180 -179.9998>}]
 dyn_hair_front_r_02 [{shapeType=Capsule, length=2.125, radius=2.875, center=<-3.2376 0.9871 0.0001>, rotation=<rot 90 180 -175.425>}]
 dyn_hair_front_r_03 [{shapeType=Capsule, length=3.6152, radius=4.3562, center=<-5.2491 0.5218 -0.9379>, rotation=<rot 90 180 -173.2913>}]
 extra_clavicle_l [{
   shapeType=Capsule
   length=12.848
   radius=6.777
   shapeName=extra_clavicle_l_capsule
   center=<11.0637 1.2076 2.093>
   rotation=<rot 83.6478 -164.1804 -174.1237>
  }]
 extra_clavicle_r [{
   shapeType=Capsule
   length=26.2958
   radius=5.625
   shapeName=extra_clavicle_r_capsule
   center=<-10.8887 -2.9602 -3.57>
   rotation=<rot 83.6478 -164.1807 5.876>
  }]
 extra_hair_neck_02_01 [{
   shapeType=Capsule
   length=35.4988
   radius=5.375
   shapeName=extra_hair_neck_02_01_capsule
   center=<3.064 2.2928 -0.1402>
   rotation=<rot -1.3118 180 179.9998>
  }]
 extra_hair_spine_04 [{
   shapeType=Capsule
   length=18.1381
   radius=13.0009
   shapeName=extra_hair_spine_04_capsule
   center=<2.5514 0.7729 0>
   rotation=<rot 0 -92.6845 0>
  }]
 extra_hair_spine_05 [{shapeType=Capsule, length=62.9432, radius=8.7055, shapeName=extra_hair_spine_05_capsule, center=<2.6959 -1.5257 0>, rotation=<rot 0 -90 0>}]
 extra_hair_spine_05_02 [
  {shapeType=Capsule, length=63.9446, radius=3.3055, shapeName=spine_05_capsule_02, center=<13.2034 2.9172 0>}
  {shapeType=Capsule, length=62.0588, radius=5.1913, shapeName=spine_05_capsule_03, center=<15.2112 4.8298 0>}
 ]
 hair_front_l_BASE [{shapeType=Capsule, length=0.505, radius=0.505, shapeName=hair_front_l_BASE_capsule, rotation=<rot 0 -90 0>}]
 hair_front_r_BASE [{shapeType=Capsule, length=0.505, radius=0.505, shapeName=hair_front_r_BASE_capsule, rotation=<rot 0 -90 0>}]
 head [{shapeType=Capsule, length=5.5, radius=5.875, center=<-0.0002 2.1581 -0.0324>, rotation=<rot 89.3496 179.9998 179.9998>}]
 neck_02 [{shapeType=Capsule, length=7.75, radius=5.375, center=<-1.6187 0.1345 -0.0184>, rotation=<rot 89.3496 179.9998 179.9998>}]
 spine_04 [{shapeType=Capsule, length=26.3667, radius=4.7724, shapeName=spine_04_capsule, center=<19.8893 2.6467 0>, rotation=<rot 0 -92.6845 0>}]
 spine_05 [
  {shapeType=Capsule, length=12.5149, radius=7.013, shapeName=spine_05_capsule_01, center=<5.7291 0.9235 -13.8994>}
  {
   shapeType=Capsule
   length=7.3544
   radius=6.645
   shapeName=spine_05_capsule_04
   center=<8.1518 1.667 0.0012>
   rotation=<rot 89.6347 179.9682 -172.8204>
  }
  {
   shapeType=Box
   extentX=4.5211
   extentY=5.2443
   extentZ=1.5126
   shapeName=spine_05_collar_01
   center=<12.2775 5.241 -7.0785>
   rotation=<rot -6.2239 -51.6071 -44.8347>
  }
  {
   shapeType=Box
   extentX=7.5
   extentY=1.011
   extentZ=1.797
   shapeName=spine_05_collar_02
   center=<9.3373 8.026 -3.3221>
   rotation=<rot -24.1169 -16.7746 -52.0749>
  }
 ]
}
constraints [
 {bone1Name=dyn_hair_back_l_01, bone2Name=dyn_hair_back_l_00, swing1LimitDegrees=45, swing2LimitDegrees=5, twistLimitDegrees=5}
 {bone1Name=dyn_hair_back_l_02, bone2Name=dyn_hair_back_l_01, swing1LimitDegrees=45, swing2LimitDegrees=5, twistLimitDegrees=5}
 {bone1Name=dyn_hair_back_l_03, bone2Name=dyn_hair_back_l_02, swing1LimitDegrees=45, swing2LimitDegrees=5, twistLimitDegrees=5}
 {bone1Name=dyn_hair_back_l_04, bone2Name=dyn_hair_back_l_03, swing1LimitDegrees=45, swing2LimitDegrees=5, twistLimitDegrees=5}
 {bone1Name=dyn_hair_back_r_01, bone2Name=dyn_hair_back_r_00, swing1LimitDegrees=45, swing2LimitDegrees=5, twistLimitDegrees=20}
 {bone1Name=dyn_hair_back_r_02, bone2Name=dyn_hair_back_r_01, swing1LimitDegrees=45, swing2LimitDegrees=5, twistLimitDegrees=20}
 {bone1Name=dyn_hair_back_m_01, bone2Name=dyn_hair_back_m_00, swing1LimitDegrees=45, swing2LimitDegrees=5, twistLimitDegrees=5}
 {bone1Name=dyn_hair_back_m_02, bone2Name=dyn_hair_back_m_01, swing1LimitDegrees=45, swing2LimitDegrees=5, twistLimitDegrees=5}
 {bone1Name=dyn_hair_back_m_03, bone2Name=dyn_hair_back_m_02, swing1LimitDegrees=45, swing2LimitDegrees=5, twistLimitDegrees=5}
 {bone1Name=dyn_hair_back_m_04, bone2Name=dyn_hair_back_m_03, swing1LimitDegrees=45, swing2LimitDegrees=5, twistLimitDegrees=5}
 {
  bone1Name=spine_05
  bone2Name=spine_04
  swing1LimitDegrees=45
  swing1Motion=Free
  swing2LimitDegrees=45
  swing2Motion=Free
  twistLimitDegrees=45
  twistMotion=Free
 }
 {
  bone1Name=clavicle_l
  bone2Name=spine_05
  swing1LimitDegrees=45
  swing1Motion=Free
  swing2LimitDegrees=45
  swing2Motion=Free
  twistLimitDegrees=45
  twistMotion=Free
 }
 {
  bone1Name=clavicle_r
  bone2Name=spine_05
  swing1LimitDegrees=45
  swing1Motion=Free
  swing2LimitDegrees=45
  swing2Motion=Free
  twistLimitDegrees=45
  twistMotion=Free
 }
 {
  bone1Name=neck_02
  bone2Name=spine_05
  swing1LimitDegrees=45
  swing1Motion=Free
  swing2LimitDegrees=45
  swing2Motion=Free
  twistLimitDegrees=45
  twistMotion=Free
 }
 {
  bone1Name=head
  bone2Name=neck_02
  swing1LimitDegrees=45
  swing1Motion=Free
  swing2LimitDegrees=45
  swing2Motion=Free
  twistLimitDegrees=45
  twistMotion=Free
 }
 {bone1Name=dyn_hair_back_r_04, bone2Name=dyn_hair_back_r_03, swing1LimitDegrees=45, swing2LimitDegrees=5, twistLimitDegrees=5}
 {bone1Name=dyn_hair_back_r_03, bone2Name=dyn_hair_back_r_02, swing1LimitDegrees=45, swing2LimitDegrees=5, twistLimitDegrees=20}
 {
  bone1Name=dyn_hair_back_l_04
  bone2Name=dyn_hair_back_m_04
  swing1LimitDegrees=45
  swing1Motion=Free
  swing2LimitDegrees=45
  swing2Motion=Free
  twistLimitDegrees=45
  twistMotion=Free
 }
 {
  bone1Name=dyn_hair_back_l_BASE
  bone2Name=head
  swing1LimitDegrees=45
  swing1Motion=Free
  swing2LimitDegrees=45
  swing2Motion=Free
  twistLimitDegrees=45
  twistMotion=Free
 }
 {bone1Name=dyn_hair_back_l_00, bone2Name=dyn_hair_back_l_BASE, swing1LimitDegrees=45, swing2LimitDegrees=10, twistLimitDegrees=5}
 {
  bone1Name=dyn_hair_back_m_BASE
  bone2Name=head
  swing1LimitDegrees=45
  swing1Motion=Free
  swing2LimitDegrees=45
  swing2Motion=Free
  twistLimitDegrees=45
  twistMotion=Free
 }
 {bone1Name=dyn_hair_back_m_00, bone2Name=dyn_hair_back_m_BASE, swing1LimitDegrees=45, swing2LimitDegrees=20, twistLimitDegrees=5}
 {
  bone1Name=dyn_hair_back_r_BASE
  bone2Name=head
  swing1LimitDegrees=45
  swing1Motion=Free
  swing2LimitDegrees=45
  swing2Motion=Free
  twistLimitDegrees=45
  twistMotion=Free
 }
 {bone1Name=dyn_hair_back_r_00, bone2Name=dyn_hair_back_r_BASE, swing1LimitDegrees=45, swing2LimitDegrees=45, twistLimitDegrees=20}
 {
  bone1Name=dyn_hair_back_m_04
  bone2Name=dyn_hair_back_r_04
  swing1LimitDegrees=45
  swing1Motion=Free
  swing2LimitDegrees=45
  swing2Motion=Free
  twistLimitDegrees=45
  twistMotion=Free
 }
 {bone1Name=dyn_hair_front_l_01, bone2Name=hair_front_l_BASE, swing1LimitDegrees=30, swing2LimitDegrees=20, twistLimitDegrees=3}
 {bone1Name=dyn_hair_front_l_02, bone2Name=dyn_hair_front_l_01, swing1LimitDegrees=30, swing2LimitDegrees=20, twistLimitDegrees=3}
 {bone1Name=dyn_hair_front_l_03, bone2Name=dyn_hair_front_l_02, swing1LimitDegrees=30, swing2LimitDegrees=20, twistLimitDegrees=3}
 {bone1Name=dyn_hair_front_r_01, bone2Name=hair_front_r_BASE, swing1LimitDegrees=30, swing2LimitDegrees=20, twistLimitDegrees=3}
 {bone1Name=dyn_hair_front_r_02, bone2Name=dyn_hair_front_r_01, swing1LimitDegrees=30, swing2LimitDegrees=20, twistLimitDegrees=3}
 {bone1Name=dyn_hair_front_r_03, bone2Name=dyn_hair_front_r_02, swing1LimitDegrees=30, swing2LimitDegrees=20, twistLimitDegrees=3}
 {
  bone1Name=extra_hair_spine_05
  bone2Name=spine_05
  swing1LimitDegrees=45
  swing1Motion=Free
  swing2LimitDegrees=45
  swing2Motion=Free
  twistLimitDegrees=45
  twistMotion=Free
 }
 {
  bone1Name=extra_hair_spine_04
  bone2Name=spine_04
  swing1LimitDegrees=45
  swing1Motion=Free
  swing2LimitDegrees=45
  swing2Motion=Free
  twistLimitDegrees=45
  twistMotion=Free
 }
 {
  bone1Name=extra_clavicle_l
  bone2Name=clavicle_l
  swing1LimitDegrees=45
  swing1Motion=Free
  swing2LimitDegrees=45
  swing2Motion=Free
  twistLimitDegrees=45
  twistMotion=Free
 }
 {
  bone1Name=extra_clavicle_r
  bone2Name=clavicle_r
  swing1LimitDegrees=45
  swing1Motion=Free
  swing2LimitDegrees=45
  swing2Motion=Free
  twistLimitDegrees=45
  twistMotion=Free
 }
 {
  bone1Name=hair_front_l_BASE
  bone2Name=head
  swing1LimitDegrees=45
  swing1Motion=Free
  swing2LimitDegrees=45
  swing2Motion=Free
  twistLimitDegrees=45
  twistMotion=Free
 }
 {
  bone1Name=hair_front_r_BASE
  bone2Name=head
  swing1LimitDegrees=45
  swing1Motion=Free
  swing2LimitDegrees=45
  swing2Motion=Free
  twistLimitDegrees=45
  twistMotion=Free
 }
 {
  bone1Name=extra_hair_neck_02_01
  bone2Name=neck_02
  swing1LimitDegrees=45
  swing1Motion=Free
  swing2LimitDegrees=45
  swing2Motion=Free
  twistLimitDegrees=45
  twistMotion=Free
 }
 {
  bone1Name=extra_hair_spine_05_02
  bone2Name=spine_05
  swing1LimitDegrees=45
  swing1Motion=Free
  swing2LimitDegrees=45
  swing2Motion=Free
  twistLimitDegrees=45
  twistMotion=Free
 }
]
