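# Physics asset for the face accessory of the male medium Patches character (sky variant), defining capsule bodies and constraints for its dynamic hair bones.
# Three back chains (left, middle, right) hang from head via free-swinging BASE bodies with tight limits along each, while two front chains attach directly with looser swing limits.
# Back-side tips cross-link to the middle tip with free angular constraints, tying strands together.
# Spine, clavicle, neck and head capsules supply upper-body shapes, joined by angularly free constraints.
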PhysicsAsset /BRCosmetics/Characters/Player/Male/Medium/Bodies/M_MED_Patches/Meshes/Parts/M_MED_Patches_FaceAcc_Physics_sky
bodies {
 clavicle_l [{
   shapeType=Capsule
   length=12.848
   radius=6.777
   shapeName=clavicle_l_capsule
   center=<11.0637 1.2076 2.093>
   rotation=<rot 83.6478 -164.1804 -174.1237>
  }]
 clavicle_r [{
   shapeType=Capsule
   length=14
   radius=5.625
   shapeName=clavicle_l_capsule
   center=<-11.6326 0.1117 -1.7248>
   rotation=<rot 83.6478 -164.1807 5.876>
  }]
 dyn_hair_back_l_00 [{shapeType=Capsule, length=2.0104, radius=2.2738, shapeName=dyn_hair_back_l_00_capsule, rotation=<rot 11.6374 -89.9851 -0.0049>}]
 dyn_hair_back_l_01 [{shapeType=Capsule, length=1.5676, radius=2.0003, center=<2.2057 -0.5796 -0.0678>, rotation=<rot 0 0.0001 33.6178>}]
 dyn_hair_back_l_02 [{shapeType=Capsule, length=2.4059, radius=1.7804, center=<1.5 0 0.7>, rotation=<rot 0 0 17.3936>}]
 dyn_hair_back_l_03 [{shapeType=Capsule, length=1.0701, radius=2.1616, center=<1.7 0 0.7>, rotation=<rot 0 0 12.7973>}]
 dyn_hair_back_l_04 [{shapeType=Capsule, length=2.7653, radius=1.6441, center=<1.5 0 0.5>, rotation=<rot 0 0 44.6267>}]
 dyn_hair_back_l_BASE [{shapeType=Capsule, length=0.505, radius=0.505, shapeName=dyn_hair_back_l_BASE_capsule, rotation=<rot 0 -90 0>}]
 dyn_hair_back_m_00 [{shapeType=Capsule, length=1.6174, radius=2.6669, shapeName=dyn_hair_back_m_00_capsule, rotation=<rot 0 -90 0>}]
 dyn_hair_back_m_01 [{shapeType=Capsule, length=6.4257, radius=2.9941, shapeName=dyn_hair_back_m_01_capsule, center=<2.1513 -1.5019 0>, rotation=<rot 0 -90 0>}]
 dyn_hair_back_m_02 [{
   shapeType=Capsule
   length=3.2185
   radius=3.3324
   shapeName=dyn_hair_back_m_02_capsule
   center=<2.3058 -0.6371 0.0327>
   rotation=<rot 0 -90 -0.8122>
  }]
 dyn_hair_back_m_03 [{shapeType=Capsule, length=6.1384, radius=3.2082, shapeName=dyn_hair_back_m_03_capsule, center=<2.2417 -2.0033 0>, rotation=<rot 0 -90 0>}]
 dyn_hair_back_m_04 [
  {shapeType=Capsule, length=7.6327, radius=3.527, shapeName=dyn_hair_back_m_04_capsule, center=<2.0124 -1.5142 0>, rotation=<rot 0 -90 0>}
  {shapeType=Capsule, length=8.8421, radius=2.3176, shapeName=dyn_hair_back_m_04_capsule, center=<6.7168 0.7506 0>, rotation=<rot 0 -90 0>}
 ]
 dyn_hair_back_m_BASE [{shapeType=Capsule, length=0.505, radius=0.505, shapeName=dyn_hair_back_m_BASE_capsule, rotation=<rot 0 -90 0>}]
 dyn_hair_back_r_00 [{shapeType=Capsule, length=2.0104, radius=2.2738, shapeName=dyn_hair_back_r_00_capsule, rotation=<rot 12.2112 -90 0>}]
 dyn_hair_back_r_01 [{shapeType=Capsule, length=2.2731, radius=1.7965, center=<-2.4898 0.0859 0.3186>, rotation=<rot 0 0 -159.9417>}]
 dyn_hair_back_r_02 [{shapeType=Capsule, length=2.8141, radius=1.8012, center=<-1.6136 0.2713 1.9756>, rotation=<rot 0 0.0001 -167.994>}]
 dyn_hair_back_r_03 [{shapeType=Capsule, length=1.0701, radius=2.1616, center=<-1.7 0.5116 1.5523>, rotation=<rot 0 0 -167.2027>}]
 dyn_hair_back_r_04 [
  {shapeType=Capsule, length=2.7653, radius=1.6441, center=<-1.5 1.5489 1.0692>, rotation=<rot 0 0 -135.3733>}
  {
   shapeType=Capsule
   length=3.5585
   radius=0.8509
   shapeName=dyn_hair_back_r_04_capsule
   center=<-5.1742 0.0668 2.5336>
   rotation=<rot -58.9453 -95.8137 -65.5513>
  }
 ]
 dyn_hair_back_r_BASE [{shapeType=Capsule, length=0.505, radius=0.505, shapeName=dyn_hair_back_r_BASE_capsule, rotation=<rot 0 -90 0>}]
 dyn_hair_front_l_01 [{shapeType=Capsule, length=2.25, radius=2, center=<2.6302 -0.3625 -0.5269>, rotation=<rot 90 0 0>}]
 dyn_hair_front_l_02 [{shapeType=Capsule, length=2.125, radius=2.875, center=<3.4512 0 0.0001>, rotation=<rot 90 0 0>}]
 dyn_hair_front_l_03 [{shapeType=Capsule, length=5.1797, radius=3.8638, center=<5.1258 -0.3303 -0.0001>, rotation=<rot 90 -8.8807 -3.9667>}]
 dyn_hair_front_r_01 [{shapeType=Capsule, length=2.7064, radius=2.0436, center=<-2.8006 0 -0.0001>, rotation=<rot 90 0 0>}]
 dyn_hair_front_r_02 [{shapeType=Capsule, length=2.125, radius=2.875, center=<-3.2104 0 0.0001>, rotation=<rot 90 0 0>}]
 dyn_hair_front_r_03 [{shapeType=Capsule, length=4.4703, radius=3.5011, center=<-5.2491 0.5218 -0.0001>, rotation=<rot 90 180 -173.2913>}]
 head [{shapeType=Capsule, length=5.5, radius=5.875, center=<-0.0002 2.1581 -0.0324>, rotation=<rot 89.3496 179.9998 179.9998>}]
 neck_02 [
  {shapeType=Capsule, length=6.2617, radius=6.8633, center=<-0.0373 0 3.2844>, rotation=<rot 89.3496 179.9998 179.9998>}
  {
   shapeType=Capsule
   length=6.2617
   radius=6.8633
   shapeName=neck_02_capsule
   center=<0.1869 0.1345 -3.2086>
   rotation=<rot 89.3496 179.9998 179.9998>
  }
 ]
 spine_04 [{shapeType=Capsule, length=7.4066, radius=10.2626, shapeName=spine_04_capsule, center=<2.5514 0.7729 0>, rotation=<rot 0 -92.6845 0>}]
 spine_05 [
  {shapeType=Capsule, length=38.8114, radius=8.3057, shapeName=spine_05_capsule_01, center=<2.6959 -1.5257 0>}
  {shapeType=Capsule, length=63.9446, radius=3.3055, shapeName=spine_05_capsule_02, center=<13.2034 2.9172 0>}
  {shapeType=Capsule, length=3.949, radius=4.3703, shapeName=spine_05_capsule, center=<13.4019 -3.7366 0>}
  {shapeType=Capsule, length=15.6999, radius=8.3057, shapeName=spine_05_capsule, center=<-2.0032 -3.2855 0>}
 ]
}
constraints [
 {bone1Name=dyn_hair_back_l_01, bone2Name=dyn_hair_back_l_00, swing1LimitDegrees=10, swing2LimitDegrees=5, twistLimitDegrees=5}
 {bone1Name=dyn_hair_back_l_02, bone2Name=dyn_hair_back_l_01, swing1LimitDegrees=10, swing2LimitDegrees=5, twistLimitDegrees=5}
 {bone1Name=dyn_hair_back_l_03, bone2Name=dyn_hair_back_l_02, swing1LimitDegrees=10, swing2LimitDegrees=5, twistLimitDegrees=5}
 {bone1Name=dyn_hair_back_l_04, bone2Name=dyn_hair_back_l_03, swing1LimitDegrees=15, swing2LimitDegrees=5, twistLimitDegrees=5}
 {bone1Name=dyn_hair_back_r_01, bone2Name=dyn_hair_back_r_00, swing1LimitDegrees=10, swing2LimitDegrees=5, twistLimitDegrees=5}
 {bone1Name=dyn_hair_back_r_02, bone2Name=dyn_hair_back_r_01, swing1LimitDegrees=10, swing2LimitDegrees=5, twistLimitDegrees=5}
 {bone1Name=dyn_hair_back_m_01, bone2Name=dyn_hair_back_m_00, swing1LimitDegrees=10, swing2LimitDegrees=5, twistLimitDegrees=2}
 {bone1Name=dyn_hair_back_m_02, bone2Name=dyn_hair_back_m_01, swing1LimitDegrees=10, swing2LimitDegrees=5, twistLimitDegrees=2}
 {bone1Name=dyn_hair_back_m_03, bone2Name=dyn_hair_back_m_02, swing1LimitDegrees=10, swing2LimitDegrees=5, twistLimitDegrees=2}
 {bone1Name=dyn_hair_back_m_04, bone2Name=dyn_hair_back_m_03, swing1LimitDegrees=15, swing2LimitDegrees=5, twistLimitDegrees=2}
 {
  bone1Name=spine_05
  bone2Name=spine_04
  swing1LimitDegrees=45
  swing1Motion=Free
  swing2LimitDegrees=45
  swing2Motion=Free
  twistLimitDegrees=45
  twistMotion=Free
 }
 {
  bone1Name=clavicle_l
  bone2Name=spine_05
  swing1LimitDegrees=45
  swing1Motion=Free
  swing2LimitDegrees=45
  swing2Motion=Free
  twistLimitDegrees=45
  twistMotion=Free
 }
 {
  bone1Name=clavicle_r
  bone2Name=spine_05
  swing1LimitDegrees=45
  swing1Motion=Free
  swing2LimitDegrees=45
  swing2Motion=Free
  twistLimitDegrees=45
  twistMotion=Free
 }
 {
  bone1Name=neck_02
  bone2Name=spine_05
  swing1LimitDegrees=45
  swing1Motion=Free
  swing2LimitDegrees=45
  swing2Motion=Free
  twistLimitDegrees=45
  twistMotion=Free
 }
 {
  bone1Name=head
  bone2Name=neck_02
  swing1LimitDegrees=45
  swing1Motion=Free
  swing2LimitDegrees=45
  swing2Motion=Free
  twistLimitDegrees=45
  twistMotion=Free
 }
 {bone1Name=dyn_hair_back_r_04, bone2Name=dyn_hair_back_r_03, swing1LimitDegrees=15, swing2LimitDegrees=5, twistLimitDegrees=5}
 {bone1Name=dyn_hair_back_r_03, bone2Name=dyn_hair_back_r_02, swing1LimitDegrees=10, swing2LimitDegrees=5, twistLimitDegrees=5}
 {
  bone1Name=dyn_hair_back_r_04
  bone2Name=dyn_hair_back_m_04
  swing1LimitDegrees=45
  swing1Motion=Free
  swing2LimitDegrees=45
  swing2Motion=Free
  twistLimitDegrees=45
  twistMotion=Free
 }
 {
  bone1Name=dyn_hair_back_l_04
  bone2Name=dyn_hair_back_m_04
  swing1LimitDegrees=45
  swing1Motion=Free
  swing2LimitDegrees=45
  swing2Motion=Free
  twistLimitDegrees=45
  twistMotion=Free
 }
 {
  bone1Name=dyn_hair_back_l_BASE
  bone2Name=head
  swing1LimitDegrees=45
  swing1Motion=Free
  swing2LimitDegrees=45
  swing2Motion=Free
  twistLimitDegrees=45
  twistMotion=Free
 }
 {bone1Name=dyn_hair_back_l_00, bone2Name=dyn_hair_back_l_BASE, swing1LimitDegrees=5, swing2LimitDegrees=5, twistLimitDegrees=5}
 {
  bone1Name=dyn_hair_back_m_BASE
  bone2Name=head
  swing1LimitDegrees=45
  swing1Motion=Free
  swing2LimitDegrees=45
  swing2Motion=Free
  twistLimitDegrees=45
  twistMotion=Free
 }
 {bone1Name=dyn_hair_back_m_00, bone2Name=dyn_hair_back_m_BASE, swing1LimitDegrees=5, swing2LimitDegrees=5, twistLimitDegrees=2}
 {
  bone1Name=dyn_hair_back_r_BASE
  bone2Name=head
  swing1LimitDegrees=45
  swing1Motion=Free
  swing2LimitDegrees=45
  swing2Motion=Free
  twistLimitDegrees=45
  twistMotion=Free
 }
 {bone1Name=dyn_hair_back_r_00, bone2Name=dyn_hair_back_r_BASE, swing1LimitDegrees=5, swing2LimitDegrees=5, twistLimitDegrees=5}
 {bone1Name=dyn_hair_front_r_03, bone2Name=dyn_hair_front_r_02, swing1LimitDegrees=30, swing2LimitDegrees=20, twistLimitDegrees=3}
 {bone1Name=dyn_hair_front_l_01, bone2Name=head, swing1LimitDegrees=30, swing2LimitDegrees=20, twistLimitDegrees=3}
 {bone1Name=dyn_hair_front_l_02, bone2Name=dyn_hair_front_l_01, swing1LimitDegrees=30, swing2LimitDegrees=20, twistLimitDegrees=3}
 {bone1Name=dyn_hair_front_l_03, bone2Name=dyn_hair_front_l_02, swing1LimitDegrees=30, swing2LimitDegrees=20, twistLimitDegrees=3}
 {bone1Name=dyn_hair_front_r_01, bone2Name=head, swing1LimitDegrees=30, swing2LimitDegrees=20, twistLimitDegrees=3}
 {bone1Name=dyn_hair_front_r_02, bone2Name=dyn_hair_front_r_01, swing1LimitDegrees=30, swing2LimitDegrees=20, twistLimitDegrees=3}
]
