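# Anim Blueprint for the Patches male character's face accessory part, adding secondary physics motion, including hair.
# It copies the attached parent's pose into a Control Rig driven by targeting, crouch, pistol and melee alphas, then nudges a back hair bone while targeting.
# Three blended RigidBody nodes with separate physics assets cover normal play, skydiving without parachute, and front end.
# The update event sets these flags and varies gravity for falling and skydive states.
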AnimBlueprint /BRCosmetics/Characters/Player/Male/Medium/Bodies/M_MED_Patches/Meshes/Parts/M_MED_Patches_FaceAcc_AnimBP
target_skeleton=&/M_MED_Patches_FaceAcc_Skeleton
graphs [
 AnimGraph {nodes [
   n1 = Control Rig!(
    Source=Copy Pose From Mesh() {bUseAttachedParent=true}
    Alpha
    Target:double=Get Targeting_Alpha() -> (Targeting_Alpha:double)
    notCrouchingTargeting:double=Get notCrouchingTargeting() -> (notCrouchingTargeting:double)
    not_pistol:double=Get not_pistol() -> (not_pistol:double)
    pistol:double=Get pistol() -> (pistol:double)
    melee_float:double=Get melee_float() -> (melee_float:double)
    not_crouching:double=Get not_crouching_Alpha() -> (not_crouching_Alpha:double)
   ) -> (Pose) {
    alphaBoolBlend {blendOption=Linear}
    controlRigAssetReference {blueprintRigClass=&/M_MED_Patches_FaceAcc_CtrlRig.M_MED_Patches_FaceAcc_CtrlRig_C}
   }
   n2 = "Transform (Modify) Bone - Bone: dyn_hair_back_r_BASE"!(
    ComponentPose=Local To Component(LocalPose=@n1)
    bAlphaBoolEnabled:bool=Get bIsTargeting()
    Translation="2.918,-1.933,0.709"
    Rotation="0,0,0"
    Scale="1,1,1"
   ) -> (Pose) {
    alphaInputType=Bool
    rotationMode=BMM_Additive
    rotationSpace=BCS_ParentBoneSpace
    scaleMode=BMM_Additive
    scaleSpace=BCS_BoneSpace
    translationMode=BMM_Additive
    translationSpace=BCS_ParentBoneSpace
    alphaBoolBlend {blendInTime=0.5, blendOption=Linear, blendOutTime=0.5}
    boneToModify {boneName=dyn_hair_back_r_BASE}
    translation=<2.918 -1.933 0.709>
   }
   n3 = RigidBody!(
    ComponentPose=@n2
    bAlphaBoolEnabled:bool=Get not_sky() -> (not_sky:bool)
    OverrideWorldGravity:Vector=Get gravity() -> (gravity:Vector)
    ExternalForce
   ) -> (Pose) {
    alphaInputType=Bool
    bClampLinearTranslationLimitToRefPose=true
    bOverrideWorldGravity=true
    overridePhysicsAsset=&/M_MED_Patches_FaceAcc_Physics
    simulationSpace=BaseBoneSpace
    alphaBoolBlend {blendInTime=0.2, blendOption=Linear, blendOutTime=0.2}
    baseBoneRef {boneName=spine_05}
    componentAppliedLinearAccClamp=<5000 3000 3000>
    componentLinearAccScale=<0.35 0.3 0.3>
    componentLinearVelScale=<0.35 0.3 0.3>
    overrideWorldGravity {z=-980}
   }
   n4 = RigidBody(ComponentPose=@n3, bAlphaBoolEnabled=Get sky() -> (sky:bool), OverrideWorldGravity=Get gravity()) {
    alphaInputType=Bool
    bClampLinearTranslationLimitToRefPose=true
    bOverrideWorldGravity=true
    overridePhysicsAsset=&/M_MED_Patches_FaceAcc_Physics_sky
    simulationSpace=BaseBoneSpace
    alphaBoolBlend {blendInTime=0.2, blendOption=Linear, blendOutTime=5}
    baseBoneRef {boneName=head}
    componentAppliedLinearAccClamp=<10 10 10>
    componentLinearAccScale=<0.35 0.3 0.3>
    componentLinearVelScale=<0.35 0.3 0.3>
    overrideWorldGravity {z=-400}
   }
   n5 = RigidBody(ComponentPose=@n4, bAlphaBoolEnabled=Get bIsInFrontEnd(), OverrideWorldGravity="0,0,-980") {
    alphaInputType=Bool
    bClampLinearTranslationLimitToRefPose=true
    bOverrideWorldGravity=true
    overridePhysicsAsset=&/M_MED_Patches_FaceAcc_Physics_FE
    simulationSpace=BaseBoneSpace
    alphaBoolBlend {blendInTime=0.2, blendOption=Linear, blendOutTime=0.2}
    baseBoneRef {boneName=spine_05}
    componentAppliedLinearAccClamp=<5000 3000 3000>
    componentLinearAccScale=<0.35 0.3 0.3>
    componentLinearVelScale=<0.35 0.3 0.3>
    overrideWorldGravity {z=-980}
   }
   Output Pose(Result=Component To Local(ComponentPose=@n5)) {layerGroup=None}
  ]}
 EventGraph {nodes [
   Event BlueprintUpdateAnimation()
   > n1 = Sequence()
   K2Node_Composite_1 = notSky_logic!() -> (ReturnValue:bool=~) {is=title K2Node_Composite}
   > Set not_sky(not_sky:bool=@K2Node_Composite_1) -> (Output_Get:bool)
   K2Node_Composite_0 = sky_logic!() -> (ReturnValue:bool=~) {is=title K2Node_Composite}
   > Set sky(sky:bool=@K2Node_Composite_0) -> (Output_Get:bool)
   > Set Targeting_Alpha(Targeting_Alpha:double="To Float (Boolean)"(InBool=Get bIsTargeting())) -> (Output_Get:double)
   n2 = NOT Boolean(A=Get bIsCrouching())
   n3 = Get bIsTargeting()
   > Set notCrouchingTargeting(notCrouchingTargeting:double="To Float (Boolean)"(InBool=AND Boolean(A=@n2, B=@n3))) -> (Output_Get:double)
   n4 = Get bIsCrouching()
   > Set not_crouching_Alpha(not_crouching_Alpha:double="To Float (Boolean)"(InBool=NOT Boolean(A=@n4))) -> (Output_Get:double)
   n5 = "Equal (Enum)"(
    A:EFortWeaponCoreAnimation Enum=GetWeaponCoreAnim!(self:Fort Player Animation Interface Object Reference=GetOwnerAnimationInterface!(self:Custom Character Part Anim Instance Object Reference) -> (ReturnValue:Fort Player Animation Interface Object Reference)) -> (ReturnValue:EFortWeaponCoreAnimation Enum)
    B:EFortWeaponCoreAnimation Enum=Pistol
   )
   > Set not_pistol(not_pistol:double="To Float (Boolean)"(InBool=NOT Boolean(A=@n5))) -> (Output_Get:double)
   > Set pistol(pistol:double="To Float (Boolean)"(InBool=@n5)) -> (Output_Get:double)
   > Set melee_float(melee_float:double="To Float (Boolean)"(InBool=Get bIsPlayingMeleeAnim())) -> (Output_Get:double)
   > n6 = Branch(Condition=Get bIsFalling())
   > Set gravity(gravity:Vector="0,0,-1200") -> (Output_Get:Vector)
   Set gravity(execute=@n6.else, gravity="0,0,-980")
   Branch(execute=@n1.then_1, Condition=Get bIsSkyDiving())
   > n7 = Gravity Override State Function DEPRECATED(CharacterPartAnimInstance=Self-Reference(), GameTimeInSeconds=GetGameTimeInSeconds(), GravityOverrideParams=Get Gravity Override Params())
   > Set gravity(gravity=@n7)
  ]}
 sky_logic {
  path=EventGraph.K2Node_Composite_0.sky_logic
  nodes [
   Inputs!()
   n1 = Get bIsParachuteOpened()
   Outputs!(ReturnValue:bool=AND Boolean(A=Get bIsSkyDiving(), B=NOT Boolean(A=@n1)))
  ]
 }
 notSky_logic {
  path=EventGraph.K2Node_Composite_1.notSky_logic
  nodes [
   Inputs()
   n1 = Get bIsInFrontEnd()
   n2 = NOT Boolean(A=Get bIsSkyDiving())
   n3 = Get bIsParachuteOpened()
   Outputs(ReturnValue=AND Boolean(A=OR Boolean(A=@n3, B=@n2), B=NOT Boolean(A=@n1)))
  ]
 }
]
variables [
 Current State Out {type=string, default=""}
 gravity {type=Vector}
 Gravity Override Params {
  type=GravityOverrideParamsStruct
  default {
   skydive_DivingDown {gravityOverride=<-1500 500 0>}
   skydive_FloatingDown {gravityOverride=<0 -700 1500>, windAmplitude=<200 -500 600>}
   skydive_Parachuting {gravityOverride=<0 0 -980>, windAmplitude=<0 0 0>, windFrequency=<0 0 0>}
   skydive_Surfing {gravityOverride=<500 -300 0>, windAmplitude=<200 -100 1>, windFrequency=<1.5 1 1>}
  }
 }
 melee_float {type=number}
 not_crouching_Alpha {type=number}
 not_pistol {type=number}
 not_sky {type=boolean}
 notCrouchingTargeting {type=number}
 pistol {type=number}
 sky {type=boolean}
 Targeting_Alpha {type=number}
]
